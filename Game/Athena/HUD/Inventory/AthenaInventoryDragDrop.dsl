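; Athena inventory drag-and-drop asset that controls the drop icon shown on the inventory drag visual.
; Its ShowDropIcon function takes an Athena drag-drop action enum, casts the default drag visual to AthenaInventoryDragVisual and forwards the action to that widget's ShowDropIcon.
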(fn ShowDropIcon
  (params (Drop%20Action :type AthenaDragDropAction%20Enum :default "NewEnumerator0"))
  (bind _asathena_inventory_drag_visual (Utilities|Casting|CastToAthenaInventoryDragVisual (Variables|DragandDrop|GetDefaultDragVisual)))
  (Class|AthenaInventoryDragVisual|ShowDropIcon _asathena_inventory_drag_visual Drop%20Action))
