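# Athena HUD inventory widget that draws one small item tier pip as a star, derived from AthenaItemTierPipWidget.
# Its only widget is the Star image, which uses the M_UI_HUD_TierStar_Base material as its brush.
# SetRarityColor takes a linear color, gets the image's dynamic material and writes it to the Rarity_Color vector parameter.
# The event graph holds only empty PreConstruct, Construct and Tick events.
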WidgetBlueprint /Game/Athena/HUD/Inventory/AthenaItemTierPip
graphs [
 EventGraph {nodes [
   Event PreConstruct()
   Event Construct()
   Event Tick()
  ]}
 SetRarityColor {nodes [
   n1 = SetRarityColor!() -> (then, RarityColor:Linear Color Structure) {is=title K2Node_FunctionEntry}
   > n2 = GetDynamicMaterial(self=Get Star() -> (Star:Image Object Reference))
   > SetVectorParameterValue(self=@n2, ParameterName=Rarity_Color, Value=@n1)
  ]}
]
variables []
widgets {
 info {inheritedWidgetCount=0, parentClass=/Script/FortniteUI.AthenaItemTierPipWidget, rootWidgetClass=&Image, widgetCount=1}
 widgets [Star {
   widgetClassPath=&Image
   bIsVariable=true
   widget=Star
   properties {
    visibility=SelfHitTestInvisible
    brush {
     bIsDynamicallyLoaded=false
     drawAs=Image
     imageType=NoImage
     mirroring=NoMirror
     resourceName=None
     resourceObject=/Game/Athena/HUD/Inventory/Art/M_UI_HUD_TierStar_Base
     tiling=NoTile
     imageSize=<8 8>
     margin=<ltrb 0 0 0 0>
     outlineSettings {
      bUseBrushTransparency=false
      roundingType=HalfHeightRadius
      width=0
      color {colorUseRule=UseColor_Specified, specifiedColor=<rgba 0 0 0 0>}
      cornerRadii=<0 0 0 0>
     }
     tintColor {colorUseRule=UseColor_Specified, specifiedColor=<rgba 1 1 1 1>}
     uVRegion {bIsValid=false, max=<0 0>, min=<0 0>}
    }
   }
  }]
}
