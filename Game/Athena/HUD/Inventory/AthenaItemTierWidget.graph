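# HUD inventory widget showing an item's tier as a row of pips over a rarity-colored backing, built on the native AthenaItemTierWidget.
# Tier updates collapse it for No_Tier; rarity updates set the backing's dynamic material colors.
# SetRarityOverride forces it visible, recolors it and rebuilds the DynamicEntryBox pips from the rarity's enum value.
# PreConstruct applies the optional backing's height and padding; a Selected animation animates its material's Animate_Color.
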WidgetBlueprint /Game/Athena/HUD/Inventory/AthenaItemTierWidget
animations [Selected {
  binding_count=1
  length_seconds=0.1
  object=&self:Selected
  sequence {bindings [Backing {
     id=#1
     possessed_class=Image
     tracks [{
       type=MovieSceneWidgetMaterialTrack
       sections [{type=MovieSceneParameterSection, channels [Animate_Color {type=MovieSceneScriptingFloatChannel, keys [[0,0],[2,1]]}]}]
      }]
    }]}
 }]
graphs [
 EventGraph {nodes [
   n1 = SetRarityOverride!() -> (OutputDelegate, then, Rarity:EFortRarity Enum=Common) {is=title K2Node_CustomEvent}
   > SetVisibility(self=Get RootSizeBox() -> (RootSizeBox:Size Box Object Reference), InVisibility=SelfHitTestInvisible)
   > n2 = GetRarityData!(execute, Rarity:EFortRarity Enum=@n1.Rarity) -> (then, ReturnValue:Fort Rarity Item Data Structure)
   > SetRarityData!(execute, self, FortRarityItemData:"Fort Rarity Item Data Structure (by ref)"=@n2) -> (then)
   > Reset!(
    execute
    self:Dynamic Entry Box Object Reference=Get EntryBox_TierPips() -> (EntryBox_TierPips:Dynamic Entry Box Object Reference)
    bDeleteWidgets:bool=false
   ) -> (then)
   > For Loop(LastIndex="To Integer (Byte)"!(InByte:Byte=@n1.Rarity) -> (ReturnValue:int=0))
   > Create Entry!(execute, self:Dynamic Entry Box Object Reference=Get EntryBox_TierPips()) -> (then, ReturnValue:User Widget Object Reference)
   n3 = Event OnItemTierUpdated!() -> (OutputDelegate, then, ItemTier:EFortItemTier Enum=No_Tier) {is=title K2Node_Event}
   > SetVisibility(
    self=Get RootSizeBox()
    InVisibility=Select(Option 0=SelfHitTestInvisible, Option 1=Collapsed, Index="Equal (Enum)"(A:EFortItemTier Enum=@n3.ItemTier, B:EFortItemTier Enum=No_Tier))
   )
   n4 = Event PreConstruct()
   > Parent: PreConstruct(IsDesignTime=@n4.IsDesignTime)
   > n5 = Branch(Condition=Get ShowBacking() -> (ShowBacking:bool))
   > SetVisibility(self=Get Backing() -> (Backing:Image Object Reference), InVisibility=SelfHitTestInvisible)
   n6 = Get RootSizeBox()
   > SetHeightOverride(self=@n6, InHeightOverride=Get BackingHeightOverride() -> (BackingHeightOverride:float))
   > SetPadding!(
    execute
    self:Overlay Slot Object Reference=SlotAsOverlaySlot(Widget=Get StarsRarityTint() -> (StarsRarityTint:Border Object Reference))
    InPadding:Margin Structure=Get BackingStarsPadding() -> (BackingStarsPadding:Margin Structure)
   ) -> (then)
   SetVisibility(execute=@n5.else, self=Get Backing(), InVisibility=Collapsed)
   > ClearHeightOverride(self:Size Box Object Reference=@n6)
   > SetPadding(self=SlotAsOverlaySlot(Widget=Get StarsRarityTint()), InPadding=Make Margin(Left=0, Top=0, Right=0, Bottom=0))
   n7 = Event OnItemRarityUpdated!() -> (OutputDelegate, then, ItemRarity:EFortRarity Enum=Common, RarityItemData:Fort Rarity Item Data Structure) {is=title K2Node_Event}
   > SetRarityData(FortRarityItemData=@n7.RarityItemData)
   Get EntryBox_TierPips()
   Get BackingStarsPadding()
  ]}
 SetRarityData {nodes [
   n1 = SetRarityData!() -> (then, FortRarityItemData:"Fort Rarity Item Data Structure (by ref)") {is=title K2Node_FunctionEntry}
   > n2 = GetDynamicMaterial(self=Get Backing())
   n3 = Break Fort Rarity Item Data!(FortRarityItemData:"Fort Rarity Item Data Structure (by ref)"=@n1) -> (
    Name:Text
    Color1:Linear Color Structure
    Color2:Linear Color Structure
    Color3:Linear Color Structure
    Color4:Linear Color Structure
    Color5:Linear Color Structure
    Radius:float=~
    Falloff:float=~
    Brightness:float=~
    Roughness:float=~
    Glow:float=~
   ) {is=title K2Node_BreakStruct}
   > SetVectorParameterValue(self=@n2, ParameterName=Inner_Color, Value=@n3.Color3)
   > SetVectorParameterValue(self=@n2, ParameterName=Outer_Color, Value=@n3.Color2)
  ]}
]
variables [
 BackingHeightOverride {type=number, default=16}
 BackingStarsPadding {type=Margin, default=<ltrb 4 0 4 0>}
 ShowBacking {type=boolean, default=true}
]
widgets {
 info {inheritedWidgetCount=1, parentClass=/Script/FortniteUI.AthenaItemTierWidget, rootWidgetClass=&SizeBox, widgetCount=5}
 widgets [
  RootSizeBox {widgetClassPath=&SizeBox, bIsVariable=true, widget=RootSizeBox, properties {bOverride_HeightOverride=true, heightOverride=16}}
  Overlay_39 {widgetClassPath=&Overlay, parent=RootSizeBox, widget=Overlay_39, slot {class=SizeBoxSlot}}
  Backing {
   widgetClassPath=&Image
   bIsVariable=true
   parent=Overlay_39
   widget=Backing
   properties {
    visibility=SelfHitTestInvisible
    brush {
     bIsDynamicallyLoaded=false
     drawAs=Box
     imageType=NoImage
     mirroring=NoMirror
     resourceName=None
     resourceObject=/Game/Athena/HUD/Inventory/Art/M_UI_HUD_RarityTier_Backing_Base
     tiling=NoTile
     imageSize=<8 8>
     margin=<ltrb 0.5 0.5 0.5 0.5>
     outlineSettings {
      bUseBrushTransparency=false
      roundingType=HalfHeightRadius
      width=0
      color {colorUseRule=UseColor_Specified, specifiedColor=<rgba 0 0 0 0>}
      cornerRadii=<0 0 0 0>
     }
     tintColor {colorUseRule=UseColor_Specified, specifiedColor=<rgba 1 1 1 1>}
     uVRegion {bIsValid=false, max=<0 0>, min=<0 0>}
    }
   }
   slot {class=OverlaySlot, properties {horizontalAlignment=HAlign_Fill, verticalAlignment=VAlign_Fill}}
  }
  StarsRarityTint {
   widgetClassPath=&Border
   bIsVariable=true
   parent=Overlay_39
   widget=StarsRarityTint
   properties {
    background {
     bIsDynamicallyLoaded=false
     drawAs=NoDrawType
     imageType=NoImage
     mirroring=NoMirror
     resourceName=None
     resourceObject=null
     tiling=NoTile
     imageSize=<32 32>
     margin=<ltrb 0 0 0 0>
     outlineSettings {
      bUseBrushTransparency=false
      roundingType=HalfHeightRadius
      width=0
      color {colorUseRule=UseColor_Specified, specifiedColor=<rgba 0 0 0 0>}
      cornerRadii=<0 0 0 0>
     }
     tintColor {colorUseRule=UseColor_Specified, specifiedColor=<rgba 1 1 1 1>}
     uVRegion {bIsValid=false, max=<0 0>, min=<0 0>}
    }
    padding=<ltrb 0 0 0 0>
   }
   slot {class=OverlaySlot, properties {horizontalAlignment=HAlign_Center, verticalAlignment=VAlign_Center, padding=<ltrb 4 0 4 0>}}
  }
  EntryBox_TierPips {
   widgetClassPath=&DynamicEntryBox
   bInherited=true
   parent=StarsRarityTint
   widget=EntryBox_TierPips
   properties {entryWidgetClass=&/AthenaItemTierPip.AthenaItemTierPip_C}
   slot {class=BorderSlot, properties {padding=<ltrb 0 0 0 0>}}
  }
 ]
}
