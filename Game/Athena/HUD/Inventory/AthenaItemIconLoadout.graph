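# Loadout item icon widget for the Athena inventory HUD, extending AthenaItemIcon with an item card, shadow, selection outline and info panel for stack count, ammo, recharge and tier.
# PreConstruct toggles panel elements from show flags; rarity updates push rarity colors into the icon, plate, outline and font dynamic materials.
# Hover, selection and recipe-hover states play short animations forward or reverse, with focus driving selection and a shared helper rewinding active ones.
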WidgetBlueprint /Game/Athena/HUD/Inventory/AthenaItemIconLoadout
animations [
 Hover {
  binding_count=2
  length_seconds=0.1
  object=&self:Hover
  sequence {bindings [
    All {
     id=#1
     possessed_class=Overlay
     tracks [{
       type=MovieScene2DTransformTrack
       sections [{
         type=MovieScene2DTransformSection
         channels [
          Translation.X {type=MovieSceneScriptingFloatChannel}
          Translation.Y {type=MovieSceneScriptingFloatChannel, keys [[0,0],[2,-4]]}
         ]
        }]
      }]
    }
    Shadow {
     id=#2
     possessed_class=Image
     tracks [
      {
       type=MovieSceneWidgetMaterialTrack
       sections [{type=MovieSceneParameterSection, channels [Animate_Alpha {type=MovieSceneScriptingFloatChannel, keys [[0,0],[2,1]]}]}]
      }
      {
       type=MovieScene2DTransformTrack
       sections [{
         type=MovieScene2DTransformSection
         channels [
          Translation.Y {type=MovieSceneScriptingFloatChannel, keys [[0,0],[2,16]]}
          Scale.X {type=MovieSceneScriptingFloatChannel, keys [[0,1.25],[2,1.5]]}
          Scale.Y {type=MovieSceneScriptingFloatChannel, keys [[0,1.25],[2,1.5]]}
         ]
        }]
      }
     ]
    }
   ]}
 }
 Selected_NoOutline {
  binding_count=6
  length_seconds=0.1
  object=&self:Selected_NoOutline
  sequence {bindings [
    HorizontalBox_168 {
     id=#3
     possessed_class=HorizontalBox
     tracks [{
       type=MovieScene2DTransformTrack
       sections [{
         type=MovieScene2DTransformSection
         channels [
          Translation.X {type=MovieSceneScriptingFloatChannel}
          Translation.Y {type=MovieSceneScriptingFloatChannel, keys [[0,0],[2,-2]]}
         ]
        }]
      }]
    }
    InfoPanel_Backing {
     id=#4
     possessed_class=Image
     tracks [
      {
       type=MovieSceneWidgetMaterialTrack
       sections [{
         type=MovieSceneParameterSection
         channels [
          Animate_Alpha {type=MovieSceneScriptingFloatChannel, keys [[0,0],[2,1]]}
          Animate_Color_Hover {type=MovieSceneScriptingFloatChannel, keys [[0,0],[2,1]]}
          Offset_Angle_Top {type=MovieSceneScriptingFloatChannel, keys [[0,-0.1],[2,-0.125]]}
          Plate_Angle {type=MovieSceneScriptingFloatChannel, keys [[0,0.03],[2,0.035]]}
         ]
        }]
      }
      {
       type=MovieScene2DTransformTrack
       sections [{
         type=MovieScene2DTransformSection
         channels [
          Translation.X {type=MovieSceneScriptingFloatChannel, const [0,1]}
          Translation.Y {type=MovieSceneScriptingFloatChannel, const [0,1]}
          Angle {type=MovieSceneScriptingFloatChannel, const [0,1]}
          Scale.X {type=MovieSceneScriptingFloatChannel, const [1,1]}
          Scale.Y {type=MovieSceneScriptingFloatChannel, keys [[0,1],[2,1.1]]}
          Shear.X {type=MovieSceneScriptingFloatChannel, const [0,1]}
          Shear.Y {type=MovieSceneScriptingFloatChannel, const [0,1]}
         ]
        }]
      }
     ]
    }
    Text_StackCount {
     id=#5
     possessed_class=CommonTextBlock
     tracks [{
       type=MovieSceneWidgetMaterialTrack
       display_name=Font.FontMaterial
       sections [{type=MovieSceneParameterSection, channels [Animate_Color {type=MovieSceneScriptingFloatChannel, keys [[0,0],[2,1]]}]}]
      }]
    }
    Image_AmmoIcon {
     id=#6
     possessed_class=Image
     tracks [{
       type=MovieSceneWidgetMaterialTrack
       sections [{type=MovieSceneParameterSection, channels [Animate_Color {type=MovieSceneScriptingFloatChannel, keys [[0,0],[2,1]]}]}]
      }]
    }
    Image_ItemIcon {
     id=#7
     possessed_class=Image
     tracks [{
       type=MovieSceneWidgetMaterialTrack
       sections [{
         type=MovieSceneParameterSection
         channels [
          Animate_Alpha {type=MovieSceneScriptingFloatChannel, keys [[0,0],[2,1]]}
          Animate_ItemIcon_Scale {type=MovieSceneScriptingFloatChannel, keys [[0,0],[2,1]]}
          Animate_ItemIcon_Translate_Y {type=MovieSceneScriptingFloatChannel, keys [[0,0],[2,1]]}
          Animate_Hotspot_Color {type=MovieSceneScriptingFloatChannel, keys [[0,0],[2,0.333]]}
          BG_Gradient_Density {type=MovieSceneScriptingFloatChannel, const [2,2]}
         ]
        }]
      }]
    }
    Image_RechargeableIcon {
     id=#8
     possessed_class=Image
     tracks [{
       type=MovieSceneWidgetMaterialTrack
       sections [{
         type=MovieSceneParameterSection
         end_frame=2
         start_frame=0
         channels [Animate_Color {type=MovieSceneScriptingFloatChannel, keys [[0,0],[2,1]]}]
        }]
      }]
    }
   ]}
 }
 RecipeHover {
  binding_count=1
  length_seconds=0.1
  object=&self:RecipeHover
  sequence {bindings [Image_ItemIcon {
     id=#9
     possessed_class=Image
     tracks [{
       type=MovieSceneWidgetMaterialTrack
       sections [{
         type=MovieSceneParameterSection
         channels [
          Animate_Alpha {type=MovieSceneScriptingFloatChannel, keys [[0,0],[2,1]]}
          Animate_Hotspot_Color {type=MovieSceneScriptingFloatChannel, keys [[0,0],[2,0.333]]}
          Animate_ItemIcon_Scale {type=MovieSceneScriptingFloatChannel, keys [[0,0],[2,1]]}
         ]
        }]
      }]
    }]}
 }
]
graphs [
 EventGraph {nodes [
   n1 = Event OnAmmoIconUpdated!() -> (OutputDelegate, then, bHasAmmoIcon:bool) {is=title K2Node_Event}
   > SetVisibility(
    self=Get Image_AmmoIcon() -> (Image_AmmoIcon:Image Object Reference)
    InVisibility=Select(Option 0=Collapsed, Option 1=SelfHitTestInvisible, Index=@n1.bHasAmmoIcon)
   )
   > UpdateRechargeIcon!(
    execute
    self:Athena Item Icon Object Reference
    ItemToRepresent:Fort Item Object Reference=GetItemToRepresent!(self:Athena Item Element Widget Base Object Reference) -> (ReturnValue:Fort Item Object Reference)
   ) -> (then)
   Event PreConstruct()
   > SetVisibility(
    self=Get Image_AmmoIcon()
    InVisibility=Select(Option 0=Collapsed, Option 1=SelfHitTestInvisible, Index=Get bShowAmmoIcon() -> (bShowAmmoIcon:bool))
   )
   > SetVisibility(
    self=Get Text_StackCount() -> (Text_StackCount:Common Text Object Reference)
    InVisibility=Select(Option 0=Collapsed, Option 1=SelfHitTestInvisible, Index=Get bShowStackCountText() -> (bShowStackCountText:bool))
   )
   > SetVisibility(
    self=Get ItemTierWidget() -> (ItemTierWidget:Athena Item Tier Widget Object Reference)
    InVisibility=Select(Option 0=Collapsed, Option 1=SelfHitTestInvisible, Index=Get bShowItemTier() -> (bShowItemTier:bool))
   )
   n2 = Get bShowAmmoIcon()
   n3 = Get bShowStackCountText()
   > SetVisibility(
    self=[Get InfoPanel_SecondaryItem() -> (InfoPanel_SecondaryItem:Overlay Object Reference), Get InfoPanel() -> (InfoPanel:Overlay Object Reference)]
    InVisibility=Select(Option 0=Collapsed, Option 1=SelfHitTestInvisible, Index=OR Boolean(A=@n2, B=@n3))
   )
   Event OnFocusLost!() -> (OutputDelegate, then, InFocusEvent:Focus Event Structure) {is=title K2Node_Event}
   > SetSelectedState!(execute, self, bIsSelected:bool=false) -> (then)
   n4 = Event OnItemRarityUpdated!() -> (OutputDelegate, then, ItemRarity:EFortRarity Enum=Common, RarityItemData:Fort Rarity Item Data Structure) {is=title K2Node_Event}
   > n5 = GetDynamicMaterial(self=Get Image_ItemIcon() -> (Image_ItemIcon:Image Object Reference))
   n6 = Break Fort Rarity Item Data!(FortRarityItemData:"Fort Rarity Item Data Structure (by ref)"=@n4.RarityItemData) -> (
    Name:Text
    Color1:Linear Color Structure
    Color2:Linear Color Structure
    Color3:Linear Color Structure
    Color4:Linear Color Structure
    Color5:Linear Color Structure
    Radius:float=~
    Falloff:float=~
    Brightness:float=~
    Roughness:float=~
    Glow:float=~
   ) {is=title K2Node_BreakStruct}
   > SetVectorParameterValue(self=@n5, ParameterName=BG_Hotspot_Color_Hover, Value=@n6.Color1)
   > SetVectorParameterValue(self=@n5, ParameterName=BG_Hotspot_Color, Value=@n6.Color2)
   > SetVectorParameterValue(self=@n5, ParameterName=BG_Falloff_Color, Value=@n6.Color3)
   > n7 = GetDynamicMaterial(self=Get InfoPanel_Backing() -> (InfoPanel_Backing:Image Object Reference))
   > SetVectorParameterValue(self=@n7, ParameterName=Plate_Color_Start, Value=@n6.Color2)
   > n8 = GetDynamicMaterial(self=Get Image_AmmoIcon())
   > SetVectorParameterValue(self=@n8, ParameterName=Color_Stop, Value=@n6.Color3)
   > n9 = GetDynamicFontMaterial!(execute, self:Text Object Reference=Get Text_StackCount()) -> (then, ReturnValue:Material Instance Dynamic Object Reference)
   > SetVectorParameterValue(self=@n9, ParameterName=Color_Stop, Value=@n6.Color3)
   > n10 = GetDynamicMaterial(self=Get SelectionOutline() -> (SelectionOutline:Image Object Reference))
   > SetVectorParameterValue(self=@n10, ParameterName=Color_Stop, Value=@n6.Color3)
   > n11 = GetDynamicMaterial(self=Get Image_RechargeableIcon() -> (Image_RechargeableIcon:Image Object Reference))
   > SetVectorParameterValue(self=@n11, ParameterName=Color_Stop, Value=@n6.Color3)
  ]}
 OnFocusReceived {nodes [
   OnFocusReceived!() -> (then, MyGeometry:Geometry Structure, InFocusEvent:Focus Event Structure) {is=title K2Node_FunctionEntry}
   > SetSelectedState(bIsSelected=true)
   > Return Node!(execute, ReturnValue:Event Reply Structure=Handled())
  ]}
 RecipeHoverState {nodes [
   n1 = RecipeHoverState!() -> (then, IsHovered:bool=~) {is=title K2Node_FunctionEntry}
   > n2 = Set IsRecipeHoveredAnimActive(IsRecipeHoveredAnimActive:bool=@n1) -> (Output_Get:bool)
   > n3 = Branch(Condition=@n2)
   > PlayAnimationForward(InAnimation=Get RecipeHover() -> (RecipeHover:Widget Animation Object Reference))
   PlayAnimationReverse(execute=@n3.else, InAnimation=Get RecipeHover())
  ]}
 ResetAnimToStart {nodes [
   n1 = ResetAnimToStart!() -> (then, IsAnimStateActive:"Boolean (by ref)", Animation:Widget Animation Object Reference) {is=title K2Node_FunctionEntry}
   > Branch(Condition=OR Boolean(A=@n1, B=IsAnimationPlaying(InAnimation=@n1.Animation)))
   > "Set Boolean (by ref)"!(execute:Exec, Target:"Boolean (by ref)"=@n1, Value:bool) -> (then:Exec) {is=title K2Node_VariableSetRef}
   > PlayAnimation(InAnimation=@n1.Animation, StartAtTime=GetEndTime(self=@n1.Animation), PlayMode=Reverse)
  ]}
 ResetAnims {nodes [
   ResetAnims() {is=title K2Node_FunctionEntry}
   > ResetRecipeHoverState()
   > ResetSelectedState()
   > ResetHoveredState()
  ]}
 ResetHoveredState {nodes [
   ResetHoveredState() {is=title K2Node_FunctionEntry}
   > ResetAnimToStart!(
    execute
    self
    IsAnimStateActive:"Boolean (by ref)"=Get IsHoveredAnimActive() -> (IsHoveredAnimActive:bool)
    Animation:Widget Animation Object Reference=Get Hover() -> (Hover:Widget Animation Object Reference)
   ) -> (then) {is=title K2Node_CallFunction}
  ]}
 ResetRecipeHoverState {nodes [
   ResetRecipeHoverState() {is=title K2Node_FunctionEntry}
   > ResetAnimToStart(IsAnimStateActive=Get IsRecipeHoveredAnimActive() -> (IsRecipeHoveredAnimActive:bool), Animation=Get RecipeHover())
  ]}
 ResetSelectedState {nodes [
   ResetSelectedState() {is=title K2Node_FunctionEntry}
   > ResetAnimToStart(
    IsAnimStateActive=Get IsSelectedAnimActive() -> (IsSelectedAnimActive:bool)
    Animation=Get Selected_NoOutline() -> (Selected_NoOutline:Widget Animation Object Reference)
   )
  ]}
 SetHoveredState {nodes [
   n1 = SetHoveredState!() -> (then, bIsHovered:bool=~) {is=title K2Node_FunctionEntry}
   > n2 = Set IsHoveredAnimActive(IsHoveredAnimActive:bool=@n1) -> (Output_Get:bool)
   > n3 = Branch(Condition=@n2)
   n4 = Get Hover()
   > PlayAnimationForward(InAnimation=@n4)
   PlayAnimationReverse(execute=@n3.else, InAnimation=@n4)
  ]}
 SetSelectedState {nodes [
   n1 = SetSelectedState!() -> (then, bIsSelected:bool=~) {is=title K2Node_FunctionEntry}
   > n2 = Set IsSelectedAnimActive(IsSelectedAnimActive:bool=@n1) -> (Output_Get:bool)
   > n3 = Branch(Condition=@n2)
   n4 = Get Selected_NoOutline()
   > PlayAnimationForward(InAnimation=@n4)
   PlayAnimationReverse(execute=@n3.else, InAnimation=@n4)
  ]}
]
variables [
 bShowAmmoIcon {type=boolean, default=true}
 bShowItemTier {type=boolean, default=true}
 bShowStackCountText {type=boolean, default=true}
 IsHoveredAnimActive {type=boolean}
 IsRecipeHoveredAnimActive {type=boolean}
 IsSelectedAnimActive {type=boolean}
]
widgets {
 info {inheritedWidgetCount=8, parentClass=/Script/FortniteUI.AthenaItemIcon, rootWidgetClass=&SizeBox, widgetCount=19}
 widgets [
  SizeBox_41 {
   widgetClassPath=&SizeBox
   widget=SizeBox_41
   properties {bOverride_HeightOverride=true, bOverride_WidthOverride=true, heightOverride=128, widthOverride=128}
  }
  All {widgetClassPath=&Overlay, parent=SizeBox_41, widget=All, slot {class=SizeBoxSlot}}
  Shadow {
   widgetClassPath=&Image
   bIsVariable=true
   parent=All
   widget=Shadow
   properties {
    clipping=ClipToBoundsWithoutIntersecting
    visibility=SelfHitTestInvisible
    brush {
     bIsDynamicallyLoaded=false
     drawAs=Box
     imageType=NoImage
     mirroring=NoMirror
     resourceName=None
     resourceObject=/Game/Athena/HUD/Inventory/Art/M_UI_Inventory_Loadout_DropShadow
     tiling=NoTile
     imageSize=<80 80>
     margin=<ltrb 0.5 0.5 0.5 0.5>
     outlineSettings {
      bUseBrushTransparency=false
      roundingType=HalfHeightRadius
      width=0
      color {colorUseRule=UseColor_Specified, specifiedColor=<rgba 0 0 0 0>}
      cornerRadii=<0 0 0 0>
     }
     tintColor {colorUseRule=UseColor_Specified, specifiedColor=<rgba 1 1 1 1>}
     uVRegion {bIsValid=false, max=<0 0>, min=<0 0>}
    }
    renderTransform {angle=0, scale=<1.5 1.5>, shear=<0 0>, translation=<0 16>}
   }
   slot {class=OverlaySlot, properties {horizontalAlignment=HAlign_Fill, verticalAlignment=VAlign_Fill}}
  }
  SelectionOutline {
   widgetClassPath=&Image
   bIsVariable=true
   parent=All
   widget=SelectionOutline
   properties {
    visibility=SelfHitTestInvisible
    brush {
     bIsDynamicallyLoaded=false
     drawAs=Image
     imageType=NoImage
     mirroring=NoMirror
     resourceName=None
     resourceObject=/Game/Athena/HUD/Inventory/Art/M_UI_LoadoutItem_Selection_Outline
     tiling=NoTile
     imageSize=<32 32>
     margin=<ltrb 0 0 0 0>
     outlineSettings {
      bUseBrushTransparency=false
      roundingType=HalfHeightRadius
      width=0
      color {colorUseRule=UseColor_Specified, specifiedColor=<rgba 0 0 0 0>}
      cornerRadii=<0 0 0 0>
     }
     tintColor {colorUseRule=UseColor_Specified, specifiedColor=<rgba 1 1 1 1>}
     uVRegion {bIsValid=false, max=<0 0>, min=<0 0>}
    }
    renderTransform {angle=0, scale=<1.075 1.075>, shear=<0 0>, translation=<0 0>}
   }
   slot {class=OverlaySlot, properties {horizontalAlignment=HAlign_Fill, verticalAlignment=VAlign_Fill}}
  }
  ItemCard {
   widgetClassPath=&Overlay
   parent=All
   widget=ItemCard
   slot {class=OverlaySlot, properties {horizontalAlignment=HAlign_Fill, verticalAlignment=VAlign_Fill}}
  }
  Image_ItemIcon {
   widgetClassPath=&Image
   bInherited=true
   parent=ItemCard
   widget=Image_ItemIcon
   properties {brush {
     bIsDynamicallyLoaded=false
     drawAs=Image
     imageType=NoImage
     mirroring=NoMirror
     resourceName=None
     resourceObject=/Game/Athena/HUD/Inventory/Art/M_UI_HUD_Inventory_ItemCard_Base
     tiling=NoTile
     imageSize=<32 32>
     margin=<ltrb 0 0 0 0>
     outlineSettings {
      bUseBrushTransparency=false
      roundingType=HalfHeightRadius
      width=0
      color {colorUseRule=UseColor_Specified, specifiedColor=<rgba 0 0 0 0>}
      cornerRadii=<0 0 0 0>
     }
     tintColor {colorUseRule=UseColor_Specified, specifiedColor=<rgba 1 1 1 1>}
     uVRegion {bIsValid=false, max=<0 0>, min=<0 0>}
    }}
   slot {class=OverlaySlot, properties {horizontalAlignment=HAlign_Fill, verticalAlignment=VAlign_Fill}}
  }
  InfoPanel_SecondaryItem {
   widgetClassPath=&Overlay
   bInherited=true
   bIsVariable=true
   parent=ItemCard
   widget=InfoPanel_SecondaryItem
   properties {visibility=Collapsed}
   slot {class=OverlaySlot, properties {horizontalAlignment=HAlign_Fill}}
  }
  HorizontalBox {
   widgetClassPath=&HorizontalBox
   parent=InfoPanel_SecondaryItem
   widget=HorizontalBox
   slot {class=OverlaySlot, properties {horizontalAlignment=HAlign_Center, verticalAlignment=VAlign_Fill}}
  }
  Text_SecondaryStackCount {
   widgetClassPath=&CommonTextBlock
   bInherited=true
   parent=HorizontalBox
   widget=Text_SecondaryStackCount
   properties {
    mobileFontSizeMultiplier=1.2
    text=999
    font {
     bForceMonospaced=false
     bMaterialIsStencil=false
     fontMaterial=/Game/Athena/HUD/Inventory/Art/M_UI_Inventory_Loadout_FontColor
     fontObject=&BurbankSmall
     letterSpacing=0
     monospacedWidth=1
     size=15
     skewAmount=0
     typefaceFontName=Black
     outlineSettings {
      bApplyOutlineToDropShadows=false
      bMiteredCorners=false
      bSeparateFillAlpha=false
      outlineMaterial=null
      outlineSize=0
      outlineColor=<rgba 0 0 0 1>
     }
    }
    shadowOffset=<0 0>
   }
   slot {class=HorizontalBoxSlot, properties {horizontalAlignment=HAlign_Center, verticalAlignment=VAlign_Center}}
  }
  Image_SecondaryAmmoIcon {
   widgetClassPath=&Image
   bInherited=true
   parent=HorizontalBox
   widget=Image_SecondaryAmmoIcon
   properties {
    visibility=SelfHitTestInvisible
    brush {
     bIsDynamicallyLoaded=false
     drawAs=Image
     imageType=NoImage
     mirroring=NoMirror
     resourceName=None
     resourceObject=/Game/Athena/HUD/Inventory/Art/M_UI_Inventory_Loadout_AmmoIcon
     tiling=NoTile
     imageSize=<22 22>
     margin=<ltrb 0 0 0 0>
     outlineSettings {
      bUseBrushTransparency=false
      roundingType=HalfHeightRadius
      width=0
      color {colorUseRule=UseColor_Specified, specifiedColor=<rgba 0 0 0 0>}
      cornerRadii=<0 0 0 0>
     }
     tintColor {colorUseRule=UseColor_Specified, specifiedColor=<rgba 1 1 1 1>}
     uVRegion {bIsValid=false, max=<0 0>, min=<0 0>}
    }
   }
   slot {class=HorizontalBoxSlot, properties {horizontalAlignment=HAlign_Center, verticalAlignment=VAlign_Center}}
  }
  InfoPanel {
   widgetClassPath=&Overlay
   bIsVariable=true
   parent=ItemCard
   widget=InfoPanel
   slot {class=OverlaySlot, properties {horizontalAlignment=HAlign_Fill, verticalAlignment=VAlign_Bottom}}
  }
  InfoPanel_Backing {
   widgetClassPath=&Image
   bIsVariable=true
   parent=InfoPanel
   widget=InfoPanel_Backing
   properties {
    visibility=SelfHitTestInvisible
    brush {
     bIsDynamicallyLoaded=false
     drawAs=Image
     imageType=NoImage
     mirroring=NoMirror
     resourceName=None
     resourceObject=/Game/Athena/HUD/Inventory/Art/M_UI_HUD_ItemIcon_LoadoutIcon_InfoPlate_Master_Base
     tiling=NoTile
     imageSize=<32 32>
     margin=<ltrb 0 0 0 0>
     outlineSettings {
      bUseBrushTransparency=false
      roundingType=HalfHeightRadius
      width=0
      color {colorUseRule=UseColor_Specified, specifiedColor=<rgba 0 0 0 0>}
      cornerRadii=<0 0 0 0>
     }
     tintColor {colorUseRule=UseColor_Specified, specifiedColor=<rgba 1 1 1 1>}
     uVRegion {bIsValid=false, max=<0 0>, min=<0 0>}
    }
    renderTransformPivot=<0.5 1>
   }
   slot {class=OverlaySlot, properties {horizontalAlignment=HAlign_Fill, verticalAlignment=VAlign_Fill}}
  }
  VerticalBox_0 {
   widgetClassPath=&VerticalBox
   parent=InfoPanel
   widget=VerticalBox_0
   slot {class=OverlaySlot, properties {horizontalAlignment=HAlign_Center, verticalAlignment=VAlign_Center, padding=<ltrb 0 4 0 2>}}
  }
  HorizontalBox_168 {
   widgetClassPath=&HorizontalBox
   parent=VerticalBox_0
   widget=HorizontalBox_168
   slot {class=VerticalBoxSlot, properties {horizontalAlignment=HAlign_Center}}
  }
  Text_StackCount {
   widgetClassPath=&CommonTextBlock
   bInherited=true
   parent=HorizontalBox_168
   widget=Text_StackCount
   properties {
    mobileFontSizeMultiplier=1.2
    text=999
    font {
     bForceMonospaced=false
     bMaterialIsStencil=false
     fontMaterial=&M_UI_Inventory_Loadout_FontColor
     fontObject=&BurbankSmall
     letterSpacing=0
     monospacedWidth=1
     size=15
     skewAmount=0
     typefaceFontName=Black
     outlineSettings {
      bApplyOutlineToDropShadows=false
      bMiteredCorners=false
      bSeparateFillAlpha=false
      outlineMaterial=null
      outlineSize=0
      outlineColor=<rgba 0 0 0 1>
     }
    }
    shadowOffset=<0 0>
   }
   slot {class=HorizontalBoxSlot, properties {horizontalAlignment=HAlign_Center, verticalAlignment=VAlign_Center}}
  }
  Image_AmmoIcon {
   widgetClassPath=&Image
   bInherited=true
   parent=HorizontalBox_168
   widget=Image_AmmoIcon
   properties {
    visibility=SelfHitTestInvisible
    brush {
     bIsDynamicallyLoaded=false
     drawAs=Image
     imageType=NoImage
     mirroring=NoMirror
     resourceName=None
     resourceObject=&M_UI_Inventory_Loadout_AmmoIcon
     tiling=NoTile
     imageSize=<22 22>
     margin=<ltrb 0 0 0 0>
     outlineSettings {
      bUseBrushTransparency=false
      roundingType=HalfHeightRadius
      width=0
      color {colorUseRule=UseColor_Specified, specifiedColor=<rgba 0 0 0 0>}
      cornerRadii=<0 0 0 0>
     }
     tintColor {colorUseRule=UseColor_Specified, specifiedColor=<rgba 1 1 1 1>}
     uVRegion {bIsValid=false, max=<0 0>, min=<0 0>}
    }
   }
   slot {class=HorizontalBoxSlot, properties {horizontalAlignment=HAlign_Center, verticalAlignment=VAlign_Center}}
  }
  Image_RechargeableIcon {
   widgetClassPath=&Image
   bInherited=true
   parent=HorizontalBox_168
   widget=Image_RechargeableIcon
   properties {
    visibility=Collapsed
    brush {
     bIsDynamicallyLoaded=false
     drawAs=Image
     imageType=NoImage
     mirroring=NoMirror
     resourceName=None
     resourceObject=/Game/Athena/HUD/Inventory/Art/M_UI_Inventory_Loadout_RechargeableIcon
     tiling=NoTile
     imageSize=<22 22>
     margin=<ltrb 0 0 0 0>
     outlineSettings {
      bUseBrushTransparency=false
      roundingType=HalfHeightRadius
      width=0
      color {colorUseRule=UseColor_Specified, specifiedColor=<rgba 0 0 0 0>}
      cornerRadii=<0 0 0 0>
     }
     tintColor {colorUseRule=UseColor_Specified, specifiedColor=<rgba 1 1 1 1>}
     uVRegion {bIsValid=false, max=<0 0>, min=<0 0>}
    }
   }
   slot {class=HorizontalBoxSlot, properties {horizontalAlignment=HAlign_Center, verticalAlignment=VAlign_Center}}
  }
  ScaleBox_0 {
   widgetClassPath=&ScaleBox
   parent=VerticalBox_0
   widget=ScaleBox_0
   properties {stretch=UserSpecified, userSpecifiedScale=0.9}
   slot {class=VerticalBoxSlot}
  }
  ItemTierWidget {
   widgetClassPath=&/AthenaItemTierWidget
   bInherited=true
   parent=ScaleBox_0
   widget=ItemTierWidget
   properties {showBacking=false}
   slot {class=ScaleBoxSlot}
  }
 ]
}
