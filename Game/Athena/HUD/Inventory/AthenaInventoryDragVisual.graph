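# Drag visual widget for the Athena inventory, showing the dragged item's semi-transparent icon with an optional drop-action badge.
# On Construct, it feeds Item to ItemIcon.
# ShowDropIcon hides the shadowed badge for the first AthenaDragDropAction value, otherwise picking a drop, swap or lock texture.
# The swap icon becomes a lock when CanSwap finds the owning player controller missing or unable to perform the ChangeEquipment native action.
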WidgetBlueprint /Game/Athena/HUD/Inventory/AthenaInventoryDragVisual
graphs [
 CanSwap {nodes [
   CanSwap() {is=title K2Node_FunctionEntry}
   > n1 = Cast To FortPlayerController(Object=GetOwningPlayer())
   > Return Node!(
    execute
    CanSwap:bool=CanPerformNativeAction!(
     self:Fort Player Controller Object Reference=@n1
     ActionTag:"Gameplay Tag Structure (by ref)"=MakeLiteralGameplayTag!(Value:Gameplay Tag Structure=(TagName="Gameplay.Action.Player.ChangeEquipment")) -> (ReturnValue:Gameplay Tag Structure)
    ) -> (ReturnValue:bool)
   )
   Return Node(execute=@n1.CastFailed, CanSwap=false)
  ]}
 EventGraph {nodes [
   Event Construct()
   > SetItemToRepresent!(
    execute
    self:Athena Item Element Widget Base Object Reference=Get ItemIcon() -> (ItemIcon:Athena Item Icon Loadout Object Reference)
    ItemToRepresent:Fort Item Object Reference=Get Item() -> (Item:Fort Item Object Reference)
   ) -> (then)
  ]}
 ShowDropIcon {nodes [
   n1 = ShowDropIcon!() -> (then, Drop Action:AthenaDragDropAction Enum=NewEnumerator0) {is=title K2Node_FunctionEntry}
   > SetVisibility(
    self=Get DropImageContainer() -> (DropImageContainer:Border Object Reference)
    InVisibility=Select(
     Option 0=Collapsed
     Option 1=HitTestInvisible
     Index="Not Equal (Enum)"!(A:AthenaDragDropAction Enum=@n1, B:AthenaDragDropAction Enum=NewEnumerator0) -> (ReturnValue:bool)
    )
   )
   n2 = CanSwap!() -> (CanSwap:bool) {is=title K2Node_CallFunction}
   > SetBrushFromTexture!(
    execute
    self:Image Object Reference=Get DropImage() -> (DropImage:Image Object Reference)
    Texture:Texture 2D Object Reference=Select!(
     NewEnumerator0:Texture 2D Object Reference
     NewEnumerator1:Texture 2D Object Reference=/Game/Athena/HUD/Inventory/Styles/T_AthenaDropItem
     NewEnumerator2:Texture 2D Object Reference=Select!(
      Option 0:Texture 2D Object Reference=/Game/UI/Foundation/Textures/Icons/Locks/T-Icon-Lock-128
      Option 1:Texture 2D Object Reference=/Game/Athena/HUD/Inventory/Styles/T-Icon-SwapSlot
      Index:bool=@n2
     ) -> (ReturnValue)
     NewEnumerator3:Texture 2D Object Reference=&T-Icon-Lock-128
     Index:AthenaDragDropAction Enum=@n1
    ) -> (ReturnValue:Texture 2D Object Reference)
    bMatchSize:bool=false
   ) -> (then)
  ]}
]
variables [Item {type=FortItem, default=null}]
widgets {
 info {inheritedWidgetCount=0, parentClass=/Script/FortniteUI.FortUserWidget, rootWidgetClass=&Overlay, widgetCount=6}
 widgets [
  Overlay_0 {widgetClassPath=&Overlay, widget=Overlay_0}
  ItemIcon {
   widgetClassPath=&/AthenaItemIconLoadout
   bIsVariable=true
   parent=Overlay_0
   widget=ItemIcon
   properties {
    bDesignTime_SetItemToRepresent=true
    designTime_ItemDefinitionToRepresent=/Game/Athena/Items/Weapons/WID_Assault_AutoHigh_Athena_SR_Ore_T03
    renderOpacity=0.75
   }
   slot {class=OverlaySlot, properties {horizontalAlignment=HAlign_Fill, verticalAlignment=VAlign_Fill}}
  }
  DropImageContainer {
   widgetClassPath=&Border
   bIsVariable=true
   parent=Overlay_0
   widget=DropImageContainer
   properties {
    visibility=Collapsed
    background {
     bIsDynamicallyLoaded=false
     drawAs=Image
     imageType=NoImage
     mirroring=NoMirror
     resourceName=None
     resourceObject=null
     tiling=NoTile
     imageSize=<32 32>
     margin=<ltrb 0 0 0 0>
     outlineSettings {
      bUseBrushTransparency=false
      roundingType=HalfHeightRadius
      width=0
      color {colorUseRule=UseColor_Specified, specifiedColor=<rgba 0 0 0 0>}
      cornerRadii=<0 0 0 0>
     }
     tintColor {colorUseRule=UseColor_Specified, specifiedColor=<rgba 0 0 0 0.125>}
     uVRegion {bIsValid=false, max=<0 0>, min=<0 0>}
    }
    padding=<ltrb 0 0 0 0>
   }
   slot {class=OverlaySlot, properties {horizontalAlignment=HAlign_Fill, verticalAlignment=VAlign_Fill}}
  }
  Overlay_1 {
   widgetClassPath=&Overlay
   parent=DropImageContainer
   widget=Overlay_1
   slot {class=BorderSlot, properties {padding=<ltrb 0 0 0 0>}}
  }
  Shadow {
   widgetClassPath=&Image
   bIsVariable=true
   parent=Overlay_1
   widget=Shadow
   properties {
    visibility=SelfHitTestInvisible
    brush {
     bIsDynamicallyLoaded=false
     drawAs=Image
     imageType=NoImage
     mirroring=NoMirror
     resourceName=None
     resourceObject=/Game/Athena/HUD/Inventory/Art/M_UI_Inventory_DropItem_Shadow
     tiling=NoTile
     imageSize=<32 32>
     margin=<ltrb 0 0 0 0>
     outlineSettings {
      bUseBrushTransparency=false
      roundingType=HalfHeightRadius
      width=0
      color {colorUseRule=UseColor_Specified, specifiedColor=<rgba 0 0 0 0>}
      cornerRadii=<0 0 0 0>
     }
     tintColor {colorUseRule=UseColor_Specified, specifiedColor=<rgba 1 1 1 1>}
     uVRegion {bIsValid=false, max=<0 0>, min=<0 0>}
    }
    renderTransform {angle=0, scale=<1.25 1.125>, shear=<0 0>, translation=<0 0>}
   }
   slot {class=OverlaySlot, properties {horizontalAlignment=HAlign_Fill, verticalAlignment=VAlign_Fill}}
  }
  DropImage {
   widgetClassPath=&Image
   bIsVariable=true
   parent=Overlay_1
   widget=DropImage
   properties {
    visibility=SelfHitTestInvisible
    brush {
     bIsDynamicallyLoaded=false
     drawAs=Image
     imageType=NoImage
     mirroring=NoMirror
     resourceName=None
     resourceObject=&T_AthenaDropItem
     tiling=NoTile
     imageSize=<80 80>
     margin=<ltrb 0 0 0 0>
     outlineSettings {
      bUseBrushTransparency=false
      roundingType=HalfHeightRadius
      width=0
      color {colorUseRule=UseColor_Specified, specifiedColor=<rgba 0 0 0 0>}
      cornerRadii=<0 0 0 0>
     }
     tintColor {colorUseRule=UseColor_Specified, specifiedColor=<rgba 1 1 1 1>}
     uVRegion {bIsValid=false, max=<0 0>, min=<0 0>}
    }
   }
   slot {class=OverlaySlot, properties {horizontalAlignment=HAlign_Center, verticalAlignment=VAlign_Center}}
  }
 ]
}
