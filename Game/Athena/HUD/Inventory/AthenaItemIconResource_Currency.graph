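# Inventory icon widget for a currency resource in the Athena HUD, an AthenaItemIcon subclass showing an icon with an "x" and stack count.
# Item changes toggle between the Populated overlay and an empty-slot image, and SetHoveredState plays the matching hover animation forward or reverse.
# Hover and Intro animations drive material parameters; RandomizeAngleFloats feeds random edge values into the icon material.
# Construct gives the count text a dynamic font material, with size and padding adjusted for mobile.
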WidgetBlueprint /Game/Athena/HUD/Inventory/AthenaItemIconResource_Currency
animations [
 Hover {
  binding_count=3
  length_seconds=0.1
  object=&self:Hover
  sequence {bindings [
    Image_ItemIcon {
     id=#1
     possessed_class=Image
     tracks [{
       type=MovieSceneWidgetMaterialTrack
       sections [{
         type=MovieSceneParameterSection
         channels [
          Animate_Alpha {type=MovieSceneScriptingFloatChannel, keys [[0,0],[2,1]]}
          Animate_Color {type=MovieSceneScriptingFloatChannel, keys [[0,0],[2,1]]}
          Animate_ItemIcon_Scale {type=MovieSceneScriptingFloatChannel, keys [[0,0],[2,1]]}
          Animate_ItemIcon_Translate_Y {type=MovieSceneScriptingFloatChannel, keys [[0,0],[2,1]]}
          Animate_Shadow {type=MovieSceneScriptingFloatChannel, keys [[0,0],[2,1]]}
         ]
        }]
      }]
    }
    Text_StackCount {
     id=#2
     possessed_class=CommonTextBlock
     tracks [
      {
       type=MovieSceneWidgetMaterialTrack
       display_name=Font.FontMaterial
       sections [{type=MovieSceneParameterSection, channels [Animate_Color {type=MovieSceneScriptingFloatChannel, keys [[0,0],[1,1]]}]}]
      }
      {
       type=MovieScene2DTransformTrack
       sections [{
         type=MovieScene2DTransformSection
         channels [
          Translation.X {type=MovieSceneScriptingFloatChannel, const [0,1]}
          Translation.Y {type=MovieSceneScriptingFloatChannel, const [0,1]}
          Angle {type=MovieSceneScriptingFloatChannel, const [0,1]}
          Scale.X {type=MovieSceneScriptingFloatChannel, keys [[0,0.9],[2,1]]}
          Scale.Y {type=MovieSceneScriptingFloatChannel, keys [[0,0.9],[2,1]]}
          Shear.X {type=MovieSceneScriptingFloatChannel, const [0,1]}
          Shear.Y {type=MovieSceneScriptingFloatChannel, const [0,1]}
         ]
        }]
      }
     ]
    }
    SizeBox_All {
     id=#3
     possessed_class=SizeBox
     tracks [{
       type=MovieScene2DTransformTrack
       sections [{
         type=MovieScene2DTransformSection
         end_frame=2
         start_frame=0
         channels [
          Translation.X {type=MovieSceneScriptingFloatChannel, const [0,1]}
          Translation.Y {type=MovieSceneScriptingFloatChannel, keys [[0,0],[2,-2]]}
          Angle {type=MovieSceneScriptingFloatChannel, const [0,1]}
          Scale.X {type=MovieSceneScriptingFloatChannel, const [1,1]}
          Scale.Y {type=MovieSceneScriptingFloatChannel, const [1,1]}
          Shear.X {type=MovieSceneScriptingFloatChannel, const [0,1]}
          Shear.Y {type=MovieSceneScriptingFloatChannel, const [0,1]}
         ]
        }]
      }]
    }
   ]}
 }
 Intro {
  binding_count=1
  length_seconds=5
  object=&self:Intro
  sequence {bindings [Image_ItemIcon {
     id=#4
     possessed_class=Image
     tracks [{
       type=MovieSceneWidgetMaterialTrack
       sections [{
         type=MovieSceneParameterSection
         end_frame=100
         start_frame=0
         channels [ItemIcon_Scale_Start {type=MovieSceneScriptingFloatChannel, keys [[0,0],[4,0.55]]}]
        }]
      }]
    }]}
 }
 Hover_Empty {
  binding_count=1
  length_seconds=0.1
  object=&self:Hover_Empty
  sequence {bindings [Empty {
     id=#5
     possessed_class=Image
     tracks [{
       type=MovieSceneWidgetMaterialTrack
       sections [{type=MovieSceneParameterSection, channels [Animate_Alpha {type=MovieSceneScriptingFloatChannel, keys [[0,0],[2,1]]}]}]
      }]
    }]}
 }
]
graphs [
 EventGraph {nodes [
   Event Construct()
   > SetFontMaterial()
   > Branch(Condition=IsMobileGame())
   > SetPadding!(
    execute
    self:Overlay Slot Object Reference=SlotAsOverlaySlot(Widget=Get Text_StackCount() -> (Text_StackCount:Common Text Object Reference))
    InPadding:Margin Structure=Make Margin(Left=0, Top=0, Right=0, Bottom=-5)
   ) -> (then)
   n1 = Event OnPostItemToRepresentChanged!() -> (OutputDelegate, then, ItemToRepresent:Fort Item Object Reference) {is=title K2Node_Event}
   n2 = IsValid(Object=@n1.ItemToRepresent)
   > SetVisibility(
    self=Get Populated() -> (Populated:Overlay Object Reference)
    InVisibility=Select(Option 0=Collapsed, Option 1=SelfHitTestInvisible, Index=@n2)
   )
   > SetVisibility(self=Get Empty() -> (Empty:Image Object Reference), InVisibility=Select(Option 0=SelfHitTestInvisible, Option 1=Collapsed, Index=@n2))
  ]}
 IntroAnim {nodes [
   IntroAnim() {is=title K2Node_FunctionEntry}
   > PlayAnimationForward(InAnimation=Get Intro())
  ]}
 RandomizeAngleFloats {nodes [
   RandomizeAngleFloats() {is=title K2Node_FunctionEntry}
   > Set RandomFloat_Top(RandomFloat_Top:double=RandomFloatInRange(Min=-0.01, Max=0.0013)) -> (Output_Get:double)
   > Set RandomFloat_Left(RandomFloat_Left:double=RandomFloatInRange(Min=-0.0045, Max=0.0013)) -> (Output_Get:double)
   > Set RandomFloat_Right(RandomFloat_Right:double=RandomFloatInRange(Min=-0.0013, Max=0.0032)) -> (Output_Get:double)
   > Set RandomFloat_Bottom(RandomFloat_Bottom:double=RandomFloatInRange(Min=-0.0013, Max=0.01)) -> (Output_Get:double)
   > n1 = GetDynamicMaterial(self=Get Image_ItemIcon() -> (Image_ItemIcon:Image Object Reference))
   > SetScalarParameterValue(self=@n1, ParameterName=RandomValue_Top, Value=Get RandomFloat_Top() -> (RandomFloat_Top:double))
   > SetScalarParameterValue(self=@n1, ParameterName=RandomValue_Left, Value=Get RandomFloat_Left() -> (RandomFloat_Left:double))
   > SetScalarParameterValue(self=@n1, ParameterName=RandomValue_Right, Value=Get RandomFloat_Right() -> (RandomFloat_Right:double))
   > SetScalarParameterValue(self=@n1, ParameterName=RandomValue_Bottom, Value=Get RandomFloat_Bottom() -> (RandomFloat_Bottom:double))
  ]}
 SetFontMaterial {nodes [
   SetFontMaterial() {is=title K2Node_FunctionEntry}
   > n1 = CreateDynamicMaterialInstance!(execute, Parent=/Game/Athena/HUD/Inventory/Art/M_UI_Inventory_Resourece_FontColor, OptionalName=None, CreationFlags=None) -> (then, ReturnValue)
   n2 = Select(Option 0:int=18, Option 1:int=22, Index=IsMobileGame()) -> (ReturnValue:int)
   > SetFont(
    self=Get Text_StackCount()
    InFontInfo=MakeSlateFontInfo!(
     FontObject:Object Reference=&BurbankSmall
     FontMaterial:Object Reference=@n1
     OutlineSettings:Font Outline Settings Structure=(OutlineSize=0,bMiteredCorners=False,bSeparateFillAlpha=False,bApplyOutlineToDropShadows=False,OutlineMaterial=None,OutlineColor=(R=0,G=0,B=0,A=1))
     TypefaceFontName:Name=Black
     Size:float="To Float (Integer)"(InInt=@n2)
     LetterSpacing:int=0
     SkewAmount:float=0
     bForceMonospaced:bool=false
     bMaterialIsStencil:bool=false
     MonospacedWidth:float=1
    ) -> (ReturnValue:Slate Font Info Structure)
   )
  ]}
 SetHoveredState {nodes [
   n1 = Branch(Condition=SetHoveredState!() -> (then, bIsHovered:bool=~) {is=title K2Node_FunctionEntry})
   n2 = "Equal (Enum)"(
    A:ESlateVisibility Enum=GetVisibility!(self:Widget Object Reference=Get Populated()) -> (ReturnValue:ESlateVisibility Enum=Visible)
    B:ESlateVisibility Enum=SelfHitTestInvisible
   )
   > n3 = Branch(Condition=@n2)
   > PlayAnimationForward(InAnimation=Get Hover() -> (Hover:Widget Animation Object Reference))
   PlayAnimationForward(execute=@n3.else, InAnimation=Get Hover_Empty() -> (Hover_Empty:Widget Animation Object Reference))
   n4 = Branch(execute=@n1.else, Condition=@n2)
   > PlayAnimationReverse(InAnimation=Get Hover())
   PlayAnimationReverse(execute=@n4.else, InAnimation=Get Hover_Empty())
  ]}
]
variables [
 RandomFloat_Bottom {type=number}
 RandomFloat_Left {type=number}
 RandomFloat_Right {type=number}
 RandomFloat_Top {type=number}
]
widgets {
 info {inheritedWidgetCount=8, parentClass=/Script/FortniteUI.AthenaItemIcon, rootWidgetClass=&SizeBox, widgetCount=15}
 widgets [
  SizeBox_All {
   widgetClassPath=&SizeBox
   widget=SizeBox_All
   properties {bOverride_HeightOverride=true, bOverride_WidthOverride=true, heightOverride=90, widthOverride=180}
  }
  Overlay_0 {widgetClassPath=&Overlay, parent=SizeBox_All, widget=Overlay_0, slot {class=SizeBoxSlot}}
  Empty {
   widgetClassPath=&Image
   bIsVariable=true
   parent=Overlay_0
   widget=Empty
   properties {
    visibility=Collapsed
    brush {
     bIsDynamicallyLoaded=false
     drawAs=Box
     imageType=NoImage
     mirroring=NoMirror
     resourceName=None
     resourceObject=/Game/Athena/HUD/Inventory/Art/M_UI_Inventory_ResourceItem_EmptySlot
     tiling=NoTile
     imageSize=<64 64>
     margin=<ltrb 0.5 0.5 0.5 0.5>
     outlineSettings {
      bUseBrushTransparency=false
      roundingType=HalfHeightRadius
      width=0
      color {colorUseRule=UseColor_Specified, specifiedColor=<rgba 0 0 0 0>}
      cornerRadii=<0 0 0 0>
     }
     tintColor {colorUseRule=UseColor_Specified, specifiedColor=<rgba 1 1 1 1>}
     uVRegion {bIsValid=false, max=<0 0>, min=<0 0>}
    }
   }
   slot {class=OverlaySlot, properties {horizontalAlignment=HAlign_Fill, verticalAlignment=VAlign_Fill, padding=<ltrb 0 10 0 0>}}
  }
  Populated {
   widgetClassPath=&Overlay
   bIsVariable=true
   parent=Overlay_0
   widget=Populated
   slot {class=OverlaySlot, properties {horizontalAlignment=HAlign_Fill, verticalAlignment=VAlign_Fill}}
  }
  HorizontalBox_0 {
   widgetClassPath=&HorizontalBox
   parent=Populated
   widget=HorizontalBox_0
   slot {class=OverlaySlot, properties {horizontalAlignment=HAlign_Fill, verticalAlignment=VAlign_Fill}}
  }
  Image_ItemIcon {
   widgetClassPath=&Image
   bInherited=true
   bIsVariable=true
   parent=HorizontalBox_0
   widget=Image_ItemIcon
   properties {
    visibility=SelfHitTestInvisible
    brush {
     bIsDynamicallyLoaded=false
     drawAs=Image
     imageType=NoImage
     mirroring=NoMirror
     resourceName=None
     resourceObject=/Game/Athena/HUD/Inventory/Art/M_UI_HUD_Inventory_ResourceIcon_Currency
     tiling=NoTile
     imageSize=<82 82>
     margin=<ltrb 0 0 0 0>
     outlineSettings {
      bUseBrushTransparency=false
      roundingType=HalfHeightRadius
      width=0
      color {colorUseRule=UseColor_Specified, specifiedColor=<rgba 0 0 0 0>}
      cornerRadii=<0 0 0 0>
     }
     tintColor {colorUseRule=UseColor_Specified, specifiedColor=<rgba 1 1 1 1>}
     uVRegion {bIsValid=false, max=<0 0>, min=<0 0>}
    }
    renderTransform {angle=0, scale=<1 1>, shear=<0 0>, translation=<-8 0>}
   }
   slot {class=HorizontalBoxSlot, properties {horizontalAlignment=HAlign_Center, verticalAlignment=VAlign_Center}}
  }
  HorizontalBox_1 {
   widgetClassPath=&HorizontalBox
   parent=HorizontalBox_0
   widget=HorizontalBox_1
   properties {renderTransform {angle=0, scale=<1 1>, shear=<0 0>, translation=<-10 0>}}
   slot {class=HorizontalBoxSlot, properties {horizontalAlignment=HAlign_Left, verticalAlignment=VAlign_Center}}
  }
  X {
   widgetClassPath=&CommonTextBlock
   bIsVariable=true
   parent=HorizontalBox_1
   widget=X
   properties {
    bAutoCollapseWithEmptyText=true
    justification=Center
    mobileFontSizeMultiplier=1.4
    text=x
    colorAndOpacity {colorUseRule=UseColor_Specified, specifiedColor=<rgba 1 1 1 0.5>}
    font {
     bForceMonospaced=false
     bMaterialIsStencil=false
     fontMaterial=&M_UI_Inventory_Resourece_FontColor
     fontObject=&BurbankSmall
     letterSpacing=0
     monospacedWidth=1
     size=16.5
     skewAmount=0
     typefaceFontName=Black
     outlineSettings {
      bApplyOutlineToDropShadows=false
      bMiteredCorners=false
      bSeparateFillAlpha=false
      outlineMaterial=null
      outlineSize=0
      outlineColor=<rgba 0 0 0 1>
     }
    }
    renderTransform {angle=0, scale=<0.9 0.9>, shear=<0 0>, translation=<0 0>}
    renderTransformPivot=<0.5 0.75>
    shadowColorAndOpacity=<rgba 0 0 0 0.375>
    shadowOffset=<0 1>
   }
   slot {class=HorizontalBoxSlot, properties {verticalAlignment=VAlign_Center}}
  }
  Text_StackCount {
   widgetClassPath=&CommonTextBlock
   bInherited=true
   bIsVariable=true
   parent=HorizontalBox_1
   widget=Text_StackCount
   properties {
    bAutoCollapseWithEmptyText=true
    justification=Center
    mobileFontSizeMultiplier=1.45
    text=999
    font {
     bForceMonospaced=false
     bMaterialIsStencil=false
     fontMaterial=null
     fontObject=&BurbankSmall
     letterSpacing=0
     monospacedWidth=1
     size=16.5
     skewAmount=0
     typefaceFontName=Black
     outlineSettings {
      bApplyOutlineToDropShadows=false
      bMiteredCorners=false
      bSeparateFillAlpha=false
      outlineMaterial=null
      outlineSize=0
      outlineColor=<rgba 0 0 0 1>
     }
    }
    renderTransform {angle=0, scale=<0.9 0.9>, shear=<0 0>, translation=<0 0>}
    renderTransformPivot=<0.5 0.75>
    shadowColorAndOpacity=<rgba 0 0 0 0.375>
    shadowOffset=<0 1>
   }
   slot {class=HorizontalBoxSlot, properties {horizontalAlignment=HAlign_Center, verticalAlignment=VAlign_Center, padding=<ltrb 4 0 0 0>}}
  }
  Text_SecondaryStackCount {widgetClassPath=&CommonTextBlock, bInherited=true, widget=null}
  Image_AmmoIcon {widgetClassPath=&Image, bInherited=true, widget=null}
  Image_SecondaryAmmoIcon {widgetClassPath=&Image, bInherited=true, widget=null}
  Image_RechargeableIcon {widgetClassPath=&Image, bInherited=true, widget=null}
  InfoPanel_SecondaryItem {widgetClassPath=&Overlay, bInherited=true, widget=null}
  ItemTierWidget {widgetClassPath=/Script/FortniteUI.AthenaItemTierWidget, bInherited=true, widget=null}
 ]
}
