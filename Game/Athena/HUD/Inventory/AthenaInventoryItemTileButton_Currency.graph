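# Inventory tile button widget for currency items in the Athena HUD, wrapping a currency item icon and an action prompt.
# Hover and intro events forward to the icon.
# Non-gamepad left-button drags create a drag-drop operation with a visual, using Shift to split stacks, and leave a debug "Hello" print; drops report whether the payload is a FortItem.
# On focus, the prompt shows only for equipable items that are not WorldResource type.
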WidgetBlueprint /Game/Athena/HUD/Inventory/AthenaInventoryItemTileButton_Currency
graphs [
 EventGraph {nodes [
   Event On Hovered()
   > n1 = Cast To AthenaItemIconResource!(execute, Object=Get ItemIcon() -> (ItemIcon:Athena Item Icon Object Reference)) -> (then, CastFailed, AsAthena Item Icon Resource:Athena Item Icon Resource Object Reference) {is=title K2Node_DynamicCast}
   > SetHoveredState!(execute, self:Athena Item Icon Resource Object Reference=@n1, bIsHovered:bool=true) -> (then)
   Event On Unhovered()
   > n2 = Cast To AthenaItemIconResource(Object=Get ItemIcon())
   > SetHoveredState(self=@n2, bIsHovered=false)
  ]}
 HandleIntro {nodes [
   HandleIntro() {is=title K2Node_FunctionEntry}
   > n1 = Cast To AthenaItemIconResource(Object=Get ItemIcon())
   > IntroAnim(self:Athena Item Icon Resource Object Reference=@n1)
  ]}
 OnDragDetected {nodes [
   OnDragDetected!() -> (then, MyGeometry:Geometry Structure, PointerEvent:"Pointer Event Structure (by ref)") {is=title K2Node_FunctionEntry}
   > PrintString!(execute, InString=Hello, bPrintToScreen, bPrintToLog, TextColor, Duration, Key) -> (then)
   > n1 = Set Local_ItemToRepresent(Local_ItemToRepresent:Fort Item Object Reference=GetItemToRepresent!(self:Athena Inventory Item Tile Button Object Reference) -> (ReturnValue:Fort Item Object Reference)) -> (Output_Get:Fort Item Object Reference)
   > n2 = Is Valid(InputObject=@n1)
   Return Node!(execute=@n2.Is Not Valid, Operation:Drag Drop Operation Object Reference)
   > n3 = Create Widget!(
    execute:Exec
    Class:User Widget Class Reference=&/AthenaInventoryDragVisual.AthenaInventoryDragVisual_C
    OwningPlayer:Player Controller Object Reference=GetOwningPlayer()
    Item:Fort Item Object Reference=Get Local_ItemToRepresent() -> (Local_ItemToRepresent:Fort Item Object Reference)
   ) -> (then:Exec, ReturnValue:Athena Inventory Drag Visual Object Reference)
   n4 = Get PointerEvent() -> (PointerEvent:"Pointer Event Structure (by ref)")
   > n5 = "Create Drag && Drop Operation"!(
    execute:Exec
    Class:Drag Drop Operation Class Reference=&/AthenaInventoryDragDrop.AthenaInventoryDragDrop_C
    Tag:String
    Payload:Object Reference=Get Local_ItemToRepresent()
    DefaultDragVisual:Widget Object Reference=@n3
    Pivot:EDragPivot Enum=CenterCenter
    Offset:Vector 2D Structure=(X=0,Y=0)
    SplitTheStack:bool=Is Shift Down!(Input:"Input Event Structure (by ref)"=GetInputEventFromPointerEvent!(Event:"Pointer Event Structure (by ref)"=@n4) -> (ReturnValue:Input Event Structure)) -> (ReturnValue:bool)
    bIsFromCreativeInventory:bool=False
    InstigatorSlotIndex:int=0
   ) -> (then:Exec, ReturnValue:Athena Inventory Drag Drop Object Reference) {is=title K2Node_CreateDragDropOperation}
   > Return Node(Operation=@n5)
  ]}
 OnDrop {nodes [
   n1 = OnDrop!() -> (then, MyGeometry:Geometry Structure, PointerEvent:Pointer Event Structure, Operation:Drag Drop Operation Object Reference) {is=title K2Node_FunctionEntry}
   > Return Node!(
    execute
    ReturnValue:bool=Cast To FortItem!(Object=Get Payload!(self:Drag Drop Operation Object Reference=@n1.Operation) -> (Payload:Object Reference)) -> (AsFort Item:Fort Item Object Reference, bSuccess:bool=~) {is=title K2Node_DynamicCast}.bSuccess
   )
  ]}
 OnFocusReceived {nodes [
   OnFocusReceived!() -> (then, MyGeometry:Geometry Structure, InFocusEvent:Focus Event Structure) {is=title K2Node_FunctionEntry}
   > UpdateInteractionWidgetVisibility()
   > Return Node(ReturnValue:Event Reply Structure=Handled())
  ]}
 OnMouseButtonDown_Border {nodes [
   OnMouseButtonDown_Border!() -> (then, MyGeometry:Geometry Structure, MouseEvent:"Pointer Event Structure (by ref)") {is=title K2Node_FunctionEntry}
   n1 = Get CommonInputSubsystem!() -> (ReturnValue:Common Input Object Reference) {is=title K2Node_GetBlueprintContextEx}
   > n2 = Branch(Condition="Not Equal (Enum)"!(A:ECommonInputType Enum=GetCurrentInputType(self=@n1), B:ECommonInputType Enum=Gamepad) -> (ReturnValue:bool))
   > Return Node(ReturnValue=DetectDrag!(Reply:"Event Reply Structure (by ref)"=Unhandled(), DragKey:Key Structure=LeftMouseButton) -> (ReturnValue:Event Reply Structure))
   Return Node(execute=@n2.else, ReturnValue=Unhandled())
  ]}
 UpdateInteractionWidgetVisibility {nodes [
   UpdateInteractionWidgetVisibility() {is=title K2Node_FunctionEntry}
   > n1 = Set Local_ItemToRepresent(Local_ItemToRepresent=GetItemToRepresent())
   > n2 = Is Valid(InputObject=@n1)
   n3 = NOT Boolean(A="Equal (Enum)"(
     A:EFortItemType Enum=GetType!(self:Fort Item Object Reference=@n1) -> (ReturnValue:EFortItemType Enum=WorldItem)
     B:EFortItemType Enum=WorldResource
    ))
   n4 = IsEquipable!(
    self:Fort Item Object Reference=@n1
    EquippingPC:Fort Player Controller Object Reference=GetOwningController!(self:Fort Item Object Reference=@n1) -> (ReturnValue:Fort Player Controller Object Reference)
    bGamepadSwitchingForHeroAbilities:bool=false
   ) -> (ReturnValue:bool)
   n5 = Get InputActionWrapper() -> (InputActionWrapper:Common Visual Attachment Object Reference)
   > SetVisibility(self=@n5, InVisibility=Select(Option 0=Collapsed, Option 1=SelfHitTestInvisible, Index=AND Boolean(A=@n4, B=@n3)))
   SetVisibility(execute=@n2.Is Not Valid, self=@n5, InVisibility=Collapsed)
  ]}
]
variables []
widgets {
 info {inheritedWidgetCount=2, parentClass=/Script/FortniteUI.AthenaInventoryItemTileButton, rootWidgetClass=&Overlay, widgetCount=6}
 widgets [
  RootOverlay {widgetClassPath=&Overlay, widget=RootOverlay}
  Border_0 {
   widgetClassPath=&Border
   parent=RootOverlay
   widget=Border_0
   properties {
    background {
     bIsDynamicallyLoaded=false
     drawAs=NoDrawType
     imageType=NoImage
     mirroring=NoMirror
     resourceName=None
     resourceObject=null
     tiling=NoTile
     imageSize=<32 32>
     margin=<ltrb 0 0 0 0>
     outlineSettings {
      bUseBrushTransparency=false
      roundingType=HalfHeightRadius
      width=0
      color {colorUseRule=UseColor_Specified, specifiedColor=<rgba 0 0 0 0>}
      cornerRadii=<0 0 0 0>
     }
     tintColor {colorUseRule=UseColor_Specified, specifiedColor=<rgba 1 1 1 1>}
     uVRegion {bIsValid=false, max=<0 0>, min=<0 0>}
    }
    padding=<ltrb 0 0 0 0>
   }
   slot {class=OverlaySlot, properties {horizontalAlignment=HAlign_Fill, verticalAlignment=VAlign_Fill}}
  }
  SizeBox_0 {
   widgetClassPath=&SizeBox
   parent=Border_0
   widget=SizeBox_0
   properties {bOverride_HeightOverride=true, bOverride_WidthOverride=true, heightOverride=90, widthOverride=180}
   slot {class=BorderSlot, properties {padding=<ltrb 0 0 0 0>}}
  }
  ItemIcon {
   widgetClassPath=&/AthenaItemIconResource_Currency
   bInherited=true
   bIsVariable=true
   parent=SizeBox_0
   widget=ItemIcon
   slot {class=SizeBoxSlot}
  }
  InputActionWrapper {
   widgetClassPath=&CommonVisualAttachment
   bIsVariable=true
   parent=RootOverlay
   widget=InputActionWrapper
   properties {contentAnchor=<0.5 0.4>}
   slot {class=OverlaySlot, properties {horizontalAlignment=HAlign_Center, verticalAlignment=VAlign_Bottom}}
  }
  InputActionWidget {
   widgetClassPath=&CommonActionWidget
   bInherited=true
   parent=InputActionWrapper
   widget=InputActionWidget
   properties {
    clipping=ClipToBoundsWithoutIntersecting
    visibility=HitTestInvisible
    iconRimBrush {
     bIsDynamicallyLoaded=false
     drawAs=Image
     imageType=NoImage
     mirroring=NoMirror
     resourceName=None
     resourceObject=/FortUILibrary/Foundation/Buttons/Materials/M-Button-Console-YellowOutline
     tiling=NoTile
     imageSize=<40 40>
     margin=<ltrb 0 0 0 0>
     outlineSettings {
      bUseBrushTransparency=false
      roundingType=HalfHeightRadius
      width=0
      color {colorUseRule=UseColor_Specified, specifiedColor=<rgba 0 0 0 0>}
      cornerRadii=<0 0 0 0>
     }
     tintColor {colorUseRule=UseColor_Specified, specifiedColor=<rgba 1 1 1 1>}
     uVRegion {bIsValid=false, max=<0 0>, min=<0 0>}
    }
   }
   slot {class=SizeBoxSlot, properties {padding=<ltrb 4 0 0 0>}}
  }
 ]
}
