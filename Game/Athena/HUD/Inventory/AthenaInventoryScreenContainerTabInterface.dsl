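; Widget blueprint for the Athena HUD inventory screen's container tab interface.
; It implements two header display queries, both returning false, so the tab shows neither a blurred header background nor a header background.
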(fn ShowBackgroundBlurHeader
  (return false))

(fn ShowHeaderBackground
  (return false))
