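# Tab button widget for the Athena inventory HUD, with a text label, optional icon and notification bang wrapper.
# On construct it selects a HUD or mobile style, resizes for mobile, and caches the border material MID.
# Selection tints the icon and, off mobile, animates label and border, feeding randomized edge offsets into the MID for varied shapes.
# Hover animations play only when unselected and the Bacchus frontend label option is off.
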WidgetBlueprint /Game/Athena/HUD/Inventory/AthenaInventoryTabButton
animations [
 Selected {
  binding_count=2
  length_seconds=0.2
  object=&self:Selected
  sequence {bindings [
    Text_DisplayName {
     id=#1
     possessed_class=CommonTextBlock
     tracks [
      {
       type=MovieScene2DTransformTrack
       sections [{
         type=MovieScene2DTransformSection
         channels [
          Translation.X {type=MovieSceneScriptingFloatChannel}
          Translation.Y {type=MovieSceneScriptingFloatChannel}
          Angle {type=MovieSceneScriptingFloatChannel}
          Scale.X {type=MovieSceneScriptingFloatChannel, keys [[0,0.825],[1,1]]}
          Scale.Y {type=MovieSceneScriptingFloatChannel, keys [[0,0.825],[1,1]]}
          Shear.X {type=MovieSceneScriptingFloatChannel}
          Shear.Y {type=MovieSceneScriptingFloatChannel}
         ]
        }]
      }
      {
       type=MovieSceneColorTrack
       display_name=Color and Opacity
       sections [{
         type=MovieSceneColorSection
         channels [
          Color.R {type=MovieSceneScriptingFloatChannel, keys [[0,1],[1,0.0125]]}
          Color.G {type=MovieSceneScriptingFloatChannel, keys [[0,1],[1,0.0125]]}
          Color.B {type=MovieSceneScriptingFloatChannel, keys [[0,1],[1,0.0125]]}
          Color.A {type=MovieSceneScriptingFloatChannel, const [1,2]}
         ]
        }]
      }
      {
       type=MovieSceneFloatTrack
       display_name=Render Opacity
       sections [{type=MovieSceneFloatSection, channels [None {type=MovieSceneScriptingFloatChannel, keys [[0,0.9],[1,1]]}]}]
      }
     ]
    }
    InternalRootButton {
     id=#2
     possessed_class=null
     tracks [{
       type=MovieSceneWidgetMaterialTrack
       display_name=Style.Normal.Brush Material
       sections [{
         type=MovieSceneParameterSection
         channels [
          Alpha {type=MovieSceneScriptingFloatChannel, keys [[0,0],[2,1],[4,0.9]]}
          AnimateSelection {type=MovieSceneScriptingFloatChannel, keys [[0,0],[2,1],[4,0.9]]}
         ]
        }]
      }]
    }
   ]}
 }
 Hovered {
  binding_count=2
  length_seconds=0.1
  object=&self:Hovered
  sequence {bindings [
    InternalRootButton {
     id=#3
     possessed_class=null
     tracks [{type=MovieSceneWidgetMaterialTrack, display_name=Style.Normal.Brush Material, sections [{type=MovieSceneParameterSection}]}]
    }
    Text_DisplayName {
     id=#4
     possessed_class=CommonTextBlock
     tracks [{
       type=MovieScene2DTransformTrack
       sections [{
         type=MovieScene2DTransformSection
         channels [
          Translation.X {type=MovieSceneScriptingFloatChannel}
          Translation.Y {type=MovieSceneScriptingFloatChannel}
          Angle {type=MovieSceneScriptingFloatChannel}
          Scale.X {type=MovieSceneScriptingFloatChannel, keys [[0,0.825],[2,1]]}
          Scale.Y {type=MovieSceneScriptingFloatChannel, keys [[0,0.825],[2,1]]}
          Shear.X {type=MovieSceneScriptingFloatChannel}
          Shear.Y {type=MovieSceneScriptingFloatChannel}
         ]
        }]
      }]
    }
   ]}
 }
]
graphs [
 EventGraph {nodes [
   n1 = OnSelectedChangedEvent!() -> (OutputDelegate, then, Button:Common Button Legacy Object Reference, Selected:bool=~) {is=title K2Node_CustomEvent}
   > HandleSelectionChangedVisuals!(execute, self, bIsSelected:bool=@n1.Selected) -> (then)
   > n2 = Branch(Condition=@n1.Selected)
   n3 = Get Image_MobileIcon() -> (Image_MobileIcon:Image Object Reference)
   > SetBrushTintColor!(execute, self:Image Object Reference=@n3, TintColor:Slate Color Structure=Make Slate Color(SpecifiedColor=(R=0,G=0,B=0,A=1))) -> (then)
   SetBrushTintColor(execute=@n2.else, self=@n3, TintColor=Make Slate Color(SpecifiedColor=(R=1,G=1,B=1,A=1)))
   n4 = HandleSelectionChangedVisuals!() -> (OutputDelegate, then, bIsSelected:bool=~) {is=title K2Node_CustomEvent}
   > n5 = Branch(Condition="Equal (Boolean)"!(A:bool=@n4.bIsSelected, B:bool=Get VisualSelectionState() -> (VisualSelectionState:bool)) -> (ReturnValue:bool))
   n6 = Set VisualSelectionState(execute=@n5.else, VisualSelectionState:bool=@n4.bIsSelected) -> (Output_Get:bool)
   > n7 = Branch(Condition=IsMobileGame())
   n8 = Branch(execute=@n7.else, Condition=@n6)
   SeedRandomValues(execute=[@n8.then, @n22.then])
   > UpdateMIDWithRandomValues()
   > PlayAnimationForward(InAnimation=Get Selected() -> (Selected:Widget Animation Object Reference))
   PlayAnimationReverse(execute=[@n8.else, @n21.then], InAnimation=Get Selected(), PlaybackSpeed=3)
   Event OnCurrentTextStyleChanged() {is=title K2Node_Event}
   > Branch(Condition=IsMobileGame())
   > SetStyle!(
    execute
    self:Common Text Object Reference=Get Text_DisplayName() -> (Text_DisplayName:Common Text Object Reference)
    InStyle:Common Text Style Class Reference=GetCurrentTextStyleClass!(self:Common Button Base Object Reference) -> (ReturnValue:Common Text Style Class Reference)
   ) -> (then)
   Event PreConstruct()
   > n9 = Cast To NormalBangWrapper!(execute, Object=Get BangWrapper() -> (BangWrapper:Fort Bang Wrapper NUI Object Reference)) -> (then, CastFailed, AsNormal Bang Wrapper:Normal Bang Wrapper Object Reference) {is=title K2Node_DynamicCast}
   n10 = GetRuntimeOptions()
   > n11 = Branch(Condition=AND Boolean(A=IsMobileGame(), B=Get bBacchusFrontendEnabled!(self:Fort Runtime Options Object Reference=@n10) -> (bBacchusFrontendEnabled:bool)))
   > Adjust Bang Image Translation!(execute, self:Normal Bang Wrapper Object Reference=@n9, Translation:Vector 2D Structure=(X=-10,Y=-4)) -> (then)
   Adjust Bang Image Translation(execute=@n11.else, self=@n9, Translation=(X=20,Y=-4))
   Event Construct()
   > InitializeButtonBorderMID()
   > SetStyle(
    self:Common Button Base Object Reference
    InStyle:Common Button Style Class Reference=Select(
     Option 0:Common Button Style Class Reference=/FortUILibrary/Foundation/Buttons/ButtonStyles/Tab/ButtonStyle-Tab-HUDInventory.ButtonStyle-Tab-HUDInventory_C
     Option 1:Common Button Style Class Reference=/FortUILibrary/Foundation/Buttons/ButtonStyles/Tab/ButtonStyle-Tab-InventoryPanel-Bacchus.ButtonStyle-Tab-InventoryPanel-Bacchus_C
     Index=IsMobileGame()
    ) -> (ReturnValue:Common Button Style Class Reference)
   )
   > Branch(Condition=IsMobileGame())
   n12 = Get SizeBox_0() -> (SizeBox_0:Size Box Object Reference)
   > SetMinDesiredWidth!(execute, self:Size Box Object Reference=@n12, InMinDesiredWidth:float=140) -> (then)
   > SetMaxDesiredWidth!(execute, self:Size Box Object Reference=@n12, InMaxDesiredWidth:float=175) -> (then)
   > SetHeightOverride(self=@n12, InHeightOverride=66)
   n13 = Get Text_DisplayName()
   > Set Justification!(execute, Justification:ETextJustify Enum=Center, self:Text Layout Widget Object Reference=@n13) -> (then, Output_Get:ETextJustify Enum=Left)
   > SetWrapTextWidth!(execute, self:Common Text Object Reference=@n13, InWrapTextAt:int=130) -> (then)
   > SetPadding!(
    execute
    self:Vertical Box Slot Object Reference=SlotAsVerticalBoxSlot(Widget=Get SizeBox_2() -> (SizeBox_2:Size Box Object Reference))
    InPadding:Margin Structure=Make Margin(Left=0, Top=-2, Right=0, Bottom=0)
   ) -> (then)
   Event OnInitialized()
   > Assign On Selected Changed!(execute:Exec, self:Common Button Legacy Object Reference, Delegate:"Delegate (by ref)"=@n1) -> (then:Exec) {is=title K2Node_AssignDelegate}
   Event On Hovered()
   n14 = Get VisualSelectionState()
   > n15 = Branch(Condition=@n14)
   n16 = Get bBacchusFrontendLabelShown!(self:Fort Runtime Options Object Reference=GetRuntimeOptions()) -> (bBacchusFrontendLabelShown:bool)
   n17 = Branch(execute=@n15.else, Condition=@n16)
   n18 = Get Hovered() -> (Hovered:Widget Animation Object Reference)
   PlayAnimationForward(execute=@n17.else, InAnimation=@n18, PlaybackSpeed=1.25)
   Event On Unhovered()
   > n19 = Branch(Condition=@n14)
   n20 = Branch(execute=@n19.else, Condition=@n16)
   PlayAnimationReverse(execute=@n20.else, InAnimation=@n18, PlaybackSpeed=3)
   n21 = Set ShapeChange_Target(ShapeChange_Target:double=0) -> (Output_Get:double)
   n22 = Set ShapeChange_Target(ShapeChange_Target=1)
  ]}
 InitializeButtonBorderMID {nodes [
   InitializeButtonBorderMID() {is=title K2Node_FunctionEntry}
   > n1 = Set MID_ButtonBorder(MID_ButtonBorder:Material Instance Dynamic Object Reference=GetSingleMaterialStyleMID!(self:Common Button Base Object Reference) -> (ReturnValue:Material Instance Dynamic Object Reference)) -> (Output_Get:Material Instance Dynamic Object Reference)
   > SetScalarParameterValue(self=@n1, ParameterName=Alpha, Value=0)
  ]}
 OverrideIcon {nodes [
   n1 = OverrideIcon!() -> (then, Show:bool=~, inSlateBrush:Slate Brush Structure) {is=title K2Node_FunctionEntry}
   > SetVisibility(
    self=Get Image_MobileIcon()
    InVisibility=Select(Option 0:ESlateVisibility Enum=Collapsed, Option 1:ESlateVisibility Enum=SelfHitTestInvisible, Index=@n1) -> (ReturnValue:ESlateVisibility Enum)
   )
   > n2 = Is Valid(InputObject=Break Slate Brush!(SlateBrush:"Slate Brush Structure (by ref)"=Get inSlateBrush() -> (inSlateBrush:Slate Brush Structure)) -> (
     TintColor:Slate Color Structure
     DrawAs:ESlateBrushDrawType Enum
     Tiling:ESlateBrushTileType Enum
     Mirroring:ESlateBrushMirrorType Enum
     ImageSize:Vector 2D Structure
     Margin:Margin Structure
     ResourceObject:Object Reference
     OutlineSettings:Slate Brush Outline Settings Structure
    ) {is=title K2Node_BreakStruct}.ResourceObject)
   SetVisibility(execute=@n2.Is Not Valid, self=Get Image_MobileIcon(), InVisibility=Collapsed)
   SetBrush(execute=@n2.Is Valid, self=Get Image_MobileIcon(), InBrush=Get inSlateBrush())
  ]}
 SeedRandomValues {nodes [
   SeedRandomValues() {is=title K2Node_FunctionEntry}
   > Set RandomFloat_Top(RandomFloat_Top:double=RandomFloatInRange(Min=-0.015, Max=0.0015)) -> (Output_Get:double)
   > Set RandomFloat_Left(RandomFloat_Left:double=RandomFloatInRange(Min=-0.0075, Max=0.0018)) -> (Output_Get:double)
   > Set RandomFloat_Right(RandomFloat_Right:double=RandomFloatInRange(Min=-0.0018, Max=0.0047)) -> (Output_Get:double)
   > Set RandomFloat_Bottom(RandomFloat_Bottom:double=RandomFloatInRange(Min=-0.0016, Max=0.0163)) -> (Output_Get:double)
  ]}
 SetText {nodes [
   n1 = SetText!() -> (then, ButtonText:Text) {is=title K2Node_FunctionEntry}
   > n2 = Set ButtonText(ButtonText:Text=@n1) -> (Output_Get:Text)
   > n3 = Branch(Condition=NOT Boolean(A=TextIsEmpty(InText=@n1)))
   n4 = Get Text_DisplayName()
   > SetVisibility(self=@n4, InVisibility=SelfHitTestInvisible)
   > "SetText (Text)"(self=@n4, InText=@n2)
   SetVisibility(execute=@n3.else, self=@n4, InVisibility=Collapsed)
  ]}
 UpdateMIDWithRandomValues {nodes [
   UpdateMIDWithRandomValues() {is=title K2Node_FunctionEntry}
   n1 = Get MID_ButtonBorder() -> (MID_ButtonBorder:Material Instance Dynamic Object Reference)
   > Is Valid(InputObject=@n1)
   > SetScalarParameterValue(self=@n1, ParameterName=RandomValue_Top, Value=Get RandomFloat_Top() -> (RandomFloat_Top:double))
   > SetScalarParameterValue(self=@n1, ParameterName=RandomValue_Left, Value=Get RandomFloat_Left() -> (RandomFloat_Left:double))
   > SetScalarParameterValue(self=@n1, ParameterName=RandomValue_Right, Value=Get RandomFloat_Right() -> (RandomFloat_Right:double))
   > SetScalarParameterValue(self=@n1, ParameterName=RandomValue_Bottom, Value=Get RandomFloat_Bottom() -> (RandomFloat_Bottom:double))
  ]}
]
variables [
 ButtonText {type=string, default=""}
 DampenMultiply {type=number, default=1}
 MassMultiply {type=number, default=0.25}
 MID_ButtonBorder {type=MaterialInstanceDynamic}
 PreviewMobileStyle {type=boolean}
 RandomFloat_Bottom {type=number}
 RandomFloat_Left {type=number}
 RandomFloat_Right {type=number}
 RandomFloat_Top {type=number}
 ShapeChange_New {type=number}
 ShapeChange_Target {type=number}
 Spring_Scale_Target {type=number}
 SpringState_Scale {type=FloatSpringState, default {}}
 SpringState_ShapeChange {type=FloatSpringState, default {}}
 StiffnessMultiply {type=number, default=20}
 UpdateSpring {type=boolean}
 VisualSelectionState {type=boolean}
]
widgets {
 info {
  inheritedWidgetCount=4
  namedSlotCount=1
  parentClass=/Script/FortniteUI.FortTopBarTabTextIconButton
  rootWidgetClass=/Game/UI/Foundation/Bangs/NormalBangWrapper.NormalBangWrapper_C
  widgetCount=9
 }
 widgets [
  BangWrapper {
   widgetClassPath=/Game/UI/Foundation/Bangs/NormalBangWrapper
   bInherited=true
   bIsVariable=true
   widget=BangWrapper
   properties {useSmallVersion=true}
  }
  SizeBox_0 {
   widgetClassPath=&SizeBox
   bIsVariable=true
   namedSlotHost=BangWrapper
   widget=SizeBox_0
   properties {bOverride_MinDesiredHeight=true, heightOverride=50, minDesiredHeight=64}
  }
  VerticalBox_0 {
   widgetClassPath=&VerticalBox
   bIsVariable=true
   parent=SizeBox_0
   widget=VerticalBox_0
   slot {class=SizeBoxSlot, properties {horizontalAlignment=HAlign_Center, verticalAlignment=VAlign_Center, padding=<ltrb 20 0 20 0>}}
  }
  SizeBox_1 {
   widgetClassPath=&SizeBox
   bIsVariable=true
   parent=VerticalBox_0
   widget=SizeBox_1
   properties {bOverride_HeightOverride=true, bOverride_WidthOverride=true, heightOverride=60, widthOverride=60}
   slot {class=VerticalBoxSlot, properties {horizontalAlignment=HAlign_Center, verticalAlignment=VAlign_Top}}
  }
  Image_MobileIcon {
   widgetClassPath=&Image
   bInherited=true
   parent=SizeBox_1
   widget=Image_MobileIcon
   properties {
    visibility=Collapsed
    brush {
     bIsDynamicallyLoaded=false
     drawAs=Image
     imageType=NoImage
     mirroring=NoMirror
     resourceName=None
     resourceObject=/Game/UI/Foundation/Textures/Icons/Navigation/t_icon_nav_mobile_lobby_128
     tiling=NoTile
     imageSize=<60 60>
     margin=<ltrb 0 0 0 0>
     outlineSettings {
      bUseBrushTransparency=false
      roundingType=HalfHeightRadius
      width=0
      color {colorUseRule=UseColor_Specified, specifiedColor=<rgba 0 0 0 0>}
      cornerRadii=<0 0 0 0>
     }
     tintColor {colorUseRule=UseColor_Specified, specifiedColor=<rgba 1 1 1 1>}
     uVRegion {bIsValid=false, max=<0 0>, min=<0 0>}
    }
   }
   slot {class=SizeBoxSlot, properties {horizontalAlignment=HAlign_Center, verticalAlignment=VAlign_Center}}
  }
  SizeBox_2 {
   widgetClassPath=&SizeBox
   bIsVariable=true
   parent=VerticalBox_0
   widget=SizeBox_2
   slot {class=VerticalBoxSlot, properties {verticalAlignment=VAlign_Center}}
  }
  Overlay_0 {widgetClassPath=&Overlay, parent=SizeBox_2, widget=Overlay_0, slot {class=SizeBoxSlot}}
  Text_DisplayName {
   widgetClassPath=&CommonTextBlock
   bInherited=true
   bIsVariable=true
   parent=Overlay_0
   widget=Text_DisplayName
   properties {
    justification=Center
    renderOpacity=0.9
    text=Lobby
    textTransformPolicy=ToUpper
    font {
     bForceMonospaced=false
     bMaterialIsStencil=false
     fontMaterial=null
     fontObject=&Burbank
     letterSpacing=0
     monospacedWidth=1
     size=20
     skewAmount=0
     typefaceFontName=BurbankRegular-Black
     outlineSettings {
      bApplyOutlineToDropShadows=false
      bMiteredCorners=false
      bSeparateFillAlpha=false
      outlineMaterial=null
      outlineSize=0
      outlineColor=<rgba 0 0 0 0>
     }
    }
    renderTransform {angle=0, scale=<0.825 0.825>, shear=<0 0>, translation=<0 0>}
    shadowOffset=<0 0>
   }
   slot {class=OverlaySlot, properties {horizontalAlignment=HAlign_Fill, verticalAlignment=VAlign_Center}}
  }
  InputActionWidget {widgetClassPath=&CommonActionWidget, bInherited=true, widget=null}
 ]
}
