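# Quickbar editor widget for the Athena inventory, laying out slot buttons in a one-row grid that can be reordered for backwards display.
# Base-class events add and remove buttons, while quickbar content changes create or drop extra slots beyond the standard count.
# A gameplay event router listener adds or removes overlay widgets on individual slots.
# Custom navigation wraps left and right, accounting for multi-slot items, and delegates other directions to the parent inventory panel.
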WidgetBlueprint /Game/Athena/HUD/Inventory/AthenaQuickBarEditorWidget.AthenaQuickbarEditorWidget
graphs [
 EventGraph {nodes [
   Event PreConstruct()
   > ClearChildren(self=Get Grid_EquipSlots() -> (Grid_EquipSlots:Grid Panel Object Reference))
   Event ChangeQuickbarOrder() {is=title K2Node_Event}
   > n1 = For Loop(
    FirstIndex=int - int(A=Get FirstQuickBarSlotIndex() -> (FirstQuickBarSlotIndex:int=0))
    LastIndex=int - int(A=Get NumQuickBarSlots() -> (NumQuickBarSlots:int=0))
   )
   n2 = GetChildAt!(self:Panel Widget Object Reference=Get Grid_EquipSlots(), Index:int=@n1) -> (ReturnValue:Widget Object Reference)
   > n3 = Cast To AthenaQuickBarSlotButtonBase!(execute, Object=@n2) -> (then, CastFailed, AsAthena Quick Bar Slot Button Base:Athena Quick Bar Slot Button Base Object Reference) {is=title K2Node_DynamicCast}
   n4 = Get SlotIndex!(self:Athena Quick Bar Slot Button Base Object Reference=@n3) -> (SlotIndex:int=0)
   n5 = SelectInt(
    A=int - int(A=Get NumQuickBarSlots() -> (NumQuickBarSlots=0), B=@n4)
    B=int - int(A=@n4, B=Get FirstQuickBarSlotIndex() -> (FirstQuickBarSlotIndex=0))
    bPickA=Get bIsBackwards() -> (bIsBackwards:bool)
   )
   n6 = SlotAsGridSlot(Widget=@n2)
   > SetColumn!(execute, self:Grid Slot Object Reference=@n6, InColumn:int=@n5) -> (then)
   > SetPadding!(
    execute
    self:Grid Slot Object Reference=@n6
    InPadding_Left:float=Select(Option 0:double=0, Option 1:double=10, Index="Not Equal (Integer)"(A=@n5)) -> (ReturnValue:double)
    InPadding_Top:float=0
    InPadding_Right:float=0
    InPadding_Bottom:float=0
   ) -> (then)
   n7 = Event OnQuickBarSlotButtonCreated!() -> (OutputDelegate, then, QuickBarSlotButton:Athena Quick Bar Slot Button Base Object Reference) {is=title K2Node_Event}
   n8 = Get Grid_EquipSlots()
   > n9 = AddChildToGrid!(
    execute
    self:Grid Panel Object Reference=@n8
    Content:Widget Object Reference=@n7.QuickBarSlotButton
    InRow:int=0
    InColumn:int=GetChildrenCount!(self:Panel Widget Object Reference=@n8) -> (ReturnValue:int=0)
   ) -> (then, ReturnValue:Grid Slot Object Reference)
   > SetHorizontalAlignment(self:Grid Slot Object Reference=@n9, InHorizontalAlignment=HAlign_Fill)
   > SetVerticalAlignment(self:Grid Slot Object Reference=@n9, InVerticalAlignment=VAlign_Fill)
   n10 = GetChildrenCount(self=Get Grid_EquipSlots()) -> (ReturnValue=0)
   > SetPadding(
    self=@n9
    InPadding_Left=Select(Option 0=0, Option 1=10, Index="integer > integer"(A=@n10, B=1) {is=title K2Node_CallFunction})
    InPadding_Top=0
    InPadding_Right=0
    InPadding_Bottom=0
   )
   Event Construct()
   n11 = GetGlobalEventRouter()
   > n12 = StartListeningToEvent!(
    execute:Exec
    Target:Gameplay Event Router Component Object Reference=@n11
    EventType:Script Struct Object Reference=/Game/Athena/HUD/Inventory/Data/ST_AddInventoryQuickBarOverlay
    Context:Object Reference
    EventBubblingRule:EEventBubblingRule Enum
    None:%=@n11
   ) -> (
    then:Exec
    AsyncTaskProxy:Async Action Start Listening To Event Object Reference
    OnEventReceived:Exec
    Payload:ST Add Inventory Quick Bar Overlay Structure
   )
   > n13 = Get!(execute:Exec) -> (then:Exec, else:Exec, Async Task Ref:Async Action Start Listening To Event Object Reference)
   > StopListeningToEvent(self:Async Action Start Listening To Event Object Reference=@n13)
   Set Async Task Ref(execute=@n13.else, Async Task Ref:Async Action Start Listening To Event Object Reference=@n12) -> (Output_Get:Async Action Start Listening To Event Object Reference)
   Branch(
    execute=@n12.OnEventReceived
    Condition=IsValidClass!(Class:Object Class Reference=Break ST Add Inventory Quick Bar Overlay!(ST_AddInventoryQuickBarOverlay:"ST Add Inventory Quick Bar Overlay Structure (by ref)"=@n12.Payload) -> (WidgetClass_10_#1:User Widget Class Reference) {is=title K2Node_BreakStruct}) -> (ReturnValue:bool)
   )
   > HandleAddingOverlayToQuickbar!(execute, self, Payload:ST Add Inventory Quick Bar Overlay Structure=@n12.Payload) -> (then)
   n14 = Event OnQuickBarSlotButtonRemoved!() -> (OutputDelegate, then, RemovedSlotIndex:int=0) {is=title K2Node_Event}
   > RemoveFromParent(self=GetChildAt(self=Get Grid_EquipSlots(), Index=int - int(A=@n14.RemovedSlotIndex)))
   Event Destruct()
   > StopListeningToEvent(self=Get Async Task Ref() -> (Async Task Ref:Async Action Start Listening To Event Object Reference))
  ]}
 Find Item to the Left {nodes [
   Find Item to the Left() {is=title K2Node_FunctionEntry}
   n1 = Get EquipButtonGroup() -> (EquipButtonGroup:Common Button Group Base Object Reference)
   n2 = Get NumQuickBarSlots() -> (NumQuickBarSlots=0)
   n3 = "% (Integer)"(A=int - int(A=GetSelectedButtonIndex!(self:Common Button Group Base Object Reference=@n1) -> (ReturnValue:int=0)), B=@n2) -> (ReturnValue=0)
   n4 = Select(Option 0:int=@n3, Option 1:int=int + int(A=@n3, B=@n2), Index="integer < integer"(A=@n3)) -> (ReturnValue:int)
   > n5 = Cast To AthenaInventoryQuickBarSlotButton!(
    execute
    Object=GetButtonBaseAtIndex!(self:Common Button Group Base Object Reference=Get EquipButtonGroup(), Index:int=@n4) -> (ReturnValue:Common Button Base Object Reference)
   ) -> (then, CastFailed, AsAthena Inventory Quick Bar Slot Button:Athena Inventory Quick Bar Slot Button Object Reference) {is=title K2Node_DynamicCast}
   n6 = GetSlottedItem!(self:Athena Quick Bar Slot Button Base Object Reference=@n5) -> (ReturnValue:Fort Item Object Reference)
   > n7 = Is Valid(InputObject=@n6)
   Return Node!(execute=[@n7.Is Not Valid, @n5.CastFailed], Item:Fort Item Object Reference, Index:int=-1)
   n8 = GetNumberOfSlotsTaken!(self:Fort Item Object Reference=@n6) -> (ReturnValue:Byte=0)
   > Return Node(Item=@n6, Index=int - int(A=@n4, B=int - int(A:Byte=@n8, B:Byte)))
  ]}
 GetFirstWidgetOfType {nodes [
   n1 = GetFirstWidgetOfType!() -> (then, Type:User Widget Class Reference, Panel:Panel Widget Object Reference) {is=title K2Node_FunctionEntry}
   > n2 = For Each Loop(Array:Array of Widget Object References=GetAllChildren(self=@n1.Panel)) -> (Array Element:Widget Object Reference)
   Return Node!(execute=@n2.Completed, Content:Widget Object Reference)
   > Branch(Condition=IsA!(Object:Object Reference=@n2, ObjectClass:Object Class Reference=@n1) -> (ReturnValue:bool))
   > Return Node(Content=@n2)
  ]}
 HandleAddingOverlayToQuickbar {nodes [
   n1 = HandleAddingOverlayToQuickbar!() -> (then, Payload:ST Add Inventory Quick Bar Overlay Structure) {is=title K2Node_FunctionEntry}
   n2 = GetAllChildren(self=Get Grid_EquipSlots())
   n3 = int - int(
    A:int=Break ST Add Inventory Quick Bar Overlay!(ST_AddInventoryQuickBarOverlay=@n1) -> (Slot_7_#2:int)
    B:int=2
   )
   > Branch(Condition=Is Valid Index!(TargetArray:Array of Widget Object References=@n2, IndexToTest:int=@n3) -> (ReturnValue:bool))
   > n4 = Cast To AthenaInventoryEquipSlot!(execute, Object="Get (a copy)"(Array:Array of Widget Object References=@n2, Dimension 1=@n3) -> (Output:Widget Object Reference)) -> (then, CastFailed, AsAthena Inventory Equip Slot:Athena Inventory Equip Slot Object Reference) {is=title K2Node_DynamicCast}
   > Set Panel(Panel:Overlay Object Reference=Get OverlaySlot!(self:Athena Inventory Equip Slot Object Reference=@n4) -> (OverlaySlot:Overlay Object Reference)) -> (Output_Get:Overlay Object Reference)
   n5 = Break ST Add Inventory Quick Bar Overlay!(ST_AddInventoryQuickBarOverlay=Get Payload() -> (Payload:ST Add Inventory Quick Bar Overlay Structure)) -> (AddOrRemove_1_#3:bool=~, WidgetClass_10_#1:User Widget Class Reference)
   > n6 = Branch(Condition=@n5)
   n7 = Get Panel() -> (Panel:Overlay Object Reference)
   > n8 = Is Valid(InputObject=GetFirstWidgetOfType!(self, Type:User Widget Class Reference=@n5.WidgetClass_10_#1, Panel:Panel Widget Object Reference=@n7) -> (Content:Widget Object Reference) {is=title K2Node_CallFunction})
   n9 = Create Widget!(
    execute:Exec=@n8.Is Not Valid
    Class:User Widget Class Reference=@n5.WidgetClass_10_#1
    OwningPlayer:Player Controller Object Reference
   ) -> (then:Exec, ReturnValue:User Widget Object Reference)
   > n10 = AddChildToOverlay!(execute, self:Overlay Object Reference=@n7, Content:Widget Object Reference=@n9) -> (then, ReturnValue:Overlay Slot Object Reference)
   > SetHorizontalAlignment(self:Overlay Slot Object Reference=@n10, InHorizontalAlignment=HAlign_Fill)
   > SetVerticalAlignment(self:Overlay Slot Object Reference=@n10, InVerticalAlignment=VAlign_Fill)
   > InvalidateLayoutAndVolatility(self:Widget Object Reference=@n7)
   RemoveChild(execute=@n6.else, self=@n7, Content=GetFirstWidgetOfType(Type=@n5.WidgetClass_10_#1, Panel=@n7))
  ]}
 LeftNavWrap {nodes [
   LeftNavWrap() {is=title K2Node_FunctionEntry}
   > n1 = Find Item to the Left!(execute, self) -> (then, Item:Fort Item Object Reference, Index:int=0) {is=title K2Node_CallFunction}
   > n2 = Is Valid(InputObject=@n1)
   > n3 = Branch(Condition="Byte > Byte"(A:Byte=GetNumberOfSlotsTaken(self=@n1) -> (ReturnValue=0), B:Byte=1) -> (ReturnValue:bool=~) {is=title K2Node_PromotableOperator})
   > Return Node!(execute, Target:Common Button Base Object Reference=GetButtonBaseAtIndex(self=Get EquipButtonGroup(), Index=@n1.Index))
   n4 = GetSelectedButtonIndex(self=Get EquipButtonGroup()) -> (ReturnValue=0)
   n5 = "integer <= integer"(A=@n4)
   n6 = int - int(A=@n4)
   n7 = int - int(A=Get NumQuickBarSlots() -> (NumQuickBarSlots=0))
   Return Node(
    execute=[@n3.else, @n2.Is Not Valid]
    Target=GetButtonBaseAtIndex(self=Get EquipButtonGroup(), Index=Select(Option 0=@n6, Option 1=@n7, Index=@n5))
   )
  ]}
 On Quickbar Contents Changed {nodes [
   n1 = On Quickbar Contents Changed!() -> (then, QuickbarIndex:EFortQuickBars Enum=Primary, ChangedSlots:Array of Integers) {is=title K2Node_FunctionEntry}
   > Switch on EFortQuickBars!(execute, Selection:EFortQuickBars Enum=@n1) -> (Primary:Exec, Secondary:Exec, Creative:Exec, Max_None:Exec) {is=title K2Node_SwitchEnum}
   > n2 = For Each Loop(Array:Array of Integers=@n1.ChangedSlots) -> (Array Element:int)
   > n3 = Set Current Slot(Current Slot:int=@n2) -> (Output_Get:int=0)
   > Branch(Condition="integer > integer"(A=@n3, B=Get NumQuickBarSlots() -> (NumQuickBarSlots=0)) {is=title K2Node_PromotableOperator})
   n4 = Get FortInventoryContext!() -> (ReturnValue:Fort Inventory Context Object Reference) {is=title K2Node_GetBlueprintContextEx}
   n5 = Get QuickbarIndex() -> (QuickbarIndex:EFortQuickBars Enum=Primary)
   n6 = Get Current Slot() -> (Current Slot:int=0)
   > Set Has Item In Current Slot(Has Item In Current Slot:bool=IsValid(Object=GetQuickBarSlottedItem!(self:Fort Inventory Context Object Reference=@n4, InQuickBar:EFortQuickBars Enum=@n5, Slot:int=@n6) -> (ReturnValue:Fort Item Object Reference))) -> (Output_Get:bool)
   > Set Additional Slot Created(Additional Slot Created:bool=Contains Item(
     TargetArray:Array of Integers=Get Additional Slots() -> (Additional Slots:Array of Integers)
     ItemToFind:"Integer (by ref)"=Get Current Slot() -> (Current Slot=0)
    )) -> (Output_Get:bool)
   n7 = Get Additional Slot Created() -> (Additional Slot Created:bool)
   > n8 = Branch(Condition=AND Boolean(A=Get Has Item In Current Slot() -> (Has Item In Current Slot:bool), B=NOT Boolean(A=@n7)))
   > AddQuickbarSlotButton!(execute, self:Athena Quickbar Editor Base Object Reference, SlotIndex:int=Get Current Slot() -> (Current Slot=0)) -> (then)
   > Add(TargetArray:Array of Integers=Get Additional Slots(), NewItem:"Integer (by ref)"=Get Current Slot() -> (Current Slot=0)) {is=title K2Node_CallArrayFunction}
   n9 = Get Has Item In Current Slot()
   Branch(execute=@n8.else, Condition=AND Boolean(A=NOT Boolean(A=@n9), B=Get Additional Slot Created()))
   > RemoveQuickbarSlotButton!(execute, self:Athena Quickbar Editor Base Object Reference, SlotIndex:int=Get Current Slot() -> (Current Slot=0)) -> (then)
   > Remove Item!(execute, TargetArray:Array of Integers=Get Additional Slots(), Item:"Integer (by ref)"=Get Current Slot() -> (Current Slot=0)) -> (then, ReturnValue:bool)
  ]}
 QuickbarNav {nodes [
   n1 = QuickbarNav!() -> (then, Navigation:EUINavigation Enum) {is=title K2Node_FunctionEntry}
   > n2 = Switch on EUINavigation!(execute, Selection:EUINavigation Enum=@n1) -> (Left:Exec, Right:Exec, Up:Exec, Down:Exec, Next:Exec, Previous:Exec, Invalid:Exec) {is=title K2Node_SwitchEnum}
   n3 = Get ParentInventoryPanel() -> (ParentInventoryPanel:Athena Inventory Panel Weak object)
   n4 = Is Valid(exec=[@n2.Down, @n2.Invalid, @n2.Next, @n2.Previous, @n2.Up], InputObject=@n3)
   Return Node!(execute=@n4.Is Not Valid, ReturnValue:Widget Object Reference)
   > Return Node(ReturnValue=NavQuickbar!(
     self:WBP Athena Inventory Panel Object Reference=Cast To WBP_AthenaInventoryPanel!(Object=@n3) -> (AsWBP Athena Inventory Panel:WBP Athena Inventory Panel Object Reference, bSuccess:bool=~) {is=title K2Node_DynamicCast}
     Navigation:EUINavigation Enum=@n1
    ) -> (ReturnValue:Widget Object Reference))
   n5 = LeftNavWrap!(execute=@n2.Left, self) -> (then, Target:Common Button Base Object Reference) {is=title K2Node_CallFunction}
   > Return Node(ReturnValue=@n5)
   n6 = RightNavWrap!(execute=@n2.Right, self) -> (then, Target:Common Button Base Object Reference)
   > Return Node(ReturnValue=@n6)
  ]}
 RightNavWrap {nodes [
   RightNavWrap() {is=title K2Node_FunctionEntry}
   n1 = Get EquipButtonGroup()
   > n2 = Cast To AthenaInventoryQuickBarSlotButton(Object=GetSelectedButtonBase!(self:Common Button Group Base Object Reference=@n1) -> (ReturnValue:Common Button Base Object Reference))
   n3 = GetSlottedItem(self=@n2)
   > n4 = Is Valid(InputObject=@n3)
   n5 = Set Offset(execute=[@n4.Is Not Valid, @n2.CastFailed], Offset:int=1) -> (Output_Get:int=0)
   n6 = int + int(A=GetSelectedButtonIndex(self=@n1) -> (ReturnValue=0), B=Get Offset() -> (Offset:int=0))
   n7 = Get NumQuickBarSlots() -> (NumQuickBarSlots=0)
   Return Node!(
    execute=[@n8.then, @n5.then]
    Target:Common Button Base Object Reference=GetButtonBaseAtIndex(self=Get EquipButtonGroup(), Index="% (Integer)"(A=@n6, B=@n7) -> (ReturnValue=0))
   )
   n8 = Set Offset(execute=@n4.Is Valid, Offset="To Integer (Byte)"!(InByte:Byte=GetNumberOfSlotsTaken(self=@n3) -> (ReturnValue=0)) -> (ReturnValue:int=0)) -> (Output_Get=0)
  ]}
]
variables [
 Additional Slots {type="array<integer>"}
 Async Task Ref {type=AsyncAction_StartListeningToEvent, default=null}
 Class {type=Class, default=null}
]
widgets {
 info {inheritedWidgetCount=0, parentClass=/Script/FortniteUI.AthenaQuickbarEditorBase, rootWidgetClass=&GridPanel, widgetCount=1}
 widgets [Grid_EquipSlots {
   widgetClassPath=&GridPanel
   bIsVariable=true
   widget=Grid_EquipSlots
   properties {navigation=&self:WidgetTree.Grid_EquipSlots.WidgetNavigation_2}
  }]
}
