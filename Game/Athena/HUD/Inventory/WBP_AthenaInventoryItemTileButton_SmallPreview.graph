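# Widget Blueprint variant of the Athena inventory item tile button that shows small preview images for its item icon.
# It adds no widgets or variables of its own and inherits everything from AthenaInventoryItemTileButton.
# On PreConstruct it runs the parent logic, then turns off the large preview image on ItemIcon.
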WidgetBlueprint /Game/Athena/HUD/Inventory/WBP_AthenaInventoryItemTileButton_SmallPreview
graphs [EventGraph {nodes [
   n1 = Event PreConstruct()
   > Parent: PreConstruct(IsDesignTime=@n1.IsDesignTime)
   > SetUseLargePreviewImage!(
    execute
    self:Athena Item Icon Object Reference=Get ItemIcon() -> (ItemIcon:Athena Item Icon Object Reference)
    bInUseLargePreviewImage:bool=false
   ) -> (then)
  ]}]
variables []
widgets {
 info {
  inheritedWidgetCount=0
  parentClass=&/AthenaInventoryItemTileButton.AthenaInventoryItemTileButton_C
  rootWidgetClass=null
  widgetCount=0
 }
 widgets []
}
