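# Inventory stat row widget that shows two stats side by side under one name, each with value text, delta text and a material-driven bar.
# The native AthenaInventoryDualItemStatEntry parent raises init events; each maps ComparisonValue to Percent and Value to Delta_Percent on the bar's dynamic material, clamped to range.
# Infinite stats force both parameters to 100.
# Item rarity Color1 and Color2 tint only the primary bar's backing and delta colors.
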WidgetBlueprint /Game/Athena/HUD/Inventory/AthenaInventoryDualItemStatEntry
graphs [EventGraph {nodes [
   n1 = Event OnStatInitialized!() -> (OutputDelegate, then, Value:float, ComparisonValue:float, MinValue:float, MaxValue:float) {is=title K2Node_Event}
   > n2 = GetDynamicMaterial(self=Get Image_StatsBar() -> (Image_StatsBar:Image Object Reference))
   > SetScalarParameterValue(
    self=@n2
    ParameterName=Percent
    Value=SelectFloat(
     A=100
     B=MapRangeClamped(Value=@n1.ComparisonValue, InRangeA=@n1.MinValue, InRangeB=@n1.MaxValue, OutRangeB=100)
     bPickA=Get bValueIsInfinite() -> (bValueIsInfinite:bool)
    )
   )
   > SetScalarParameterValue(
    self=@n2
    ParameterName=Delta_Percent
    Value=SelectFloat(A=100, B=MapRangeClamped(Value=@n1.Value, InRangeA=@n1.MinValue, InRangeB=@n1.MaxValue, OutRangeB=100), bPickA=Get bValueIsInfinite())
   )
   n3 = Event OnItemRaritySet!() -> (OutputDelegate, then, ItemRarity:EFortRarity Enum=Common, RarityItemData:Fort Rarity Item Data Structure) {is=title K2Node_Event}
   > n4 = GetDynamicMaterial(self=Get Image_StatsBar())
   n5 = Break Fort Rarity Item Data!(FortRarityItemData:"Fort Rarity Item Data Structure (by ref)"=@n3.RarityItemData) -> (
    Name:Text
    Color1:Linear Color Structure
    Color2:Linear Color Structure
    Color3:Linear Color Structure
    Color4:Linear Color Structure
    Color5:Linear Color Structure
    Radius:float=~
    Falloff:float=~
    Brightness:float=~
    Roughness:float=~
    Glow:float=~
   ) {is=title K2Node_BreakStruct}
   > SetVectorParameterValue(self=@n4, ParameterName=Backing_Color, Value=@n5.Color1)
   > SetVectorParameterValue(self=@n4, ParameterName=Delta_Color, Value=@n5.Color2)
   n6 = Event OnSecondaryStatInitialized!() -> (OutputDelegate, then, Value:float, ComparisonValue:float, MinValue:float, MaxValue:float, bIsInfinite:bool) {is=title K2Node_Event}
   > n7 = GetDynamicMaterial(self=Get Image_SecondaryStatsBar() -> (Image_SecondaryStatsBar:Image Object Reference))
   > SetScalarParameterValue(
    self=@n7
    ParameterName=Percent
    Value=SelectFloat(A=100, B=MapRangeClamped(Value=@n6.ComparisonValue, InRangeA=@n6.MinValue, InRangeB=@n6.MaxValue, OutRangeB=100), bPickA=@n6.bIsInfinite)
   )
   > SetScalarParameterValue(
    self=@n7
    ParameterName=Delta_Percent
    Value=SelectFloat(A=100, B=MapRangeClamped(Value=@n6.Value, InRangeA=@n6.MinValue, InRangeB=@n6.MaxValue, OutRangeB=100), bPickA=@n6.bIsInfinite)
   )
  ]}]
variables []
widgets {
 info {
  inheritedWidgetCount=5
  parentClass=/Script/FortniteUI.AthenaInventoryDualItemStatEntry
  rootWidgetClass=&VerticalBox
  widgetCount=17
 }
 widgets [
  VerticalBox_0 {widgetClassPath=&VerticalBox, widget=VerticalBox_0}
  SizeBox_0 {
   widgetClassPath=&SizeBox
   parent=VerticalBox_0
   widget=SizeBox_0
   properties {bOverride_WidthOverride=true, widthOverride=175, renderTransform {angle=0, scale=<1 1>, shear=<-12 0>, translation=<0 0>}}
   slot {class=VerticalBoxSlot}
  }
  Border_0 {
   widgetClassPath=&Border
   parent=SizeBox_0
   widget=Border_0
   properties {
    renderOpacity=0.5
    background {
     bIsDynamicallyLoaded=false
     drawAs=NoDrawType
     imageType=NoImage
     mirroring=NoMirror
     resourceName=None
     resourceObject=null
     tiling=NoTile
     imageSize=<32 32>
     margin=<ltrb 0 0 0 0>
     outlineSettings {
      bUseBrushTransparency=false
      roundingType=HalfHeightRadius
      width=0
      color {colorUseRule=UseColor_Specified, specifiedColor=<rgba 0 0 0 0>}
      cornerRadii=<0 0 0 0>
     }
     tintColor {colorUseRule=UseColor_Specified, specifiedColor=<rgba 1 1 1 1>}
     uVRegion {bIsValid=false, max=<0 0>, min=<0 0>}
    }
    padding=<ltrb 0 0 0 0>
   }
   slot {class=SizeBoxSlot, properties {verticalAlignment=VAlign_Center}}
  }
  Text_StatName {
   widgetClassPath=&CommonTextBlock
   bInherited=true
   parent=Border_0
   widget=Text_StatName
   properties {
    mobileFontSizeMultiplier=1.15
    scrollStyle=/FortUILibrary/Styles/Typography/Foundation/TextScrollStyle-NoFade.TextScrollStyle-NoFade_C
    text="{Stat Name}"
    textTransformPolicy=ToUpper
    font {
     bForceMonospaced=false
     bMaterialIsStencil=false
     fontMaterial=null
     fontObject=&BurbankSmall
     letterSpacing=0
     monospacedWidth=1
     size=10.75
     skewAmount=0
     typefaceFontName=Black
     outlineSettings {
      bApplyOutlineToDropShadows=false
      bMiteredCorners=false
      bSeparateFillAlpha=false
      outlineMaterial=null
      outlineSize=0
      outlineColor=<rgba 0 0 0 1>
     }
    }
    shadowOffset=<0 0>
   }
   slot {class=BorderSlot, properties {padding=<ltrb 0 0 0 0>}}
  }
  HorizontalBox_2 {
   widgetClassPath=&HorizontalBox
   parent=VerticalBox_0
   widget=HorizontalBox_2
   properties {flowDirectionPreference=Culture}
   slot {class=VerticalBoxSlot, properties {padding=<ltrb 0 0 0 2>}}
  }
  VerticalBox_322 {widgetClassPath=&VerticalBox, parent=HorizontalBox_2, widget=VerticalBox_322, slot {class=HorizontalBoxSlot}}
  SizeBox_1 {
   widgetClassPath=&SizeBox
   parent=VerticalBox_322
   widget=SizeBox_1
   properties {bOverride_WidthOverride=true, widthOverride=140}
   slot {class=VerticalBoxSlot, properties {verticalAlignment=VAlign_Center}}
  }
  HorizontalBox_1 {
   widgetClassPath=&HorizontalBox
   parent=SizeBox_1
   widget=HorizontalBox_1
   slot {class=SizeBoxSlot, properties {verticalAlignment=VAlign_Center}}
  }
  Text_StatValue {
   widgetClassPath=&CommonTextBlock
   bInherited=true
   parent=HorizontalBox_1
   widget=Text_StatValue
   properties {
    mobileFontSizeMultiplier=1.2
    text=888
    font {
     bForceMonospaced=false
     bMaterialIsStencil=false
     fontMaterial=null
     fontObject=&BurbankSmall
     letterSpacing=0
     monospacedWidth=1
     size=12.75
     skewAmount=0
     typefaceFontName=Bold
     outlineSettings {
      bApplyOutlineToDropShadows=false
      bMiteredCorners=false
      bSeparateFillAlpha=false
      outlineMaterial=null
      outlineSize=0
      outlineColor=<rgba 0 0 0 1>
     }
    }
    shadowOffset=<0 0>
   }
   slot {class=HorizontalBoxSlot, properties {verticalAlignment=VAlign_Center}}
  }
  Text_StatDeltaValue {
   widgetClassPath=&CommonTextBlock
   bInherited=true
   parent=HorizontalBox_1
   widget=Text_StatDeltaValue
   properties {
    mobileFontSizeMultiplier=1.2
    text=(+888)
    font {
     bForceMonospaced=false
     bMaterialIsStencil=false
     fontMaterial=/Game/Athena/HUD/Inventory/Art/M_UI_HUD_Stats_DeltaAmount_FontColor
     fontObject=&BurbankSmall
     letterSpacing=0
     monospacedWidth=1
     size=10.75
     skewAmount=0
     typefaceFontName=Bold
     outlineSettings {
      bApplyOutlineToDropShadows=false
      bMiteredCorners=false
      bSeparateFillAlpha=false
      outlineMaterial=null
      outlineSize=0
      outlineColor=<rgba 0 0 0 1>
     }
    }
    renderTransform {angle=0, scale=<1 1>, shear=<-12 0>, translation=<0 0>}
    shadowOffset=<0 0>
   }
   slot {class=HorizontalBoxSlot, properties {verticalAlignment=VAlign_Center, padding=<ltrb 8 0 0 0>}}
  }
  Image_StatsBar {
   widgetClassPath=&Image
   bIsVariable=true
   parent=VerticalBox_322
   widget=Image_StatsBar
   properties {
    visibility=SelfHitTestInvisible
    brush {
     bIsDynamicallyLoaded=false
     drawAs=Image
     imageType=NoImage
     mirroring=NoMirror
     resourceName=None
     resourceObject=/Game/Athena/HUD/Inventory/Art/M_UI_HUD_Inventory_StatsBar_Base
     tiling=NoTile
     imageSize=<32 4>
     margin=<ltrb 0 0 0 0>
     outlineSettings {
      bUseBrushTransparency=false
      roundingType=HalfHeightRadius
      width=0
      color {colorUseRule=UseColor_Specified, specifiedColor=<rgba 0 0 0 0>}
      cornerRadii=<0 0 0 0>
     }
     tintColor {colorUseRule=UseColor_Specified, specifiedColor=<rgba 1 1 1 1>}
     uVRegion {bIsValid=false, max=<0 0>, min=<0 0>}
    }
   }
   slot {class=VerticalBoxSlot, properties {verticalAlignment=VAlign_Bottom, size {sizeRule=Fill, value=1}}}
  }
  VerticalBox {
   widgetClassPath=&VerticalBox
   parent=HorizontalBox_2
   widget=VerticalBox
   slot {class=HorizontalBoxSlot, properties {padding=<ltrb 4 0 0 0>}}
  }
  SizeBox {
   widgetClassPath=&SizeBox
   parent=VerticalBox
   widget=SizeBox
   properties {bOverride_WidthOverride=true, widthOverride=140}
   slot {class=VerticalBoxSlot, properties {verticalAlignment=VAlign_Center}}
  }
  HorizontalBox {
   widgetClassPath=&HorizontalBox
   parent=SizeBox
   widget=HorizontalBox
   slot {class=SizeBoxSlot, properties {verticalAlignment=VAlign_Center}}
  }
  Text_SecondaryStatValue {
   widgetClassPath=&CommonTextBlock
   bInherited=true
   parent=HorizontalBox
   widget=Text_SecondaryStatValue
   properties {
    mobileFontSizeMultiplier=1.2
    text=888
    font {
     bForceMonospaced=false
     bMaterialIsStencil=false
     fontMaterial=null
     fontObject=&BurbankSmall
     letterSpacing=0
     monospacedWidth=1
     size=12.75
     skewAmount=0
     typefaceFontName=Bold
     outlineSettings {
      bApplyOutlineToDropShadows=false
      bMiteredCorners=false
      bSeparateFillAlpha=false
      outlineMaterial=null
      outlineSize=0
      outlineColor=<rgba 0 0 0 1>
     }
    }
    shadowOffset=<0 0>
   }
   slot {class=HorizontalBoxSlot, properties {verticalAlignment=VAlign_Center}}
  }
  Text_SecondaryStatDeltaValue {
   widgetClassPath=&CommonTextBlock
   bInherited=true
   parent=HorizontalBox
   widget=Text_SecondaryStatDeltaValue
   properties {
    mobileFontSizeMultiplier=1.2
    text=(+888)
    font {
     bForceMonospaced=false
     bMaterialIsStencil=false
     fontMaterial=&M_UI_HUD_Stats_DeltaAmount_FontColor
     fontObject=&BurbankSmall
     letterSpacing=0
     monospacedWidth=1
     size=10.75
     skewAmount=0
     typefaceFontName=Bold
     outlineSettings {
      bApplyOutlineToDropShadows=false
      bMiteredCorners=false
      bSeparateFillAlpha=false
      outlineMaterial=null
      outlineSize=0
      outlineColor=<rgba 0 0 0 1>
     }
    }
    renderTransform {angle=0, scale=<1 1>, shear=<-12 0>, translation=<0 0>}
    shadowOffset=<0 0>
   }
   slot {class=HorizontalBoxSlot, properties {verticalAlignment=VAlign_Center, padding=<ltrb 8 0 0 0>}}
  }
  Image_SecondaryStatsBar {
   widgetClassPath=&Image
   bIsVariable=true
   parent=VerticalBox
   widget=Image_SecondaryStatsBar
   properties {
    visibility=SelfHitTestInvisible
    brush {
     bIsDynamicallyLoaded=false
     drawAs=Image
     imageType=NoImage
     mirroring=NoMirror
     resourceName=None
     resourceObject=&M_UI_HUD_Inventory_StatsBar_Base
     tiling=NoTile
     imageSize=<32 4>
     margin=<ltrb 0 0 0 0>
     outlineSettings {
      bUseBrushTransparency=false
      roundingType=HalfHeightRadius
      width=0
      color {colorUseRule=UseColor_Specified, specifiedColor=<rgba 0 0 0 0>}
      cornerRadii=<0 0 0 0>
     }
     tintColor {colorUseRule=UseColor_Specified, specifiedColor=<rgba 1 1 1 1>}
     uVRegion {bIsValid=false, max=<0 0>, min=<0 0>}
    }
   }
   slot {class=VerticalBoxSlot, properties {verticalAlignment=VAlign_Bottom, size {sizeRule=Fill, value=1}}}
  }
 ]
}
