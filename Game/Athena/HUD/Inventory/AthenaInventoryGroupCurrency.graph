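# Currency group widget for the Athena inventory HUD, built on the native AthenaInventoryGroupCurrency class.
# A vertical box holds the inherited ItemTileView, a horizontal row of inventory tile buttons with a Wads empty-tile placeholder for Athena playlists.
# The Nav function delegates navigation to the parent WBP_AthenaInventoryPanel, passing the tile view as the section, and returns nothing without a valid panel.
# The event graph holds only unconnected event stubs.
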WidgetBlueprint /Game/Athena/HUD/Inventory/AthenaInventoryGroupCurrency
graphs [
 EventGraph {nodes [
   Event PreConstruct()
   Event Construct()
   Event Tick()
  ]}
 Nav {nodes [
   n1 = Nav!() -> (then, Navigation:EUINavigation Enum) {is=title K2Node_FunctionEntry}
   n2 = Get ParentInventoryPanel() -> (ParentInventoryPanel:Athena Inventory Panel Weak object)
   > n3 = Is Valid(InputObject=@n2)
   > n4 = Cast To WBP_AthenaInventoryPanel!(execute, Object=@n2) -> (then, CastFailed, AsWBP Athena Inventory Panel:WBP Athena Inventory Panel Object Reference) {is=title K2Node_DynamicCast}
   > Return Node!(
    execute
    ReturnValue:Widget Object Reference=GetNavigatedToWidget!(
     self:WBP Athena Inventory Panel Object Reference=@n4
     Section:Fort Item Tile View Object Reference=Get ItemTileView() -> (ItemTileView:Fort Item Tile View Object Reference)
     Navigation:EUINavigation Enum=@n1
    ) -> (ReturnValue:Widget Object Reference)
   )
   Return Node(execute=[@n4.CastFailed, @n3.Is Not Valid])
  ]}
]
variables []
widgets {
 info {inheritedWidgetCount=1, parentClass=/Script/FortniteUI.AthenaInventoryGroupCurrency, rootWidgetClass=&VerticalBox, widgetCount=2}
 widgets [
  VerticalBox_18 {widgetClassPath=&VerticalBox, widget=VerticalBox_18}
  ItemTileView {
   widgetClassPath=/Script/FortniteUI.FortItemTileView
   bInherited=true
   bIsVariable=true
   parent=VerticalBox_18
   widget=ItemTileView
   properties {
    bSupportEmptyTile=true
    childItemWidgetCardSizeOverride=S
    entryHeight=90
    entryWidgetClass=&/AthenaInventoryItemTileButton.AthenaInventoryItemTileButton_C
    entryWidth=80
    navigation=&self:WidgetTree.ItemTileView.WidgetNavigation_0
    numDesignerPreviewEntries=1
    orientation=Orient_Horizontal
    tileAlignment=EvenlySize
    defaultEmptyTileVisualizationData [{
      emptyTileItemDefinition=/Game/Items/ResourcePickups/Athena_WadsItemData
      allowedPlaylist {gameplayTags [{tagName=Athena.Playlist}]}
      unallowedPlaylist {gameplayTags []}
     }]
   }
   slot {class=VerticalBoxSlot}
  }
 ]
}
