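# Athena inventory widget for the resources group, showing a horizontal tile row with wood, stone and metal empty tiles gated by playlist tags.
# An EmptyView behind the tiles builds its resource array on Construct or design-time PreConstruct.
# Inventory updates pass the tile count to EmptyView to refresh its visuals.
# The Nav function delegates navigation to the parent WBP_AthenaInventoryPanel, returning no widget if that panel is unavailable.
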WidgetBlueprint /Game/Athena/HUD/Inventory/AthenaInventoryGroupResources
graphs [
 EventGraph {nodes [
   "On Inventory Updated Event (ItemTileView)"() {is=title K2Node_ComponentBoundEvent}
   > UpdateResourceViewVisual!(
    execute
    self:Athena Empty Resource Slot Presentation Object Reference=Get EmptyView() -> (EmptyView:Athena Empty Resource Slot Presentation Object Reference)
    ResourcesNumItems:int=GetNumItems!(self:List View Object Reference=Get ItemTileView() -> (ItemTileView:Fort Item Tile View Object Reference)) -> (ReturnValue:int=0)
   ) -> (then)
   n1 = Event PreConstruct()
   > n2 = Branch(Condition=@n1.IsDesignTime)
   MakeResourcesArray(execute=[@n2.then, @n3.then], self:Athena Empty Resource Slot Presentation Object Reference=Get EmptyView())
   n3 = Event Construct()
  ]}
 Nav {nodes [
   n1 = Nav!() -> (then, Navigation:EUINavigation Enum) {is=title K2Node_FunctionEntry}
   n2 = Get ParentInventoryPanel() -> (ParentInventoryPanel:Athena Inventory Panel Weak object)
   > n3 = Is Valid(InputObject=@n2)
   > n4 = Cast To WBP_AthenaInventoryPanel!(execute, Object=@n2) -> (then, CastFailed, AsWBP Athena Inventory Panel:WBP Athena Inventory Panel Object Reference) {is=title K2Node_DynamicCast}
   > Return Node!(
    execute
    ReturnValue:Widget Object Reference=GetNavigatedToWidget!(
     self:WBP Athena Inventory Panel Object Reference=@n4
     Section:Fort Item Tile View Object Reference=Get ItemTileView()
     Navigation:EUINavigation Enum=@n1
    ) -> (ReturnValue:Widget Object Reference)
   )
   Return Node(execute=[@n4.CastFailed, @n3.Is Not Valid])
  ]}
]
variables []
widgets {
 info {inheritedWidgetCount=1, parentClass=/Script/FortniteUI.AthenaInventoryGroupResources, rootWidgetClass=&SizeBox, widgetCount=4}
 widgets [
  SizeBox_1 {widgetClassPath=&SizeBox, widget=SizeBox_1, properties {bOverride_MinDesiredHeight=true, minDesiredHeight=90}}
  Overlay_0 {widgetClassPath=&Overlay, parent=SizeBox_1, widget=Overlay_0, slot {class=SizeBoxSlot}}
  EmptyView {
   widgetClassPath=&/AthenaEmptyResourceSlotPresentation
   bIsVariable=true
   parent=Overlay_0
   widget=EmptyView
   slot {class=OverlaySlot, properties {horizontalAlignment=HAlign_Fill, padding=<ltrb 0 10 0 0>}}
  }
  ItemTileView {
   widgetClassPath=/Script/FortniteUI.FortItemTileView
   bInherited=true
   parent=Overlay_0
   widget=ItemTileView
   properties {
    allowOverscroll=false
    bSupportEmptyTile=true
    childItemWidgetCardSizeOverride=S
    entryHeight=90
    entryWidgetClass=&/AthenaInventoryItemTileButton.AthenaInventoryItemTileButton_C
    entryWidth=80
    horizontalEntrySpacing=8
    navigation=&self:WidgetTree.ItemTileView.WidgetNavigation_0
    numDesignerPreviewEntries=4
    orientation=Orient_Horizontal
    tileAlignment=EvenlySize
    verticalEntrySpacing=8
    defaultEmptyTileVisualizationData [
     {
      emptyTileItemDefinition=/Game/Items/ResourcePickups/WoodItemData
      allowedPlaylist {gameplayTags [{tagName=Athena.Playlist}]}
      unallowedPlaylist {gameplayTags [{tagName=Athena.Playlist.NoBuildingMaterials}]}
     }
     {
      emptyTileItemDefinition=/Game/Items/ResourcePickups/StoneItemData
      allowedPlaylist {gameplayTags [{tagName=Athena.Playlist}]}
      unallowedPlaylist {gameplayTags [{tagName=Athena.Playlist.NoBuildingMaterials}]}
     }
     {
      emptyTileItemDefinition=/Game/Items/ResourcePickups/MetalItemData
      allowedPlaylist {gameplayTags [{tagName=Athena.Playlist}]}
      unallowedPlaylist {gameplayTags [{tagName=Athena.Playlist.NoBuildingMaterials}]}
     }
    ]
   }
   slot {class=OverlaySlot, properties {horizontalAlignment=HAlign_Fill}}
  }
 ]
}
