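# Default Athena HUD inventory panel widget stacking currency, ingredient, resource and ammo groups, with two slots for injected content.
# On Construct it collapses the currency section and registers as tag-keyed widget host, placing added widgets in an extension row or fill-sized special row and broadcasting added groups through the event router.
# GetOrCreateInventoryGroups collects built-in and injected groups.
# Removal handles only the normal row, and the playlist-based currency check is unwired.
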WidgetBlueprint /Game/Athena/HUD/Inventory/AthenaInventoryGroupPanelDefault
graphs [
 EventGraph {nodes [
   n1 = PlaylistDataReadyAsync!(execute:Exec) -> (
    then:Exec
    OnReady:Exec
    GameState:Fort Game State Athena Object Reference
    Playlist:Fort Playlist Object Reference
    PlaylistContextTags:"Gameplay Tag Container Structure (by ref)"
   ) {is=title K2Node_AsyncAction}
   Branch(
    execute=@n1.OnReady
    Condition=HasAnyTags!(
     TagContainer:"Gameplay Tag Container Structure (by ref)"=@n1.PlaylistContextTags
     OtherContainer:"Gameplay Tag Container Structure (by ref)"=Get HideCurrencyInPlaylists() -> (HideCurrencyInPlaylists:Gameplay Tag Container Structure)
     bExactMatch:bool=true
    ) -> (ReturnValue:bool)
   )
   Event Construct()
   > RegisterWidgetHost!(
    execute
    WidgetHost:Widget Object Reference
    WidgetHostTag:"Gameplay Tag Structure (by ref)"=Get RowSlotTag() -> (RowSlotTag:Gameplay Tag Structure)
    IsPlayerSpecific:bool=true
   ) -> (then)
   > RegisterWidgetHost(WidgetHostTag=Get SpecialRowSlotTag() -> (SpecialRowSlotTag:Gameplay Tag Structure), IsPlayerSpecific=true)
   > SetVisibility(self=Get HBox_Currency() -> (HBox_Currency:Vertical Box Object Reference), InVisibility=Collapsed)
   n2 = Event Add Content To Host Widget!() -> (
    OutputDelegate
    then
    WidgetHostTag:"Gameplay Tag Structure (by ref)"
    ContentWidget:Widget Object Reference
    Padding:"Margin Structure (by ref)"
    bFill:bool
   ) {is=title K2Node_Event}
   > n3 = Branch(Condition=MatchesTag!(TagOne:Gameplay Tag Structure=@n2.WidgetHostTag, TagTwo:Gameplay Tag Structure=Get RowSlotTag(), bExactMatch:bool=true) -> (ReturnValue:bool))
   > n4 = AddChild(self=Get ExtensionRowHolder() -> (ExtensionRowHolder:Vertical Box Object Reference), Content=@n2.ContentWidget)
   n5 = Cast To AthenaInventoryGroupBase!(execute=[@n4.then, @n8.then], Object=@n2.ContentWidget) -> (then, CastFailed, AsAthena Inventory Group Base:Athena Inventory Group Base Object Reference) {is=title K2Node_DynamicCast}
   n6 = GetGlobalEventRouter()
   > Broadcast Event!(
    execute
    self:Gameplay Event Router Component Object Reference=@n6
    EventData:On Extension Group Added Structure=Make On Extension Group Added!(AthenaInventoryGroup:Athena Inventory Group Base Weak object=@n5) -> (OnExtensionGroupAdded:On Extension Group Added Structure) {is=title K2Node_MakeStruct}
    EventContext:Object Reference=GetOwningPlayer()
   ) -> (then, ReturnValue:bool)
   Branch(execute=@n3.else, Condition=MatchesTag(TagOne=@n2.WidgetHostTag, TagTwo=Get SpecialRowSlotTag(), bExactMatch=true))
   > n7 = AddChildToHorizontalBox!(
    execute
    self:Horizontal Box Object Reference=Get HBox_SpecialExtensionHolder() -> (HBox_SpecialExtensionHolder:Horizontal Box Object Reference)
    Content:Widget Object Reference=@n2.ContentWidget
   ) -> (then, ReturnValue:Horizontal Box Slot Object Reference)
   > SetPadding!(
    execute
    self:Horizontal Box Slot Object Reference=@n7
    InPadding_Left:float=0
    InPadding_Top:float=10
    InPadding_Right:float=0
    InPadding_Bottom:float=0
   ) -> (then)
   > SetHorizontalAlignment(self:Horizontal Box Slot Object Reference=@n7, InHorizontalAlignment=HAlign_Fill)
   > n8 = SetSize!(execute, self:Horizontal Box Slot Object Reference=@n7, InSize_Value:float=1, InSize_SizeRule:ESlateSizeRule Enum=Fill) -> (then)
   Event Destruct()
   > UnregisterWidgetHost!(execute, WidgetHost:Widget Object Reference, WidgetHostTag:"Gameplay Tag Structure (by ref)"=Get RowSlotTag()) -> (then)
   > UnregisterWidgetHost(WidgetHostTag=Get SpecialRowSlotTag())
   Broadcast Stateful Event!(execute, self:Gameplay Event Router Component Object Reference=@n6, EventData:"Wildcard (by ref)", EventContext:Object Reference) -> (then, ReturnValue:bool)
  ]}
 GetOrCreateInventoryGroups {nodes [
   GetOrCreateInventoryGroups() {is=title K2Node_FunctionEntry}
   > Add(
    TargetArray:Array of Athena Inventory Group Base Object References=Get InventoryGroupArray() -> (InventoryGroupArray:Array of Athena Inventory Group Base Object References)
    NewItem:"Athena Inventory Group Base Object Reference (by ref)"=Get AthenaInventoryGroupCurrency() -> (AthenaInventoryGroupCurrency:Athena Inventory Group Currency Object Reference)
   ) {is=title K2Node_CallArrayFunction}
   > n1 = For Each Loop(Array:Array of Widget Object References=GetAllChildren(self=Get HBox_SpecialExtensionHolder())) -> (Array Element:Widget Object Reference)
   n2 = Get InventoryGroupArray()
   Add(
    execute=@n1.Completed
    TargetArray=@n2
    NewItem=Get AthenaInventoryGroupIngredients() -> (AthenaInventoryGroupIngredients:Athena Inventory Group Ingredients Object Reference)
   )
   > Add(
    TargetArray=@n2
    NewItem=Get AthenaInventoryGroupResources() -> (AthenaInventoryGroupResources:Athena Inventory Group Resources Object Reference)
   )
   > Add(TargetArray=@n2, NewItem=Get AthenaInventoryGroupAmmo() -> (AthenaInventoryGroupAmmo:Athena Inventory Group Ammo Object Reference))
   > n3 = For Each Loop(Array=GetAllChildren(self=Get ExtensionRowHolder()))
   Return Node!(execute=@n3.Completed, OutInventoryGroups:Array of Athena Inventory Group Base Object References=Get InventoryGroupArray())
   > n4 = Cast To AthenaInventoryGroupBase(Object=@n3)
   > Add Unique(
    TargetArray:Array of Athena Inventory Group Base Object References=Get InventoryGroupArray()
    NewItem:"Athena Inventory Group Base Object Reference (by ref)"=@n4
   )
   n5 = Cast To AthenaInventoryGroupBase(execute=@n1.LoopBody, Object=@n1)
   > Add Unique(TargetArray=Get InventoryGroupArray(), NewItem=@n5)
  ]}
 RemoveContentFromWidgetHost {nodes [
   Remove Content From Host Widget!() -> (
    then
    WidgetHostTag:"Gameplay Tag Structure (by ref)"
    ContentWidget:Widget Object Reference
    RemovalType:EWidgetSlottingRemovalType Enum=Unparent
   ) {is=title K2Node_FunctionEntry}
   > Branch(Condition=MatchesTag(TagOne=Get RowSlotTag(), TagTwo=Get WidgetHostTag() -> (WidgetHostTag:"Gameplay Tag Structure (by ref)"), bExactMatch=true))
   > Is Valid(InputObject=Get ExtensionRowHolder())
   > Branch(Condition=HasChild!(
     self:Panel Widget Object Reference=Get ExtensionRowHolder()
     Content:Widget Object Reference=Get ContentWidget() -> (ContentWidget:Widget Object Reference)
    ) -> (ReturnValue:bool))
   > RemoveFromParent(self=Get ContentWidget())
   > Return Node!(execute, ReturnValue:bool=true)
  ]}
]
variables [
 HideCurrencyInPlaylists {type=GameplayTagContainer, default {gameplayTags [{tagName=Athena.Playlist.Figment}]}}
 InventoryGroupArray {type="array<AthenaInventoryGroupBase>", default []}
 RowSlotTag {type=GameplayTag, default {tagName=DynamicUI.HUD.Fort.Athena.Inventory.Row}}
 SpecialRowSlotTag {type=GameplayTag, default {tagName=DynamicUI.HUD.Fort.Athena.Inventory.SpecialRow}}
]
widgets {
 info {inheritedWidgetCount=1, parentClass=/Script/FortniteUI.AthenaInventoryGroupPanel, rootWidgetClass=&VerticalBox, widgetCount=12}
 widgets [
  VBox_Inventory {widgetClassPath=&VerticalBox, widget=VBox_Inventory}
  VerticalBox_0 {
   widgetClassPath=&VerticalBox
   parent=VBox_Inventory
   widget=VerticalBox_0
   properties {bOverride_Cursor=true}
   slot {class=VerticalBoxSlot}
  }
  HBox_Currency {widgetClassPath=&VerticalBox, bIsVariable=true, parent=VerticalBox_0, widget=HBox_Currency, slot {class=VerticalBoxSlot}}
  CommonTextBlock {
   widgetClassPath=&CommonTextBlock
   parent=HBox_Currency
   widget=CommonTextBlock
   properties {
    lineHeightPercentage=0.9
    mobileFontSizeMultiplier=1.2
    scrollStyle=&TextScrollStyle-Base_C
    text=Currencies
    textTransformPolicy=ToUpper
    font {
     bForceMonospaced=false
     bMaterialIsStencil=false
     fontMaterial=/Game/Athena/HUD/Inventory/Art/M_UI_Inventory_CatergoryHeader_FontColor
     fontObject=&BurbankSmall
     letterSpacing=0
     monospacedWidth=1
     size=13
     skewAmount=0
     typefaceFontName=Black
     outlineSettings {
      bApplyOutlineToDropShadows=false
      bMiteredCorners=false
      bSeparateFillAlpha=true
      outlineMaterial=&M-FontShadow
      outlineSize=2
      outlineColor=<rgba 0 0 0 0.5>
     }
    }
    margin=<ltrb 0 -2 0 0>
    renderTransform {angle=0, scale=<1 1>, shear=<-12 0>, translation=<0 0>}
    shadowOffset=<0 0>
   }
   slot {class=VerticalBoxSlot}
  }
  AthenaInventoryGroupCurrency {
   widgetClassPath=&/AthenaInventoryGroupCurrency
   bIsVariable=true
   parent=HBox_Currency
   widget=AthenaInventoryGroupCurrency
   slot {class=VerticalBoxSlot}
  }
  HBox_SpecialExtensionHolder {
   widgetClassPath=&HorizontalBox
   bIsVariable=true
   parent=VerticalBox_0
   widget=HBox_SpecialExtensionHolder
   properties {navigation=&self:WidgetTree.HBox_SpecialExtensionHolder.WidgetNavigation_0}
   slot {class=VerticalBoxSlot, properties {size {sizeRule=Fill, value=1}}}
  }
  VBox_Resources {
   widgetClassPath=&VerticalBox
   parent=VBox_Inventory
   widget=VBox_Resources
   slot {class=VerticalBoxSlot, properties {verticalAlignment=VAlign_Top, padding=<ltrb 0 10 0 0>}}
  }
  CommonTextBlock_1 {
   widgetClassPath=&CommonTextBlock
   parent=VBox_Resources
   widget=CommonTextBlock_1
   properties {
    lineHeightPercentage=0.9
    mobileFontSizeMultiplier=1.2
    scrollStyle=&TextScrollStyle-Base_C
    text=Resources
    textTransformPolicy=ToUpper
    font {
     bForceMonospaced=false
     bMaterialIsStencil=false
     fontMaterial=&M_UI_Inventory_CatergoryHeader_FontColor
     fontObject=&BurbankSmall
     letterSpacing=0
     monospacedWidth=1
     size=13
     skewAmount=0
     typefaceFontName=Black
     outlineSettings {
      bApplyOutlineToDropShadows=false
      bMiteredCorners=false
      bSeparateFillAlpha=true
      outlineMaterial=&M-FontShadow
      outlineSize=2
      outlineColor=<rgba 0 0 0 0.5>
     }
    }
    margin=<ltrb 0 -2 0 0>
    renderTransform {angle=0, scale=<1 1>, shear=<-12 0>, translation=<0 0>}
    shadowOffset=<0 0>
   }
   slot {class=VerticalBoxSlot}
  }
  AthenaInventoryGroupIngredients {
   widgetClassPath=&/AthenaInventoryGroupIngredients
   bIsVariable=true
   parent=VBox_Resources
   widget=AthenaInventoryGroupIngredients
   properties {inventoryContentSortType=BySubtype}
   slot {class=VerticalBoxSlot, properties {padding=<ltrb 0 3 0 3>}}
  }
  AthenaInventoryGroupResources {
   widgetClassPath=&/AthenaInventoryGroupResources
   bIsVariable=true
   parent=VBox_Resources
   widget=AthenaInventoryGroupResources
   slot {class=VerticalBoxSlot, properties {padding=<ltrb 0 3 0 3>}}
  }
  AthenaInventoryGroupAmmo {
   widgetClassPath=&/AthenaInventoryGroupAmmo
   bIsVariable=true
   parent=VBox_Resources
   widget=AthenaInventoryGroupAmmo
   slot {class=VerticalBoxSlot, properties {padding=<ltrb 0 3 0 3>}}
  }
  ExtensionRowHolder {
   widgetClassPath=&VerticalBox
   bInherited=true
   bIsVariable=true
   parent=VBox_Resources
   widget=ExtensionRowHolder
   slot {class=VerticalBoxSlot, properties {padding=<ltrb 0 3 0 3>}}
  }
 ]
}
