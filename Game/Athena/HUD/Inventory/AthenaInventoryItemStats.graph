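# Widget Blueprint for the Athena inventory HUD that lays out item stat entries, built on the native AthenaInventoryItemStatsWidget parent.
# An overlay stacks two fixed-width size boxes, each holding a wrapping, top-left aligned dynamic entry box inherited from the parent.
# One box spawns standard stat entry widgets and the other spawns dual-item stat entries.
# The event graph is empty, so the Blueprint supplies only layout and entry classes.
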WidgetBlueprint /Game/Athena/HUD/Inventory/AthenaInventoryItemStats
graphs [EventGraph {links [], nodes []}]
variables []
widgets {
 info {inheritedWidgetCount=2, parentClass=/Script/FortniteUI.AthenaInventoryItemStatsWidget, rootWidgetClass=&Overlay, widgetCount=5}
 widgets [
  Overlay_Root {widgetClassPath=&Overlay, widget=Overlay_Root}
  SizeBox_Width {
   widgetClassPath=&SizeBox
   parent=Overlay_Root
   widget=SizeBox_Width
   properties {bOverride_WidthOverride=true, widthOverride=688}
   slot {class=OverlaySlot}
  }
  EntryBox_Stats {
   widgetClassPath=&DynamicEntryBox
   bInherited=true
   parent=SizeBox_Width
   widget=EntryBox_Stats
   properties {
    entryBoxType=Wrap
    entryHorizontalAlignment=HAlign_Left
    entryVerticalAlignment=VAlign_Top
    entryWidgetClass=&/AthenaInventoryItemStatEntry.AthenaInventoryItemStatEntry_C
    numDesignerPreviewEntries=4
    entrySpacing=<30 12>
   }
   slot {class=SizeBoxSlot}
  }
  SizeBox_Width_DualItem {
   widgetClassPath=&SizeBox
   parent=Overlay_Root
   widget=SizeBox_Width_DualItem
   properties {bOverride_WidthOverride=true, widthOverride=688}
   slot {class=OverlaySlot}
  }
  EntryBox_Stats_DualItem {
   widgetClassPath=&DynamicEntryBox
   bInherited=true
   parent=SizeBox_Width_DualItem
   widget=EntryBox_Stats_DualItem
   properties {
    entryBoxType=Wrap
    entryHorizontalAlignment=HAlign_Left
    entryVerticalAlignment=VAlign_Top
    entryWidgetClass=&/AthenaInventoryDualItemStatEntry.AthenaInventoryDualItemStatEntry_C
    numDesignerPreviewEntries=4
    entrySpacing=<30 12>
   }
   slot {class=SizeBoxSlot}
  }
 ]
}
