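# Athena inventory panel widget that stacks scale-to-fit item group panels above an item info and quickbar editor area.
# Up/Down navigation searches InventorySections for focusable, non-empty tile views, falling back to the first quickbar button.
# Inventory drags over the panel show a drop icon, and dropping cancels the swap.
# It also computes desired height from both holders, fades in the lower area, and changes quickbar order when the backwards combat count changes.
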WidgetBlueprint /Game/Athena/HUD/Inventory/WBP_AthenaInventoryPanel
animations [
 HoverResources {binding_count=0, length_seconds=0.2, object=&self:HoverResources, sequence {bindings []}}
 Intro {
  binding_count=2
  length_seconds=0.3001
  object=&self:Intro
  sequence {
   bindings [
    ItemInfoWidget {
     id=#1
     possessed_class=AthenaInventoryItemInfo_C
     tracks [{
       type=MovieSceneFloatTrack
       display_name=Render Opacity
       sections [{type=MovieSceneFloatSection, channels [None {type=MovieSceneScriptingFloatChannel, keys [[-6,0],[-0.001,0],[12.005,1]]}]}]
      }]
    }
    DefaultQuickbarEditor {
     id=#2
     possessed_class=AthenaQuickbarEditorWidget_C
     tracks [{
       type=MovieSceneFloatTrack
       display_name=Render Opacity
       sections [{type=MovieSceneFloatSection, channels [None {type=MovieSceneScriptingFloatChannel, keys [[-6,0],[-0.001,0],[12.005,1]]}]}]
      }]
    }
   ]
   display_rate {numerator=60}
  }
 }
]
graphs [
 EventGraph {nodes [
   n1 = ResourceAnimFinished() {is=title K2Node_CustomEvent}
   > Set IsResourceAnimFinished(IsResourceAnimFinished:bool=true) -> (Output_Get:bool)
   n2 = Branch()
   > PlayAnimationForward(InAnimation=Get HoverResources() -> (HoverResources:Widget Animation Object Reference))
   > BindToAnimationFinished(Animation=Get HoverResources(), Delegate=@n1)
   Branch(execute=@n2.else, Condition=Get IsResourceAnimFinished() -> (IsResourceAnimFinished:bool))
   > PlayAnimationReverse(InAnimation=Get HoverResources())
   > Set IsResourceAnimFinished(IsResourceAnimFinished=false)
  ]}
 GetDesiredHeight {nodes [
   GetDesiredHeight() {is=title K2Node_FunctionEntry}
   n1 = BreakVector2D(InVec=GetDesiredSize!(self:Widget Object Reference=Get QuickbarEditorHolder() -> (QuickbarEditorHolder:Vertical Box Object Reference)) -> (ReturnValue:Vector 2D Structure))
   n2 = BreakVector2D(InVec=GetDesiredSize(self=Get GroupPanelHolder() -> (GroupPanelHolder:Panel Widget Object Reference)))
   > Return Node!(execute, OutHeight:double="Max (Float)"(A=704, B=float + float!(A=@n2.Y, B=@n1.Y, C:double=100) -> (ReturnValue)))
  ]}
 GetFirstQuickbarButton {nodes [
   Inputs!()
   n1 = GetFirstEquipSlotButton!(self:Athena Quickbar Editor Base Object Reference=Get CurrentQuickbarEditor() -> (CurrentQuickbarEditor:Athena Quickbar Editor Base Object Reference)) -> (ReturnValue:Widget Object Reference)
   Outputs!(Widget:Widget Object Reference=Select(Option 0:Widget Object Reference, Option 1:Widget Object Reference=@n1, Index=IsValid(Object=@n1)) -> (ReturnValue:Widget Object Reference))
  ]}
 GetNavigatedToWidget {nodes [
   n1 = GetNavigatedToWidget!() -> (then, Section:Fort Item Tile View Object Reference, Navigation:EUINavigation Enum) {is=title K2Node_FunctionEntry}
   n2 = GetPreviousSection!(CurrentSection:Fort Item Tile View Object Reference=@n1) -> (PreviousSection:Fort Item Tile View Object Reference)
   n3 = IsValid(Object=@n2)
   n4 = Get First Quickbar Button!() -> (Widget:Widget Object Reference) {is=title K2Node_MacroInstance}
   n5 = GetNextSection!(CurrentSection:Fort Item Tile View Object Reference=@n1) -> (NextSection:Fort Item Tile View Object Reference)
   n6 = IsValid(Object=@n5)
   n7 = Get First Quickbar Button()
   > Return Node!(
    execute
    ReturnValue:Widget Object Reference=Select!(
     Left:Widget Object Reference
     Right:Widget Object Reference
     Up:Widget Object Reference=Select!(Option 0:Widget Object Reference=@n4, Option 1:Widget Object Reference=@n2, Index:bool=@n3) -> (ReturnValue)
     Down:Widget Object Reference=Select(Option 0=@n7, Option 1=@n5, Index=@n6)
     Next:Widget Object Reference
     Previous:Widget Object Reference
     Invalid:Widget Object Reference
     Index:EUINavigation Enum=@n1.Navigation
    ) -> (ReturnValue)
   )
  ]}
 GetNextSection {nodes [
   n1 = GetNextSection!() -> (then, CurrentSection:Fort Item Tile View Object Reference) {is=title K2Node_FunctionEntry}
   n2 = Get InventorySections() -> (InventorySections:Array of Fort Item Tile View Object References)
   n3 = Find Item!(TargetArray:Array of Fort Item Tile View Object References=@n2, ItemToFind:"Fort Item Tile View Object Reference (by ref)"=@n1) -> (ReturnValue)
   > n4 = For Loop(
    FirstIndex=Select(
     Option 0:int=0
     Option 1:int=int + int(A=@n3, B=1) -> (ReturnValue=0) {is=title K2Node_CommutativeAssociativeBinaryOperator}
     Index=IsValid(Object=@n1)
    ) -> (ReturnValue:int)
    LastIndex=int - int(A=Length(TargetArray:Array of Fort Item Tile View Object References=@n2), B=1) {is=title K2Node_CallFunction}
   )
   n5 = Get InventorySections()
   > n6 = Branch(Condition=Is Valid Index!(TargetArray:Array of Fort Item Tile View Object References=@n5, IndexToTest:int=@n4) -> (ReturnValue:bool))
   n7 = "Get (a copy)"(Array:Array of Fort Item Tile View Object References=@n5, Dimension 1=@n4) -> (Output:Fort Item Tile View Object Reference)
   > Branch(Condition=Get bIsFocusable!(self:List View Object Reference=@n7) -> (bIsFocusable:bool))
   > Branch(Condition="integer > integer"(A=GetNumItems!(self:List View Object Reference=@n7) -> (ReturnValue:int=0), B=0) {is=title K2Node_CallFunction})
   > Return Node!(execute, NextSection:Fort Item Tile View Object Reference=@n7)
   Return Node(execute=[@n6.else, @n4.Completed])
  ]}
 GetPreviousSection {nodes [
   n1 = GetPreviousSection!() -> (then, CurrentSection:Fort Item Tile View Object Reference) {is=title K2Node_FunctionEntry}
   n2 = Get InventorySections()
   n3 = Find Item(TargetArray=@n2, ItemToFind=@n1)
   n4 = Length(TargetArray=@n2)
   > n5 = Set Local_Index(Local_Index:int=Select(Option 0=int - int(A=@n4, B=1), Option 1=int - int(A=@n3, B=1), Index=IsValid(Object=@n1))) -> (Output_Get:int=0)
   > n6 = While Loop!(execute:Exec, Condition:bool="integer >= integer"!(A:int=@n5, B:int=0) -> (ReturnValue:bool) {is=title K2Node_CallFunction}) -> (LoopBody:Exec, Completed:Exec) {is=title K2Node_MacroInstance}
   n7 = Get InventorySections()
   n8 = Get Local_Index() -> (Local_Index:int=0)
   > n9 = Branch(Condition=Is Valid Index(TargetArray=@n7, IndexToTest=@n8))
   n10 = "Get (a copy)"(Array=@n7, Dimension 1=@n8)
   > n11 = Branch(Condition=Get bIsFocusable(self=@n10))
   > n12 = Branch(Condition="integer > integer"(A=GetNumItems(self=@n10) -> (ReturnValue=0), B=0))
   > Return Node!(execute, PreviousSection:Fort Item Tile View Object Reference=@n10)
   Decrement Int(" "=[@n12.else, @n11.else], Value=Get Local_Index() -> (Local_Index=0))
   Return Node(execute=[@n9.else, @n6.Completed])
  ]}
 HandleIntroAnim {nodes [
   PlayAnimationForward(InAnimation=Get Intro(), PlaybackSpeed=2)
   HandleIntroAnim() {is=title K2Node_FunctionEntry}
  ]}
 HandleQuickbarsCountChanged {nodes [
   n1 = HandleQuickbarsCountChanged!() -> (then, NewBackwardsCombatQuickbarsCount:int) {is=title K2Node_FunctionEntry}
   > ChangeQuickbarOrder!(execute, self:Athena Inventory Panel Object Reference, bIsBackwards:bool="integer > integer"(A=@n1) {is=title K2Node_PromotableOperator}) -> (then)
  ]}
 NavQuickbar {nodes [
   n1 = NavQuickbar!() -> (then, Navigation:EUINavigation Enum) {is=title K2Node_FunctionEntry}
   > Return Node!(
    execute
    ReturnValue:Widget Object Reference=GetNavigatedToWidget!(Section:Fort Item Tile View Object Reference, Navigation:EUINavigation Enum=@n1) -> (ReturnValue:Widget Object Reference) {is=title K2Node_CallFunction}
   )
  ]}
 OnDragOver {nodes [
   n1 = OnDragOver!() -> (then, MyGeometry:Geometry Structure, PointerEvent:Pointer Event Structure, Operation:Drag Drop Operation Object Reference) {is=title K2Node_FunctionEntry}
   > n2 = Cast To AthenaInventoryDragDrop!(execute, Object=@n1.Operation) -> (then, CastFailed, AsAthena Inventory Drag Drop:Athena Inventory Drag Drop Object Reference) {is=title K2Node_DynamicCast}
   > ShowDropIcon!(execute, self:Athena Inventory Drag Drop Object Reference=@n2, Drop Action:AthenaDragDropAction Enum=NewEnumerator0) -> (then)
   > Return Node(ReturnValue:bool=true)
   Return Node(execute=@n2.CastFailed, ReturnValue=false)
  ]}
 OnDrop {nodes [
   n1 = OnDrop!() -> (then, MyGeometry:Geometry Structure, PointerEvent:Pointer Event Structure, Operation:Drag Drop Operation Object Reference) {is=title K2Node_FunctionEntry}
   > n2 = Sequence()
   n3 = Cast To AthenaInventoryDragDrop!(Object=@n1.Operation) -> (AsAthena Inventory Drag Drop, bSuccess:bool=~)
   > Branch(Condition=@n3.bSuccess)
   > CancelSwap(self:Athena Inventory Context Object Reference=Get AthenaInventoryContext!() -> (ReturnValue:Athena Inventory Context Object Reference) {is=title K2Node_GetBlueprintContextEx})
   Return Node(execute=@n2.then_1, ReturnValue=@n3.bSuccess)
  ]}
]
variables [
 BacchusQuickbarViewModelClass {type=Class, default=/Script/FortniteUI.BacchusQuickbarViewModel}
 BacchusQuickbarVM_Ref {type=MVVMViewModelBase, default=null}
 IsResourceAnimFinished {type=boolean}
]
widgets {
 info {inheritedWidgetCount=5, parentClass=/Script/FortniteUI.AthenaInventoryPanel, rootWidgetClass=&Overlay, widgetCount=8}
 widgets [
  Overlay_Root {widgetClassPath=&Overlay, widget=Overlay_Root}
  VerticalBox_0 {
   widgetClassPath=&VerticalBox
   parent=Overlay_Root
   widget=VerticalBox_0
   slot {class=OverlaySlot, properties {horizontalAlignment=HAlign_Fill, verticalAlignment=VAlign_Fill}}
  }
  ScaleBox_0 {
   widgetClassPath=&ScaleBox
   parent=VerticalBox_0
   widget=ScaleBox_0
   properties {stretch=ScaleToFitY, stretchDirection=DownOnly}
   slot {class=VerticalBoxSlot, properties {padding=<ltrb 35 10 35 0>, size {sizeRule=Fill, value=1}}}
  }
  GroupPanelHolder {
   widgetClassPath=&VerticalBox
   bInherited=true
   parent=ScaleBox_0
   widget=GroupPanelHolder
   slot {class=ScaleBoxSlot, properties {horizontalAlignment=HAlign_Fill, verticalAlignment=VAlign_Fill}}
  }
  DefaultGroupPanel {
   widgetClassPath=&/AthenaInventoryGroupPanelDefault
   bInherited=true
   bIsVariable=true
   parent=GroupPanelHolder
   widget=DefaultGroupPanel
   slot {class=VerticalBoxSlot}
  }
  QuickbarEditorHolder {
   widgetClassPath=&VerticalBox
   bInherited=true
   parent=VerticalBox_0
   widget=QuickbarEditorHolder
   slot {class=VerticalBoxSlot, properties {verticalAlignment=VAlign_Bottom}}
  }
  ItemInfoWidget {
   widgetClassPath=&/AthenaInventoryItemInfo
   bInherited=true
   bIsVariable=true
   parent=QuickbarEditorHolder
   widget=ItemInfoWidget
   slot {class=VerticalBoxSlot, properties {verticalAlignment=VAlign_Bottom}}
  }
  DefaultQuickbarEditor {
   widgetClassPath=&/AthenaQuickBarEditorWidget.AthenaQuickbarEditorWidget
   bInherited=true
   parent=QuickbarEditorHolder
   widget=DefaultQuickbarEditor
   slot {class=VerticalBoxSlot, properties {horizontalAlignment=HAlign_Center, padding=<ltrb 20 20 20 40>}}
  }
 ]
}
