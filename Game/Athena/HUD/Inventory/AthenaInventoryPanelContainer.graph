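# Battle Royale inventory screen container holding a tab bar and switcher with the inventory panel on the right.
# It registers as a widget host for two gameplay tags, placing extension widgets on the left and top-left.
# Activation, after a short delay, sizes the panel and positions left extensions by console variable.
# Drag-drops drop items or half stacks via the inventory context; a CVar-gated touch tap dismisses the screen.
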WidgetBlueprint /Game/Athena/HUD/Inventory/AthenaInventoryPanelContainer
animations [
 FadeOut {
  binding_count=1
  length_seconds=0.25
  object=&self:FadeOut
  sequence {bindings [Overlay_1 {
     id=#1
     possessed_class=Overlay
     tracks [{
       type=MovieSceneFloatTrack
       display_name=Render Opacity
       sections [{type=MovieSceneFloatSection, channels [None {type=MovieSceneScriptingFloatChannel, keys [[0,1],[5,0.25]]}]}]
      }]
    }]}
 }
 ShowTopBar {
  binding_count=1
  length_seconds=0.4
  object=&self:ShowTopBar
  sequence {bindings [TabList {
     id=#2
     possessed_class=HorizontalTabList_Legacy_C
     tracks [
      {
       type=MovieSceneFloatTrack
       display_name=Render Opacity
       sections [{type=MovieSceneFloatSection, channels [None {type=MovieSceneScriptingFloatChannel, keys [[-2,0],[0,0],[6,1]]}]}]
      }
      {
       type=MovieScene2DTransformTrack
       sections [{
         type=MovieScene2DTransformSection
         channels [
          Translation.X {type=MovieSceneScriptingFloatChannel}
          Translation.Y {type=MovieSceneScriptingFloatChannel, keys [[0,-128],[6,0]]}
          Angle {type=MovieSceneScriptingFloatChannel}
          Scale.X {type=MovieSceneScriptingFloatChannel}
          Scale.Y {type=MovieSceneScriptingFloatChannel}
          Shear.X {type=MovieSceneScriptingFloatChannel}
          Shear.Y {type=MovieSceneScriptingFloatChannel}
         ]
        }]
      }
     ]
    }]}
 }
]
graphs [
 EventGraph {nodes [
   "On Tab Button Creation (TabList)"!() -> (OutputDelegate, then, TabId:Name=None, TabButton:Common Button Base Object Reference) {is=title K2Node_ComponentBoundEvent}
   n1 = Get TabList() -> (TabList:Fort Tab List Widget Base Legacy Object Reference)
   n2 = GetTabCount!(self:Common Tab List Widget Base Object Reference=@n1) -> (ReturnValue:int=0)
   > SetVisibility(
    self=@n1
    InVisibility=Select(Option 0=Collapsed, Option 1=SelfHitTestInvisible, Index="integer > integer"(A=@n2, B=1) {is=title K2Node_CallFunction})
   )
   "On Clicked (DismissInventory)"() {is=title K2Node_ComponentBoundEvent}
   > n3 = GetConsoleVariableBoolValue!(execute, VariableName:String=Fort.Touch.DismissInventoryWithTapAnywhere) -> (then, ReturnValue:bool)
   n4 = GetOwningPlayer()
   > Branch(Condition=AND Boolean(A=@n3, B=IsUsingTouch(PlayerController=@n4)))
   > n5 = Cast To FortPlayerController(Object=@n4)
   > ToggleInventory(self:Fort Player Controller Object Reference=@n5)
   Event Construct()
   > SetTouchStyle()
   > RegisterWidgetHost!(
    execute
    WidgetHost:Widget Object Reference
    WidgetHostTag:"Gameplay Tag Structure (by ref)"=Get LeftSideSlotTag() -> (LeftSideSlotTag:Gameplay Tag Structure)
    IsPlayerSpecific:bool=true
   ) -> (then)
   > RegisterWidgetHost(WidgetHostTag=Get TopLeftSlotTag() -> (TopLeftSlotTag:Gameplay Tag Structure), IsPlayerSpecific=true)
   Event On Activated()
   > Delay(Duration=0.2)
   > HandleInventoryExtenstion()
   > n6 = GetConsoleVariableBoolValue(VariableName=BattleRoyalGameUI.SquadListPositionAtBottom.Enabled)
   > SetLeftVBoxPosition!(
    execute
    self
    NewPosition:"Athena Inventory Panel Padding Alignment Structure (by ref)"=Select(
     Option 0:"Athena Inventory Panel Padding Alignment Structure (by ref)"=Get LeftVBoxPosition_Default() -> (LeftVBoxPosition_Default:Athena Inventory Panel Padding Alignment Structure)
     Option 1:"Athena Inventory Panel Padding Alignment Structure (by ref)"=Get LeftVBoxPosition_BottomSquad() -> (LeftVBoxPosition_BottomSquad:Athena Inventory Panel Padding Alignment Structure)
     Index=@n6
    ) -> (ReturnValue:"Athena Inventory Panel Padding Alignment Structure (by ref)")
   ) -> (then)
   n7 = Event Add Content To Host Widget!() -> (
    OutputDelegate
    then
    WidgetHostTag:"Gameplay Tag Structure (by ref)"
    ContentWidget:Widget Object Reference
    Padding:"Margin Structure (by ref)"
    bFill:bool
   ) {is=title K2Node_Event}
   > n8 = Branch(Condition=MatchesTag!(TagOne:Gameplay Tag Structure=@n7.WidgetHostTag, TagTwo:Gameplay Tag Structure=Get LeftSideSlotTag(), bExactMatch:bool=true) -> (ReturnValue:bool))
   > n9 = AddChild(
    self=Get VerticalBox_InventoryScreenLeftSideExtensionHolder() -> (VerticalBox_InventoryScreenLeftSideExtensionHolder:Vertical Box Object Reference)
    Content=@n7.ContentWidget
   )
   Branch(execute=[@n9.then, @n8.else], Condition=MatchesTag(TagOne=@n7.WidgetHostTag, TagTwo=Get TopLeftSlotTag(), bExactMatch=true))
   > AddChild(
    self=Get VerticalBox_InventoryScreenTopLeftExtensionHolder() -> (VerticalBox_InventoryScreenTopLeftExtensionHolder:Vertical Box Object Reference)
    Content=@n7.ContentWidget
   )
   Event Destruct()
   > UnregisterWidgetHost!(execute, WidgetHost:Widget Object Reference, WidgetHostTag:"Gameplay Tag Structure (by ref)"=Get LeftSideSlotTag()) -> (then)
   > UnregisterWidgetHost(WidgetHostTag=Get TopLeftSlotTag())
  ]}
 HandleIntroAnim {nodes [
   HandleIntroAnim() {is=title K2Node_FunctionEntry}
   > PlayAnimationForward(InAnimation=Get ShowTopBar() -> (ShowTopBar:Widget Animation Object Reference), PlaybackSpeed=1.25)
   > n1 = Cast To WBP_AthenaInventoryPanel!(execute, Object=Get InventoryPanel() -> (InventoryPanel:Athena Inventory Panel Object Reference)) -> (then, CastFailed, AsWBP Athena Inventory Panel:WBP Athena Inventory Panel Object Reference) {is=title K2Node_DynamicCast}
   > HandleIntroAnim(self:WBP Athena Inventory Panel Object Reference=@n1) {is=title K2Node_CallFunction}
  ]}
 HandleInventoryExtenstion {nodes [
   HandleInventoryExtenstion() {is=title K2Node_FunctionEntry}
   > n1 = Cast To WBP_AthenaInventoryPanel(Object=Get InventoryPanel())
   n2 = GetDesiredHeight!(self:WBP Athena Inventory Panel Object Reference=@n1) -> (OutHeight:double)
   > n3 = Branch(Condition="Equal (Float)"(A=@n2))
   > SetHeightOverride(
    self=Get SizeBox_InventoryPanel() -> (SizeBox_InventoryPanel:Size Box Object Reference)
    InHeightOverride=Get InitSizeBoxHeight() -> (InitSizeBoxHeight:float)
   )
   SetHeightOverride(execute=@n3.else, self=Get SizeBox_InventoryPanel(), InHeightOverride=@n2)
  ]}
 IsDragItemDroppable {nodes [
   n1 = IsDragItemDroppable!() -> (then, DragDropObject:Object Reference) {is=title K2Node_FunctionEntry}
   > n2 = Cast To AthenaInventoryDragDrop!(execute, Object=@n1) -> (then, CastFailed, AsAthena Inventory Drag Drop:Athena Inventory Drag Drop Object Reference) {is=title K2Node_DynamicCast}
   > n3 = Cast To FortWorldItem!(execute, Object=Get Payload!(self:Drag Drop Operation Object Reference=@n2) -> (Payload:Object Reference)) -> (then, CastFailed, AsFort World Item:Fort World Item Object Reference) {is=title K2Node_DynamicCast}
   > Return Node!(execute, bIsDroppable:bool=CanBeDropped!(self:Fort Item Object Reference=@n3) -> (ReturnValue:bool))
   Return Node(execute=@n3.CastFailed, bIsDroppable=false)
   Return Node(execute=@n2.CastFailed, bIsDroppable=false)
  ]}
 OnDragOver {nodes [
   n1 = OnDragOver!() -> (then, MyGeometry:Geometry Structure, PointerEvent:Pointer Event Structure, Operation:Drag Drop Operation Object Reference) {is=title K2Node_FunctionEntry}
   > n2 = Cast To AthenaInventoryDragDrop(Object=@n1.Operation)
   > ShowDropIcon!(
    execute
    self:Athena Inventory Drag Drop Object Reference=@n2
    Drop Action:AthenaDragDropAction Enum=Select(
     Option 0:AthenaDragDropAction Enum=NewEnumerator3
     Option 1:AthenaDragDropAction Enum=NewEnumerator1
     Index=IsDragItemDroppable!(DragDropObject:Object Reference=@n2) -> (bIsDroppable:bool) {is=title K2Node_CallFunction}
    ) -> (ReturnValue:AthenaDragDropAction Enum)
   ) -> (then)
   > Return Node!(execute, ReturnValue:bool=true)
   Return Node(execute=@n2.CastFailed, ReturnValue=false)
  ]}
 OnDrop {nodes [
   n1 = OnDrop!() -> (then, MyGeometry:Geometry Structure, PointerEvent:Pointer Event Structure, Operation:Drag Drop Operation Object Reference) {is=title K2Node_FunctionEntry}
   > n2 = Branch(Condition=IsDragItemDroppable(DragDropObject=@n1.Operation))
   > n3 = Cast To AthenaInventoryDragDrop(Object=@n1.Operation)
   > Set Local_InventoryDragDrop(Local_InventoryDragDrop:Athena Inventory Drag Drop Object Reference=@n3) -> (Output_Get:Athena Inventory Drag Drop Object Reference)
   > n4 = Cast To FortWorldItem(Object=Get Payload(self=@n3))
   > Set Local_ItemBeingDropped(Local_ItemBeingDropped:Fort World Item Object Reference=@n4) -> (Output_Get:Fort World Item Object Reference)
   n5 = Get Switcher_Tabs() -> (Switcher_Tabs:Common Widget Switcher Legacy Object Reference)
   > Branch(Condition="Equal (Object)"(A=GetActiveWidget!(self:Widget Switcher Object Reference=@n5) -> (ReturnValue:Widget Object Reference), B=Get InventoryPanel()) {is=title K2Node_CallFunction})
   n6 = GetNumInStack!(self:Fort Item Object Reference=Get Local_ItemBeingDropped() -> (Local_ItemBeingDropped:Fort World Item Object Reference)) -> (ReturnValue:int=0)
   n7 = int / int!(A:int=@n6, B:int=2) -> (ReturnValue:int=0) {is=title K2Node_CallFunction}
   n8 = Is Shift Down!(Input:"Input Event Structure (by ref)"=GetInputEventFromPointerEvent!(Event:"Pointer Event Structure (by ref)"=Get PointerEvent() -> (PointerEvent:Pointer Event Structure)) -> (ReturnValue:Input Event Structure)) -> (ReturnValue:bool)
   n9 = Get SplitTheStack!(self:Athena Inventory Drag Drop Object Reference=Get Local_InventoryDragDrop() -> (Local_InventoryDragDrop:Athena Inventory Drag Drop Object Reference)) -> (SplitTheStack:bool)
   > DropItem!(
    execute
    self:Fort Inventory Context Object Reference=Get FortInventoryContext!() -> (ReturnValue:Fort Inventory Context Object Reference) {is=title K2Node_GetBlueprintContextEx}
    ItemBeingDropped:Fort World Item Object Reference=Get Local_ItemBeingDropped()
    Quantity:int=Select(Option 0:int=@n6, Option 1:int="Max (Integer)"(A=@n7, B=1), Index=OR Boolean(A=@n9, B=@n8)) -> (ReturnValue:int)
   ) -> (then, ReturnValue:bool)
   > Return Node(ReturnValue=true)
   Return Node(execute=[@n4.CastFailed, @n3.CastFailed, @n2.else], ReturnValue=false)
  ]}
 RemoveContentFromWidgetHost {nodes [
   n1 = Remove Content From Host Widget!() -> (
    then
    WidgetHostTag:"Gameplay Tag Structure (by ref)"
    ContentWidget:Widget Object Reference
    RemovalType:EWidgetSlottingRemovalType Enum=Unparent
   ) {is=title K2Node_FunctionEntry}
   > n2 = Branch(Condition=MatchesTag(TagOne=@n1, TagTwo=Get LeftSideSlotTag(), bExactMatch=true))
   > n3 = RemoveChild(self=Get VerticalBox_InventoryScreenLeftSideExtensionHolder(), Content=@n1.ContentWidget)
   n4 = Branch(
    execute=[@n3.then, @n2.else]
    Condition=MatchesTag(TagOne=Get WidgetHostTag() -> (WidgetHostTag:"Gameplay Tag Structure (by ref)"), TagTwo=Get TopLeftSlotTag(), bExactMatch=true)
   )
   > n5 = RemoveChild(self=Get VerticalBox_InventoryScreenTopLeftExtensionHolder(), Content=@n1.ContentWidget)
   > Return Node(ReturnValue=@n5)
   Return Node(execute=@n4.else, ReturnValue=@n3)
  ]}
 SetLeftVBoxPosition {nodes [
   n1 = SetLeftVBoxPosition!() -> (
    then
    NewPosition_Padding_2_#3:Margin Structure
    NewPosition_HorizontalAlignment_5_#4:EHorizontalAlignment Enum
    NewPosition_VerticalAlignment_8_#5:EVerticalAlignment Enum
   ) {is=title K2Node_FunctionEntry}
   > n2 = Sequence!(execute) -> (then_0, then_1, then_2:Exec)
   n3 = SlotAsOverlaySlot(Widget=Get VerticalBox_InventoryScreenLeftSideExtensionHolder())
   > SetPadding!(execute, self:Overlay Slot Object Reference=@n3, InPadding:Margin Structure=@n1) -> (then)
   SetHorizontalAlignment(execute=@n2.then_1, self=@n3, InHorizontalAlignment=@n1.NewPosition_HorizontalAlignment_5_#4)
   SetVerticalAlignment(execute=@n2.then_2, self=@n3, InVerticalAlignment=@n1.NewPosition_VerticalAlignment_8_#5)
  ]}
 SetTouchStyle {nodes [
   SetTouchStyle() {is=title K2Node_FunctionEntry}
   n1 = AND Boolean(
    A=IsUsingCompactUI()
    B=IsUsingInputType!(InputType:ECommonInputType Enum=Touch, PlayerController:Player Controller Object Reference=GetOwningPlayer()) -> (ReturnValue:bool)
   )
   > SetRenderTranslation(
    self=Get Border_TablistOffset() -> (Border_TablistOffset:Border Object Reference)
    Translation=MakeVector2D(X=0, Y=Select(Option 0:double=-22, Option 1:double=-44, Index=@n1) -> (ReturnValue:double))
   )
  ]}
 ShowBackgroundBlurHeader {nodes [
   ShowBackgroundBlurHeader() {is=title K2Node_FunctionEntry}
   > Return Node!(execute, bShowBackgroundBlurHeader:bool=false)
  ]}
 ShowHeaderBackground {nodes [
   ShowHeaderBackground() {is=title K2Node_FunctionEntry}
   > Return Node!(execute, bShowHeaderBackground:bool=false)
  ]}
]
variables [
 InitSizeBoxHeight {type=number, default=850}
 IntSpacerSize {type=Vector2D, default=<1 136>}
 LeftSideSlotTag {type=GameplayTag, default {tagName=DynamicUI.HUD.Fort.Athena.Inventory.LeftSide}}
 LeftVBoxPosition_BottomSquad {
  type=AthenaInventoryPanel_PaddingAlignment
  default {horizontalAlignment=HAlign_Left, verticalAlignment=VAlign_Center, padding=<ltrb 50 0 0 0>}
 }
 LeftVBoxPosition_Default {
  type=AthenaInventoryPanel_PaddingAlignment
  default {horizontalAlignment=HAlign_Left, verticalAlignment=VAlign_Bottom, padding=<ltrb 50 0 0 100>}
 }
 TopLeftSlotTag {type=GameplayTag, default {tagName=DynamicUI.HUD.Fort.Athena.Inventory.TopLeft}}
]
widgets {
 info {inheritedWidgetCount=3, parentClass=/Script/FortniteUI.AthenaInventoryPanelContainer, rootWidgetClass=&Overlay, widgetCount=16}
 widgets [
  Root_Overlay {widgetClassPath=&Overlay, widget=Root_Overlay, properties {visibility=Visible}}
  DismissInventory {
   widgetClassPath=&Button
   parent=Root_Overlay
   widget=DismissInventory
   properties {widgetStyle {
     clickedSlateSound {resourceObject=null}
     disabled {
      bIsDynamicallyLoaded=false
      drawAs=NoDrawType
      imageType=NoImage
      mirroring=NoMirror
      resourceName=None
      resourceObject=null
      tiling=NoTile
      imageSize=<0 0>
      margin=<ltrb 0 0 0 0>
      outlineSettings {
       bUseBrushTransparency=false
       roundingType=HalfHeightRadius
       width=0
       color {colorUseRule=UseColor_Specified, specifiedColor=<rgba 0 0 0 0>}
       cornerRadii=<0 0 0 0>
      }
      tintColor {colorUseRule=UseColor_Specified, specifiedColor=<rgba 1 1 1 1>}
      uVRegion {bIsValid=false, max=<0 0>, min=<0 0>}
     }
     disabledForeground {colorUseRule=UseColor_Specified, specifiedColor=<rgba 0.5271 0.5271 0.5271 1>}
     hovered {
      bIsDynamicallyLoaded=false
      drawAs=NoDrawType
      imageType=NoImage
      mirroring=NoMirror
      resourceName=None
      resourceObject=null
      tiling=NoTile
      imageSize=<32 32>
      margin=<ltrb 0 0 0 0>
      outlineSettings {
       bUseBrushTransparency=true
       roundingType=FixedRadius
       width=1
       color {colorUseRule=UseColor_Specified, specifiedColor=<rgba 0.7243 0.7243 0.7243 1>}
       cornerRadii=<4 4 4 4>
      }
      tintColor {colorUseRule=UseColor_Specified, specifiedColor=<rgba 0.7243 0.7243 0.7243 1>}
      uVRegion {bIsValid=false, max=<0 0>, min=<0 0>}
     }
     hoveredForeground {colorUseRule=UseColor_Specified, specifiedColor=<rgba 0 1 1 1>}
     hoveredSlateSound {resourceObject=null}
     normal {
      bIsDynamicallyLoaded=false
      drawAs=NoDrawType
      imageType=NoImage
      mirroring=NoMirror
      resourceName=None
      resourceObject=null
      tiling=NoTile
      imageSize=<32 32>
      margin=<ltrb 0 0 0 0>
      outlineSettings {
       bUseBrushTransparency=true
       roundingType=FixedRadius
       width=1
       color {colorUseRule=UseColor_Specified, specifiedColor=<rgba 0.6951 0.6951 0.6951 1>}
       cornerRadii=<4 4 4 4>
      }
      tintColor {colorUseRule=UseColor_Specified, specifiedColor=<rgba 0.4955 0.4955 0.4955 1>}
      uVRegion {bIsValid=false, max=<0 0>, min=<0 0>}
     }
     normalForeground {colorUseRule=UseColor_Specified, specifiedColor=<rgba 0 1 1 1>}
     normalPadding=<ltrb 12 1.5 12 1.5>
     pressed {
      bIsDynamicallyLoaded=false
      drawAs=NoDrawType
      imageType=NoImage
      mirroring=NoMirror
      resourceName=None
      resourceObject=null
      tiling=NoTile
      imageSize=<32 32>
      margin=<ltrb 0 0 0 0>
      outlineSettings {
       bUseBrushTransparency=true
       roundingType=FixedRadius
       width=1
       color {colorUseRule=UseColor_Specified, specifiedColor=<rgba 0.7243 0.7243 0.7243 1>}
       cornerRadii=<4 4 4 4>
      }
      tintColor {colorUseRule=UseColor_Specified, specifiedColor=<rgba 0.3843 0.3843 0.3843 1>}
      uVRegion {bIsValid=false, max=<0 0>, min=<0 0>}
     }
     pressedForeground {colorUseRule=UseColor_Specified, specifiedColor=<rgba 0 1 1 1>}
     pressedPadding=<ltrb 12 1.5 12 1.5>
     pressedSlateSound {resourceObject=null}
    }}
   slot {class=OverlaySlot, properties {horizontalAlignment=HAlign_Fill, verticalAlignment=VAlign_Fill}}
  }
  SizeBox_InventoryContainer {
   widgetClassPath=&SizeBox
   parent=Root_Overlay
   widget=SizeBox_InventoryContainer
   properties {bOverride_WidthOverride=true, widthOverride=688}
   slot {class=OverlaySlot, properties {horizontalAlignment=HAlign_Right, verticalAlignment=VAlign_Fill}}
  }
  SafeZone {
   widgetClassPath=&SafeZone
   parent=SizeBox_InventoryContainer
   widget=SafeZone
   properties {padLeft=false, padRight=false}
   slot {class=SizeBoxSlot}
  }
  Overlay_2 {widgetClassPath=&Overlay, parent=SafeZone, widget=Overlay_2, slot {class=SafeZoneSlot}}
  VerticalBox_25 {widgetClassPath=&VerticalBox, parent=Overlay_2, widget=VerticalBox_25, slot {class=OverlaySlot}}
  Border_TablistOffset {
   widgetClassPath=&Border
   bIsVariable=true
   parent=VerticalBox_25
   widget=Border_TablistOffset
   properties {
    visibility=SelfHitTestInvisible
    background {
     bIsDynamicallyLoaded=false
     drawAs=NoDrawType
     imageType=NoImage
     mirroring=NoMirror
     resourceName=None
     resourceObject=null
     tiling=NoTile
     imageSize=<32 32>
     margin=<ltrb 0 0 0 0>
     outlineSettings {
      bUseBrushTransparency=false
      roundingType=HalfHeightRadius
      width=0
      color {colorUseRule=UseColor_Specified, specifiedColor=<rgba 0 0 0 0>}
      cornerRadii=<0 0 0 0>
     }
     tintColor {colorUseRule=UseColor_Specified, specifiedColor=<rgba 1 1 1 1>}
     uVRegion {bIsValid=false, max=<0 0>, min=<0 0>}
    }
    padding=<ltrb 0 0 0 0>
   }
   slot {class=VerticalBoxSlot, properties {horizontalAlignment=HAlign_Center, verticalAlignment=VAlign_Top}}
  }
  TabList {
   widgetClassPath=/Game/UI/Foundation/TabbedView/HorizontalTabList_Legacy
   bInherited=true
   parent=Border_TablistOffset
   widget=TabList
   properties {
    bAutoListenForInput=true
    onTabButtonCreation=(&self:WidgetTree.TabList.HandleOnTabButtonCreated)
    onTabButtonRemoval=(&self:WidgetTree.TabList.HandleOnTabButtonRemoved)
    nextTabInputActionData {dataTable=/FortUILibrary/Input/AthenaInventoryActionDataTable, rowName=InventoryTabRight}
    previousTabInputActionData {dataTable=&AthenaInventoryActionDataTable, rowName=InventoryTabLeft}
   }
   slot {class=BorderSlot, properties {padding=<ltrb 0 0 0 0>}}
  }
  Overlay_1 {
   widgetClassPath=&Overlay
   parent=VerticalBox_25
   widget=Overlay_1
   slot {class=VerticalBoxSlot, properties {size {sizeRule=Fill, value=1}}}
  }
  Image_Backing {
   widgetClassPath=&Image
   bIsVariable=true
   parent=Overlay_1
   widget=Image_Backing
   properties {brush {
     bIsDynamicallyLoaded=false
     drawAs=Image
     imageType=NoImage
     mirroring=NoMirror
     resourceName=None
     resourceObject=/Game/Athena/HUD/Inventory/Art/M_UI_InventoryPanel_Backing
     tiling=NoTile
     imageSize=<8 8>
     margin=<ltrb 0.5 0.5 0.5 0.5>
     outlineSettings {
      bUseBrushTransparency=false
      roundingType=HalfHeightRadius
      width=0
      color {colorUseRule=UseColor_Specified, specifiedColor=<rgba 0 0 0 0>}
      cornerRadii=<0 0 0 0>
     }
     tintColor {colorUseRule=UseColor_Specified, specifiedColor=<rgba 1 1 1 1>}
     uVRegion {bIsValid=false, max=<0 0>, min=<0 0>}
    }}
   slot {class=OverlaySlot, properties {horizontalAlignment=HAlign_Fill, verticalAlignment=VAlign_Fill}}
  }
  SizeBox_InventoryPanel {
   widgetClassPath=&SizeBox
   bIsVariable=true
   parent=Overlay_1
   widget=SizeBox_InventoryPanel
   properties {
    bOverride_HeightOverride=true
    bOverride_WidthOverride=true
    heightOverride=850
    maxDesiredHeight=850
    minDesiredHeight=860
    widthOverride=690
   }
   slot {class=OverlaySlot, properties {horizontalAlignment=HAlign_Fill, verticalAlignment=VAlign_Fill}}
  }
  Switcher_Tabs {
   widgetClassPath=&CommonWidgetSwitcherLegacy
   bInherited=true
   parent=SizeBox_InventoryPanel
   widget=Switcher_Tabs
   properties {transitionDuration=0}
   slot {class=SizeBoxSlot}
  }
  InventoryPanel {
   widgetClassPath=&/WBP_AthenaInventoryPanel
   bInherited=true
   parent=Switcher_Tabs
   widget=InventoryPanel
   slot {class=WidgetSwitcherSlot}
  }
  Spacer_52 {widgetClassPath=&Spacer, parent=VerticalBox_25, widget=Spacer_52, properties {size=<1 75>}, slot {class=VerticalBoxSlot}}
  VerticalBox_InventoryScreenLeftSideExtensionHolder {
   widgetClassPath=&VerticalBox
   bIsVariable=true
   parent=Root_Overlay
   widget=VerticalBox_InventoryScreenLeftSideExtensionHolder
   slot {class=OverlaySlot, properties {verticalAlignment=VAlign_Bottom, padding=<ltrb 50 0 0 100>}}
  }
  VerticalBox_InventoryScreenTopLeftExtensionHolder {
   widgetClassPath=&VerticalBox
   bIsVariable=true
   parent=Root_Overlay
   widget=VerticalBox_InventoryScreenTopLeftExtensionHolder
   slot {class=OverlaySlot, properties {padding=<ltrb 50 0 0 0>}}
  }
 ]
}
