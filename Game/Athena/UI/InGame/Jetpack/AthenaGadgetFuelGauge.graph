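# AthenaGadgetFuelGauge is the in-game widget that shows a gadget's fuel gauge and charge count, filed under the Jetpack UI.
# An Overlay layers a right-aligned, full-height FortFuelGaugeIndicator beneath a numeric CurrentChargesText anchored to the bottom right.
# Both children are inherited from the native AthenaGadgetFuelWidget parent, and the event graph holds only unconnected PreConstruct, Construct and Tick events, so the Blueprint supplies layout and styling only.
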WidgetBlueprint /Game/Athena/UI/InGame/Jetpack/AthenaGadgetFuelGauge
graphs [EventGraph {nodes [
   Event PreConstruct()
   Event Construct()
   Event Tick()
  ]}]
variables []
widgets {
 info {inheritedWidgetCount=2, parentClass=/Script/FortniteUI.AthenaGadgetFuelWidget, rootWidgetClass=&Overlay, widgetCount=3}
 widgets [
  Overlay_0 {widgetClassPath=&Overlay, widget=Overlay_0}
  FuelGaugeIndicator_Main {
   widgetClassPath=/Game/Athena/HUD/FortFuelGaugeIndicator
   bInherited=true
   bIsVariable=true
   parent=Overlay_0
   widget=FuelGaugeIndicator_Main
   slot {class=OverlaySlot, properties {horizontalAlignment=HAlign_Right, verticalAlignment=VAlign_Fill}}
  }
  CurrentChargesText {
   widgetClassPath=/Script/CommonUI.CommonNumericTextBlock
   bInherited=true
   bIsVariable=true
   parent=Overlay_0
   widget=CurrentChargesText
   properties {
    style=/FortUILibrary/Styles/Typography/Foundation/TextStyle-Base-S-B.TextStyle-Base-S-B_C
    text=Text Block
    font {
     bForceMonospaced=false
     bMaterialIsStencil=false
     fontMaterial=null
     fontObject=&NotoSans
     letterSpacing=0
     monospacedWidth=1
     size=18
     skewAmount=0
     typefaceFontName=NotoSans-Bold
     outlineSettings {
      bApplyOutlineToDropShadows=false
      bMiteredCorners=false
      bSeparateFillAlpha=false
      outlineMaterial=null
      outlineSize=0
      outlineColor=<rgba 0 0 0 1>
     }
    }
    shadowOffset=<0 0>
   }
   slot {class=OverlaySlot, properties {horizontalAlignment=HAlign_Right, verticalAlignment=VAlign_Bottom, padding=<ltrb 0 0 9 10>}}
  }
 ]
}
