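# Sound cue named for the start of an equip-slot swap drag in the inventory UI.
# Its node graph is empty with no root, so the file references no audio to play.
# It still carries playback settings: non-critical in-game UI sound class, reduced volume, lowest priority and virtualization disabled.
# A concurrency override allows one instance at a time, stopping the oldest.
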SoundCue /Game/Sounds/Fort_UI_Sounds/Gameplay/Inventory/UI_Inventory_EquipSlot_Swap_Drag_Start_Cue
root=null
nodes []
properties {
 bOverrideConcurrency=true
 maxDistance=2097152
 priority=0
 soundClassObject=&Fort_UI_InGame_NonCritical
 virtualizationMode=Disabled
 volumeMultiplier=0.5
 concurrencyOverrides {maxCount=1, resolutionRule=StopOldest}
}
