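# Sound cue for inventory UI feedback, named for the end of an equip-slot swap drag.
# A single wave player plays a front-end top-bar hover wave at reduced cue volume, routed to the non-critical in-game UI sound class.
# A concurrency override keeps one instance at a time by stopping the oldest, with virtualization disabled and priority set to zero.
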SoundCue /Game/Sounds/Fort_UI_Sounds/Gameplay/Inventory/UI_Inventory_EquipSlot_Swap_Drag_Stop_Cue
root=WavePlayer_0
nodes [{id=WavePlayer_0, properties {soundWaveAssetPtr=/Game/Sounds/Fort_UI_Sounds/FrontEnd/Tabs/wav/UI_MP_Hover_TopBar_01}}]
properties {
 bOverrideConcurrency=true
 maxDistance=2097152
 priority=0
 soundClassObject=&Fort_UI_InGame_NonCritical
 virtualizationMode=Disabled
 volumeMultiplier=0.5
 concurrencyOverrides {maxCount=1, resolutionRule=StopOldest}
}
