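# UI sound cue that, by its name, plays when the inventory opens.
# A single wave player feeds a mixer whose second input is empty, with output heavily attenuated by the cue's volume multiplier.
# It uses the non-critical in-game UI sound class at zero priority with virtualization disabled, and a concurrency override keeps one instance by stopping the oldest.
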SoundCue /Game/Sounds/Fort_UI_Sounds/Gameplay/Inventory/UI_Inventory_Open_Cue
root=Mixer_0
nodes [
 {id=Mixer_0, children [WavePlayer_3,null], properties {inputVolume [1,0.5]}}
 {id=WavePlayer_3, properties {soundWaveAssetPtr=&/UI_Inventory_Open_01}}
]
properties {
 bOverrideConcurrency=true
 maxDistance=2097152
 priority=0
 soundClassObject=&Fort_UI_InGame_NonCritical
 virtualizationMode=Disabled
 volumeMultiplier=0.1
 concurrencyOverrides {maxCount=1, resolutionRule=StopOldest}
}
