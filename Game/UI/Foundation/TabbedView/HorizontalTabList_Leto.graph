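# Horizontal CommonUI tab list on FortTabListWidgetBase, with previous/next buttons and input-action glyphs that hide when one tab or fewer exists.
# Created tabs are cached, styled, then placed in either a plain box or a masked RetainerBox ScrollBox, chosen by IsScrollable.
# Scrollable mode sizes its viewport from tab size and visible count, adds optional edge spacers, and centres clicked or selected tabs.
# PreConstruct spawns debug tabs for designer preview.
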WidgetBlueprint /Game/UI/Foundation/TabbedView/HorizontalTabList_Leto
graphs [
 AddSpacerToScrollableBox {nodes [
   n1 = Inputs!() -> (execute, Size:Vector 2D Structure)
   > n2 = Construct Object from Class(Class=&Spacer) -> (ReturnValue:Spacer Object Reference)
   > SetSize!(execute, self:Spacer Object Reference=@n2, InSize:Vector 2D Structure=@n1) -> (then)
   > AddChildToHorizontalBox!(
    execute
    self:Horizontal Box Object Reference=Get TabButtonBox_Scrollable() -> (TabButtonBox_Scrollable:Horizontal Box Object Reference)
    Content:Widget Object Reference=@n2
   ) -> (then, ReturnValue:Horizontal Box Slot Object Reference)
   > Outputs!(then:Exec, "Size ":Vector 2D Structure=@n1)
  ]}
 DebugCreateTabsForDesigner {nodes [
   DebugCreateTabsForDesigner() {is=title K2Node_FunctionEntry}
   > ClearChildren(self=[Get TabButtonBox_Scrollable(), Get TabButtonBox() -> (TabButtonBox:Horizontal Box Object Reference)])
   > Clear(TargetArray:Array of Common Button Base Object References=Get TabButtonCache() -> (TabButtonCache:Array of Common Button Base Object References))
   > For Loop(FirstIndex=1, LastIndex=Get DebugTabCount() -> (DebugTabCount:int=0))
   > n1 = Create Widget!(
    execute:Exec
    Class:User Widget Class Reference=Get DebugButtonTypeDisplay() -> (DebugButtonTypeDisplay:Common Button Base Class Reference)
    OwningPlayer:Player Controller Object Reference=GetOwningPlayer()
    Style:Common Button Style Class Reference
    bApplyAlphaOnDisable:bool=True
    bLocked:bool=False
    bSelectable:bool=False
    bShouldSelectUponReceivingFocus:bool=False
    bInteractableWhenSelected:bool=False
    bToggleable:bool=False
    bTriggerClickedAfterSelection:bool=False
    bDisplayInputActionWhenNotInteractable:bool=True
    bHideInputActionWithKeyboard:bool=False
    bShouldUseFallbackDefaultInputAction:bool=True
    bRequiresHold:bool=False
    HoldData:Common UIHold Data Class Reference
    ClickMethod:EButtonClickMethod Enum=DownAndUp
    InputPriority:int=0
    TriggeringInputAction:Data Table Row Handle Structure
   ) -> (then:Exec, ReturnValue:Common Button Base Object Reference)
   > n2 = Cast To CommonButtonBase!(execute, Object=@n1) -> (then, CastFailed, AsCommon Button Base:Common Button Base Object Reference) {is=title K2Node_DynamicCast}
   > UpdateTabStyles!(execute, self, ButtonWidget:Common Button Base Object Reference=@n2) -> (then)
  ]}
 EventGraph {nodes [
   n1 = CustomEvent_1!() -> (OutputDelegate, then, InVisibility:ESlateVisibility Enum=Visible) {is=title K2Node_CustomEvent}
   > Refresh Next/Previous Visibility()
   Refresh Next/Previous Visibility() {is=title K2Node_CustomEvent}
   > n2 = Sequence!(execute) -> (then_0, then_1, then_2:Exec)
   n3 = GetTabCount!(self:Common Tab List Widget Base Object Reference) -> (ReturnValue:int=0)
   > SetVisibility(
    self=Get PreviousTabActionWrapper() -> (PreviousTabActionWrapper:Size Box Object Reference)
    InVisibility=Select(Option 0=Visible, Option 1=Hidden, Index="integer <= integer"!(A:int=@n3, B:int=1) -> (ReturnValue:bool) {is=title K2Node_CallFunction})
   )
   n4 = GetTabCount() -> (ReturnValue=0)
   SetVisibility(
    execute=@n2.then_1
    self=Get NextTabActionWrapper() -> (NextTabActionWrapper:Size Box Object Reference)
    InVisibility=Select(Option 0=Visible, Option 1=Hidden, Index="integer <= integer"(A=@n4, B=1))
   )
   SetupScrollableList() {is=title K2Node_CustomEvent}
   > SetVisibility(
    self=Get TabButtonBox()
    InVisibility=Select(Option 0=SelfHitTestInvisible, Option 1=Collapsed, Index=Get IsScrollable() -> (IsScrollable:bool))
   )
   > SetVisibility(
    self=Get RetainerBox_Scrollable() -> (RetainerBox_Scrollable:Retainer Box Object Reference)
    InVisibility=Select(Option 0=Collapsed, Option 1=SelfHitTestInvisible, Index=Get IsScrollable())
   )
   n5 = Get!() -> (TabButtonPadding_Left:float=~, TabButtonPadding_Top:float=~, TabButtonPadding_Right:float=~, TabButtonPadding_Bottom:float=~)
   n6 = float + float!(A=Get!() -> (TabSize_X:double, TabSize_Y:double), B=@n5, C:float=@n5.TabButtonPadding_Right) -> (ReturnValue)
   n7 = Get ScrollableVisible() -> (ScrollableVisible:int=0)
   > SetWidthOverride(
    self=Get SizeBox_Scrollable() -> (SizeBox_Scrollable:Size Box Object Reference)
    InWidthOverride=float + float!(A=float * float(A=16, B=3) {is=title K2Node_PromotableOperator}, B=float * float(A=@n6, B:int=@n7)) -> (ReturnValue)
   )
   n8 = CustomEvent_0!() -> (OutputDelegate, then, TabId:Name) {is=title K2Node_CustomEvent}
   > Refresh Next/Previous Visibility() {is=title K2Node_CallFunction}
   ScrollableTabsCreated() {is=title K2Node_CustomEvent}
   > n9 = Branch(Condition=Get bUseScrollingSpacers() -> (bUseScrollingSpacers:bool))
   > n10 = Add Spacer to Scrollable Box!(execute:Exec, Size:Vector 2D Structure=(X=16,Y=1)) -> (then:Exec, "Size ":Vector 2D Structure) {is=title K2Node_MacroInstance}
   n11 = For Each Loop(Exec=[@n10.then, @n9.else], Array:Array of Common Button Base Object References=Get TabButtonCache()) -> (Array Element:Common Button Base Object Reference)
   > UpdateTabStyles(ButtonWidget=@n11)
   > Bind Event to On Selected Changed(self:Common Button Base Object Reference=@n11, Delegate=Create Event()) {is=title K2Node_AddDelegate}
   > Bind Event to On Clicked(self:Common Button Base Object Reference=@n11, Delegate=Create Event()) {is=title K2Node_AddDelegate}
   n12 = Branch(execute=@n11.Completed, Condition=Get bUseScrollingSpacers())
   > n13 = Add Spacer to Scrollable Box!(execute, Size=(X=16,Y=1)) -> (then, Size _X:double, Size _Y:double)
   n14 = Get!() -> (TabButtonPadding_Left:float=~, TabButtonPadding_Top:float=~, TabButtonPadding_Right:float=~, TabButtonPadding_Bottom:float=~)
   n15 = float + float!(A=Get!() -> (TabSize_X:double, TabSize_Y:double), B=@n14, C:float=@n14.TabButtonPadding_Right) -> (ReturnValue)
   n16 = Get ScrollableVisible() -> (ScrollableVisible=0)
   > n17 = SetWidthOverride(
    self=Get SizeBox_Scrollable()
    InWidthOverride=float + float!(A=float * float(A=@n13, B:double=3), B=float * float(A=@n15, B:int=@n16)) -> (ReturnValue)
   )
   DelayUntilNextTick!(execute=[@n17.then, @n12.else]) -> (then)
   > SetRetainRendering!(execute, self:Retainer Box Object Reference=Get RetainerBox_Scrollable(), bInRetainRendering:bool=true) -> (then)
   n18 = OnTabClicked!() -> (OutputDelegate, then, Button:Common Button Base Object Reference) {is=title K2Node_CustomEvent}
   > ScrollWidgetIntoView!(
    execute
    self:Scroll Box Object Reference=Get ScrollBox() -> (ScrollBox:Scroll Box Object Reference)
    WidgetToFind:Widget Object Reference=@n18.Button
    AnimateScroll:bool=true
    ScrollDestination:EDescendantScrollDestination Enum=Center
    Padding:float=0.01
   ) -> (then)
   > Set LastSelectedTab(LastSelectedTab:Common Button Base Object Reference=@n18.Button) -> (Output_Get:Common Button Base Object Reference)
   n19 = OnButtonSelected!() -> (OutputDelegate, then, Button:Common Button Base Object Reference, Selected:bool=~) {is=title K2Node_CustomEvent}
   > Branch(Condition=@n19.Selected)
   > ScrollWidgetIntoView(self=Get ScrollBox(), WidgetToFind=@n19.Button, AnimateScroll=true, ScrollDestination=Center, Padding=0.01)
   > Set LastSelectedTab(LastSelectedTab=@n19.Button)
   n20 = Event HandleTabCreation!() -> (OutputDelegate, then, TabNameID:Name=None, TabButton:Common Button Base Object Reference) {is=title K2Node_Event}
   > Add(
    TargetArray:Array of Common Button Base Object References=Get TabButtonCache()
    NewItem:"Common Button Base Object Reference (by ref)"=@n20.TabButton
   ) {is=title K2Node_CallArrayFunction}
   > Parent: HandleTabCreation!(execute, TabNameID:Name=@n20.TabNameID, TabButton:Common Button Base Object Reference=@n20.TabButton) -> (then) {is=title K2Node_CallParentFunction}
   > n21 = Branch(Condition=Get IsScrollable())
   > RetriggerableDelay(Duration=0)
   > ScrollableTabsCreated() {is=title K2Node_CallFunction}
   UpdateTabStyles(execute=@n21.else, ButtonWidget=@n20.TabButton)
   n22 = Event HandleTabRemoval!() -> (OutputDelegate, then, TabNameID:Name=None, TabButton:Common Button Base Object Reference) {is=title K2Node_Event}
   > Remove Item!(
    execute
    TargetArray:Array of Common Button Base Object References=Get TabButtonCache()
    Item:"Common Button Base Object Reference (by ref)"=@n22.TabButton
   ) -> (then, ReturnValue:bool)
   > Parent: HandleTabRemoval!(execute, TabNameID:Name=@n22.TabNameID, TabButton:Common Button Base Object Reference=@n22.TabButton) -> (then) {is=title K2Node_CallParentFunction}
   > Refresh Next/Previous Visibility()
   n23 = Event HandlePreviousNextButtonCreation!() -> (OutputDelegate, then, PreviousNextButton:User Widget Object Reference, ButtonType:EPreviousNextButtonType Enum=Previous) {is=title K2Node_Event}
   > Parent: HandlePreviousNextButtonCreation!(execute, PreviousNextButton:User Widget Object Reference=@n23.PreviousNextButton, ButtonType:EPreviousNextButtonType Enum=@n23.ButtonType) -> (then) {is=title K2Node_CallParentFunction}
   n24 = Select!(
    Previous:Size Box Object Reference=Get PreviousButtonWrapper() -> (PreviousButtonWrapper:Size Box Object Reference)
    Next:Size Box Object Reference=Get NextButtonWrapper() -> (NextButtonWrapper:Size Box Object Reference)
    Index:EPreviousNextButtonType Enum=@n23.ButtonType
   ) -> (ReturnValue:Size Box Object Reference)
   > SetVisibility(self=@n24, InVisibility=SelfHitTestInvisible)
   > AddChild(self=@n24, Content=@n23.PreviousNextButton)
   > SetRenderScale!(
    execute
    self:Widget Object Reference=@n23.PreviousNextButton
    Scale:Vector 2D Structure=Select(Previous:Vector 2D Structure=(X=-1,Y=1), Next:Vector 2D Structure=(X=1,Y=1), Index=@n23.ButtonType) -> (ReturnValue:Vector 2D Structure)
   ) -> (then)
   n25 = Event PreConstruct()
   > Parent: PreConstruct(IsDesignTime=@n25.IsDesignTime)
   > n26 = Sequence()
   > SetInputAction!(
    execute
    self:Common Action Widget Object Reference=Get PreviousTabAction() -> (PreviousTabAction:Common Action Widget Object Reference)
    InputActionRow:Data Table Row Handle Structure=Get PreviousTabInputActionData() -> (PreviousTabInputActionData:Data Table Row Handle Structure)
   ) -> (then)
   > SetInputAction(
    self=Get NextTabAction() -> (NextTabAction:Common Action Widget Object Reference)
    InputActionRow=Get NextTabInputActionData() -> (NextTabInputActionData:Data Table Row Handle Structure)
   )
   > Assign On Tab Selected!(execute:Exec, self:Horizontal Tab List Leto Object Reference, Delegate:"Delegate (by ref)"=@n8) -> (then:Exec) {is=title K2Node_AssignDelegate}
   > Bind Event to On Visibility Changed(self:User Widget Object Reference, Delegate=@n1) {is=title K2Node_AddDelegate}
   n27 = Get DebugButtonTypeDisplay()
   Branch(execute=@n26.then_1, Condition=AND Boolean(A=@n25.IsDesignTime, B=IsValidClass!(Class:Object Class Reference=@n27) -> (ReturnValue:bool)))
   > DebugCreateTabsForDesigner() {is=title K2Node_CallFunction}
   SetupScrollableList(execute=@n26.then_2) {is=title K2Node_CallFunction}
   Event Construct()
   > Refresh Next/Previous Visibility()
   > SetVisibility(
    self=Get ImageHorizRule() -> (ImageHorizRule:Image Object Reference)
    InVisibility=Select!(Option 0:ESlateVisibility Enum=Collapsed, Option 1:ESlateVisibility Enum=SelfHitTestInvisible, Index:bool=Get ShowRail() -> (ShowRail:bool)) -> (ReturnValue:ESlateVisibility Enum)
   )
  ]}
 RefreshGamepad {nodes [RefreshGamepad() {is=title K2Node_FunctionEntry}]}
 SetVerticalAlignment {nodes [
   n1 = SetVerticalAlignment!() -> (then, FillVertically:EVerticalAlignment Enum=VAlign_Fill) {is=title K2Node_FunctionEntry}
   > SetVerticalAlignment(
    self:Horizontal Box Slot Object Reference=SlotAsHorizontalBoxSlot!(Widget:Widget Object Reference=Get Overlay_Container() -> (Overlay_Container:Overlay Object Reference)) -> (ReturnValue:Horizontal Box Slot Object Reference)
    InVerticalAlignment=@n1
   ) {is=title K2Node_CallFunction}
  ]}
 UpdateTabStyles {nodes [
   n1 = UpdateTabStyles!() -> (then, ButtonWidget:Common Button Base Object Reference) {is=title K2Node_FunctionEntry}
   > n2 = Branch(Condition=Get ForceTabLabelsCollapsed() -> (ForceTabLabelsCollapsed:bool))
   > n3 = Cast To IconTabButton_Leto!(execute, Object=@n1) -> (then, CastFailed, AsIcon Tab Button Leto:Icon Tab Button Leto Object Reference) {is=title K2Node_DynamicCast}
   > n4 = Set ButtonText!(execute, ButtonText:Text=GetEmptyText!() -> (ReturnValue:Text), self:Icon Tab Button Leto Object Reference=@n3) -> (then, Output_Get:Text)
   n5 = Branch(execute=[@n4.then, @n3.CastFailed, @n2.else], Condition=Get ForceTabIconsCollapsed() -> (ForceTabIconsCollapsed:bool))
   > n6 = Cast To IconTabButton_Leto(Object=@n1)
   > n7 = SetVisibility(self=Get LeftSideImage!(self:Icon Tab Button Leto Object Reference=@n6) -> (LeftSideImage:Image Object Reference), InVisibility=Collapsed)
   n8 = AddChildToHorizontalBox(
    execute=[@n7.then, @n6.CastFailed, @n5.else]
    self=Select(
     Option 0:Horizontal Box Object Reference=Get TabButtonBox()
     Option 1:Horizontal Box Object Reference=Get TabButtonBox_Scrollable()
     Index=Get IsScrollable()
    ) -> (ReturnValue:Horizontal Box Object Reference)
    Content=@n1
   )
   > SetPadding!(
    execute
    self:Horizontal Box Slot Object Reference=@n8
    InPadding:Margin Structure=Get TabButtonPadding() -> (TabButtonPadding:Margin Structure)
   ) -> (then)
   n9 = SlotAsSizeBoxSlot!(Widget:Widget Object Reference=Get NextTabAction()) -> (ReturnValue:Size Box Slot Object Reference)
   n10 = Break Margin!(Margin:"Margin Structure (by ref)"=Get Padding!(self:Size Box Slot Object Reference=@n9) -> (Padding:Margin Structure)) -> (Left:float=~, Top:float=~, Right:float=~, Bottom:float=~) {is=title K2Node_BreakStruct}
   n11 = Get NextTabActionPaddingLeft() -> (NextTabActionPaddingLeft:double)
   n12 = Get IsScrollable()
   > SetPadding(
    self:Size Box Slot Object Reference=@n9
    InPadding=Make Margin(
     Left=Select(Option 0:double=@n11, Option 1:double=-16, Index=@n12) -> (ReturnValue:double)
     Top=@n10.Top
     Right=@n10.Right
     Bottom=@n10.Bottom
    )
   )
   n13 = SlotAsSizeBoxSlot(Widget=Get PreviousTabAction())
   n14 = Break Margin(Margin=Get Padding(self=@n13)) -> (Left=~, Top=~, Right=~, Bottom=~)
   n15 = Get PreviousTabActionPaddingRight() -> (PreviousTabActionPaddingRight:double)
   n16 = Get IsScrollable()
   > SetPadding(self=@n13, InPadding=Make Margin(Left=@n14, Top=@n14.Top, Right=Select(Option 0=@n15, Option 1=-16, Index=@n16), Bottom=@n14.Bottom))
   > SetMinDimensions!(execute, self:Common Button Base Object Reference=@n1, InMinWidth:int=Get MinTabWidth() -> (MinTabWidth:int=0), InMinHeight:int=0) -> (then)
   > Refresh Next/Previous Visibility()
   > Branch(Condition=Get UseButtonStyle() -> (UseButtonStyle:bool))
   > SetStyle!(
    execute
    self:Common Button Base Object Reference=@n1
    InStyle:Common Button Style Class Reference=Get ButtonStyle() -> (ButtonStyle:Common Button Style Class Reference)
   ) -> (then)
  ]}
]
variables [
 bUseScrollingSpacers {type=boolean, default=true}
 ButtonStyle {type=Class, default=null}
 DebugButtonTypeDisplay {type=Class, default=null}
 DebugTabCount {type=integer, default=3}
 ForceTabIconsCollapsed {type=boolean}
 ForceTabLabelsCollapsed {type=boolean}
 IsScrollable {type=boolean}
 LastSelectedTab {type=CommonButtonBase, default=null}
 MinTabWidth {type=integer}
 NextTabActionPaddingLeft {type=number}
 PreviousTabActionPaddingRight {type=number}
 RefreshGamepad {}
 ScrollableVisible {type=integer}
 ShowRail {type=boolean}
 TabButtonCache {type="array<CommonButtonBase>", default []}
 TabButtonPadding {type=Margin, default=<ltrb 0 0 0 0>}
 TabSize {type=Vector2D, default=<52 52>}
 UseButtonStyle {type=boolean}
]
widgets {
 info {inheritedWidgetCount=0, parentClass=/Script/FortniteUI.FortTabListWidgetBase, rootWidgetClass=&SizeBox, widgetCount=16}
 widgets [
  SizeBox_231 {widgetClassPath=&SizeBox, widget=SizeBox_231, properties {heightOverride=88}}
  HorizontalBox_1 {widgetClassPath=&HorizontalBox, parent=SizeBox_231, widget=HorizontalBox_1, slot {class=SizeBoxSlot}}
  Overlay_Container {
   widgetClassPath=&Overlay
   bIsVariable=true
   parent=HorizontalBox_1
   widget=Overlay_Container
   slot {class=HorizontalBoxSlot, properties {verticalAlignment=VAlign_Center}}
  }
  HorizontalBox_0 {
   widgetClassPath=&HorizontalBox
   parent=Overlay_Container
   widget=HorizontalBox_0
   slot {class=OverlaySlot, properties {verticalAlignment=VAlign_Fill}}
  }
  PreviousTabActionWrapper {
   widgetClassPath=&SizeBox
   bIsVariable=true
   parent=HorizontalBox_0
   widget=PreviousTabActionWrapper
   slot {class=HorizontalBoxSlot}
  }
  PreviousTabAction {
   widgetClassPath=&CommonActionWidget
   bIsVariable=true
   parent=PreviousTabActionWrapper
   widget=PreviousTabAction
   slot {class=SizeBoxSlot}
  }
  PreviousButtonWrapper {
   widgetClassPath=&SizeBox
   bIsVariable=true
   parent=HorizontalBox_0
   widget=PreviousButtonWrapper
   properties {visibility=Collapsed}
   slot {class=HorizontalBoxSlot, properties {verticalAlignment=VAlign_Center, padding=<ltrb 6 0 0 0>}}
  }
  TabButtonBox {
   widgetClassPath=&HorizontalBox
   bIsVariable=true
   parent=HorizontalBox_0
   widget=TabButtonBox
   slot {class=HorizontalBoxSlot, properties {padding=<ltrb 6 0 6 0>, size {sizeRule=Fill, value=1}}}
  }
  RetainerBox_Scrollable {
   widgetClassPath=/Script/UMG.RetainerBox
   bIsVariable=true
   parent=HorizontalBox_0
   widget=RetainerBox_Scrollable
   properties {bRetainRender=false, effectMaterial=&/M_MaskScrollableList}
   slot {class=HorizontalBoxSlot, properties {padding=<ltrb 6 0 6 0>, size {sizeRule=Fill, value=1}}}
  }
  SizeBox_Scrollable {
   widgetClassPath=&SizeBox
   bIsVariable=true
   parent=RetainerBox_Scrollable
   widget=SizeBox_Scrollable
   properties {bOverride_WidthOverride=true, clipping=ClipToBounds, widthOverride=250}
   slot {class=PanelSlot}
  }
  ScrollBox {
   widgetClassPath=&ScrollBox
   bIsVariable=true
   parent=SizeBox_Scrollable
   widget=ScrollBox
   properties {
    allowOverscroll=false
    bAllowRightClickDragScrolling=false
    orientation=Orient_Horizontal
    scrollBarVisibility=Collapsed
    scrollbarPadding=<ltrb 0 0 0 0>
    scrollbarThickness=<0 0>
    widgetBarStyle {
     thickness=0
     draggedThumbImage {
      bIsDynamicallyLoaded=false
      drawAs=NoDrawType
      imageType=FullColor
      mirroring=NoMirror
      resourceName=../../../Engine/Content/Slate/Common/Scrollbar_Thumb.png
      resourceObject=null
      tiling=NoTile
      imageSize=<0 0>
      margin=<ltrb 0.25 0.25 0.25 0.25>
      outlineSettings {
       bUseBrushTransparency=false
       roundingType=HalfHeightRadius
       width=0
       color {colorUseRule=UseColor_Specified, specifiedColor=<rgba 0 0 0 0>}
       cornerRadii=<0 0 0 0>
      }
      tintColor {colorUseRule=UseColor_Specified, specifiedColor=<rgba 1 1 1 1>}
      uVRegion {bIsValid=false, max=<0 0>, min=<0 0>}
     }
     horizontalBackgroundImage {
      bIsDynamicallyLoaded=false
      drawAs=NoDrawType
      imageType=NoImage
      mirroring=NoMirror
      resourceName=None
      resourceObject=null
      tiling=NoTile
      imageSize=<0 0>
      margin=<ltrb 0 0 0 0>
      outlineSettings {
       bUseBrushTransparency=false
       roundingType=HalfHeightRadius
       width=0
       color {colorUseRule=UseColor_Specified, specifiedColor=<rgba 0 0 0 0>}
       cornerRadii=<0 0 0 0>
      }
      tintColor {colorUseRule=UseColor_Specified, specifiedColor=<rgba 1 1 1 1>}
      uVRegion {bIsValid=false, max=<0 0>, min=<0 0>}
     }
     horizontalBottomSlotImage {
      bIsDynamicallyLoaded=false
      drawAs=NoDrawType
      imageType=FullColor
      mirroring=NoMirror
      resourceName=../../../Engine/Content/Slate/Common/Scrollbar_Background_Horizontal.png
      resourceObject=null
      tiling=NoTile
      imageSize=<8 8>
      margin=<ltrb 0 0 0 0>
      outlineSettings {
       bUseBrushTransparency=false
       roundingType=HalfHeightRadius
       width=0
       color {colorUseRule=UseColor_Specified, specifiedColor=<rgba 0 0 0 0>}
       cornerRadii=<0 0 0 0>
      }
      tintColor {colorUseRule=UseColor_Specified, specifiedColor=<rgba 1 1 1 1>}
      uVRegion {bIsValid=false, max=<0 0>, min=<0 0>}
     }
     horizontalTopSlotImage {
      bIsDynamicallyLoaded=false
      drawAs=NoDrawType
      imageType=FullColor
      mirroring=NoMirror
      resourceName=../../../Engine/Content/Slate/Common/Scrollbar_Background_Horizontal.png
      resourceObject=null
      tiling=NoTile
      imageSize=<8 8>
      margin=<ltrb 0 0 0 0>
      outlineSettings {
       bUseBrushTransparency=false
       roundingType=HalfHeightRadius
       width=0
       color {colorUseRule=UseColor_Specified, specifiedColor=<rgba 0 0 0 0>}
       cornerRadii=<0 0 0 0>
      }
      tintColor {colorUseRule=UseColor_Specified, specifiedColor=<rgba 1 1 1 1>}
      uVRegion {bIsValid=false, max=<0 0>, min=<0 0>}
     }
     hoveredThumbImage {
      bIsDynamicallyLoaded=false
      drawAs=NoDrawType
      imageType=FullColor
      mirroring=NoMirror
      resourceName=../../../Engine/Content/Slate/Common/Scrollbar_Thumb.png
      resourceObject=null
      tiling=NoTile
      imageSize=<0 0>
      margin=<ltrb 0.25 0.25 0.25 0.25>
      outlineSettings {
       bUseBrushTransparency=false
       roundingType=HalfHeightRadius
       width=0
       color {colorUseRule=UseColor_Specified, specifiedColor=<rgba 0 0 0 0>}
       cornerRadii=<0 0 0 0>
      }
      tintColor {colorUseRule=UseColor_Specified, specifiedColor=<rgba 1 1 1 1>}
      uVRegion {bIsValid=false, max=<0 0>, min=<0 0>}
     }
     normalThumbImage {
      bIsDynamicallyLoaded=false
      drawAs=NoDrawType
      imageType=FullColor
      mirroring=NoMirror
      resourceName=../../../Engine/Content/Slate/Common/Scrollbar_Thumb.png
      resourceObject=null
      tiling=NoTile
      imageSize=<0 0>
      margin=<ltrb 0.25 0.25 0.25 0.25>
      outlineSettings {
       bUseBrushTransparency=false
       roundingType=HalfHeightRadius
       width=0
       color {colorUseRule=UseColor_Specified, specifiedColor=<rgba 0 0 0 0>}
       cornerRadii=<0 0 0 0>
      }
      tintColor {colorUseRule=UseColor_Specified, specifiedColor=<rgba 1 1 1 1>}
      uVRegion {bIsValid=false, max=<0 0>, min=<0 0>}
     }
     verticalBackgroundImage {
      bIsDynamicallyLoaded=false
      drawAs=NoDrawType
      imageType=NoImage
      mirroring=NoMirror
      resourceName=None
      resourceObject=null
      tiling=NoTile
      imageSize=<0 0>
      margin=<ltrb 0 0 0 0>
      outlineSettings {
       bUseBrushTransparency=false
       roundingType=HalfHeightRadius
       width=0
       color {colorUseRule=UseColor_Specified, specifiedColor=<rgba 0 0 0 0>}
       cornerRadii=<0 0 0 0>
      }
      tintColor {colorUseRule=UseColor_Specified, specifiedColor=<rgba 1 1 1 1>}
      uVRegion {bIsValid=false, max=<0 0>, min=<0 0>}
     }
     verticalBottomSlotImage {
      bIsDynamicallyLoaded=false
      drawAs=Box
      imageType=FullColor
      mirroring=NoMirror
      resourceName=../../../Engine/Content/Slate/Common/Scrollbar_Background_Vertical.png
      resourceObject=null
      tiling=NoTile
      imageSize=<8 8>
      margin=<ltrb 0 0 0 0>
      outlineSettings {
       bUseBrushTransparency=false
       roundingType=HalfHeightRadius
       width=0
       color {colorUseRule=UseColor_Specified, specifiedColor=<rgba 0 0 0 0>}
       cornerRadii=<0 0 0 0>
      }
      tintColor {colorUseRule=UseColor_Specified, specifiedColor=<rgba 1 1 1 1>}
      uVRegion {bIsValid=false, max=<0 0>, min=<0 0>}
     }
     verticalTopSlotImage {
      bIsDynamicallyLoaded=false
      drawAs=NoDrawType
      imageType=FullColor
      mirroring=NoMirror
      resourceName=../../../Engine/Content/Slate/Common/Scrollbar_Background_Vertical.png
      resourceObject=null
      tiling=NoTile
      imageSize=<8 8>
      margin=<ltrb 0 0 0 0>
      outlineSettings {
       bUseBrushTransparency=false
       roundingType=HalfHeightRadius
       width=0
       color {colorUseRule=UseColor_Specified, specifiedColor=<rgba 0 0 0 0>}
       cornerRadii=<0 0 0 0>
      }
      tintColor {colorUseRule=UseColor_Specified, specifiedColor=<rgba 1 1 1 1>}
      uVRegion {bIsValid=false, max=<0 0>, min=<0 0>}
     }
    }
    widgetStyle {
     barThickness=0
     bottomShadowBrush {
      bIsDynamicallyLoaded=false
      drawAs=NoDrawType
      imageType=FullColor
      mirroring=NoMirror
      resourceName=../../../Engine/Content/Slate/Common/ScrollBoxShadowBottom.png
      resourceObject=null
      tiling=NoTile
      imageSize=<16 8>
      margin=<ltrb 0.5 0 0.5 1>
      outlineSettings {
       bUseBrushTransparency=false
       roundingType=HalfHeightRadius
       width=0
       color {colorUseRule=UseColor_Specified, specifiedColor=<rgba 0 0 0 0>}
       cornerRadii=<0 0 0 0>
      }
      tintColor {colorUseRule=UseColor_Specified, specifiedColor=<rgba 1 1 1 1>}
      uVRegion {bIsValid=false, max=<0 0>, min=<0 0>}
     }
     horizontalScrolledContentPadding=<ltrb 0 0 1 0>
     leftShadowBrush {
      bIsDynamicallyLoaded=false
      drawAs=NoDrawType
      imageType=FullColor
      mirroring=NoMirror
      resourceName=../../../Engine/Content/Slate/Common/ScrollBoxShadowLeft.png
      resourceObject=null
      tiling=NoTile
      imageSize=<8 16>
      margin=<ltrb 1 0.5 0 0.5>
      outlineSettings {
       bUseBrushTransparency=false
       roundingType=HalfHeightRadius
       width=0
       color {colorUseRule=UseColor_Specified, specifiedColor=<rgba 0 0 0 0>}
       cornerRadii=<0 0 0 0>
      }
      tintColor {colorUseRule=UseColor_Specified, specifiedColor=<rgba 1 1 1 1>}
      uVRegion {bIsValid=false, max=<0 0>, min=<0 0>}
     }
     rightShadowBrush {
      bIsDynamicallyLoaded=false
      drawAs=NoDrawType
      imageType=FullColor
      mirroring=NoMirror
      resourceName=../../../Engine/Content/Slate/Common/ScrollBoxShadowRight.png
      resourceObject=null
      tiling=NoTile
      imageSize=<8 16>
      margin=<ltrb 0 0.5 1 0.5>
      outlineSettings {
       bUseBrushTransparency=false
       roundingType=HalfHeightRadius
       width=0
       color {colorUseRule=UseColor_Specified, specifiedColor=<rgba 0 0 0 0>}
       cornerRadii=<0 0 0 0>
      }
      tintColor {colorUseRule=UseColor_Specified, specifiedColor=<rgba 1 1 1 1>}
      uVRegion {bIsValid=false, max=<0 0>, min=<0 0>}
     }
     topShadowBrush {
      bIsDynamicallyLoaded=false
      drawAs=NoDrawType
      imageType=FullColor
      mirroring=NoMirror
      resourceName=../../../Engine/Content/Slate/Common/ScrollBoxShadowTop.png
      resourceObject=null
      tiling=NoTile
      imageSize=<16 8>
      margin=<ltrb 0.5 1 0.5 0>
      outlineSettings {
       bUseBrushTransparency=false
       roundingType=HalfHeightRadius
       width=0
       color {colorUseRule=UseColor_Specified, specifiedColor=<rgba 0 0 0 0>}
       cornerRadii=<0 0 0 0>
      }
      tintColor {colorUseRule=UseColor_Specified, specifiedColor=<rgba 1 1 1 1>}
      uVRegion {bIsValid=false, max=<0 0>, min=<0 0>}
     }
     verticalScrolledContentPadding=<ltrb 0 0 0 1>
    }
   }
   slot {class=SizeBoxSlot, properties {horizontalAlignment=HAlign_Center}}
  }
  TabButtonBox_Scrollable {
   widgetClassPath=&HorizontalBox
   bIsVariable=true
   parent=ScrollBox
   widget=TabButtonBox_Scrollable
   slot {class=ScrollBoxSlot, properties {size {sizeRule=Fill, value=1}}}
  }
  NextButtonWrapper {
   widgetClassPath=&SizeBox
   bIsVariable=true
   parent=HorizontalBox_0
   widget=NextButtonWrapper
   properties {visibility=Collapsed}
   slot {class=HorizontalBoxSlot, properties {verticalAlignment=VAlign_Center, padding=<ltrb 0 0 6 0>}}
  }
  NextTabActionWrapper {widgetClassPath=&SizeBox, bIsVariable=true, parent=HorizontalBox_0, widget=NextTabActionWrapper, slot {class=HorizontalBoxSlot}}
  NextTabAction {
   widgetClassPath=&CommonActionWidget
   bIsVariable=true
   parent=NextTabActionWrapper
   widget=NextTabAction
   slot {class=SizeBoxSlot}
  }
  ImageHorizRule {
   widgetClassPath=&Image
   bIsVariable=true
   parent=Overlay_Container
   widget=ImageHorizRule
   properties {
    visibility=Collapsed
    brush {
     bIsDynamicallyLoaded=false
     drawAs=Box
     imageType=NoImage
     mirroring=NoMirror
     resourceName=None
     resourceObject=/Game/UI/Foundation/Textures/Buttons/T-Tab-Secondary-Base-32x64
     tiling=NoTile
     imageSize=<48 88>
     margin=<ltrb 0.49 0.49 0.49 0.49>
     outlineSettings {
      bUseBrushTransparency=false
      roundingType=HalfHeightRadius
      width=0
      color {colorUseRule=UseColor_Specified, specifiedColor=<rgba 0 0 0 0>}
      cornerRadii=<0 0 0 0>
     }
     tintColor {colorUseRule=UseColor_Specified, specifiedColor=<rgba 1 1 1 0.4>}
     uVRegion {bIsValid=false, max=<0 0>, min=<0 0>}
    }
   }
   slot {class=OverlaySlot, properties {horizontalAlignment=HAlign_Fill, verticalAlignment=VAlign_Center, padding=<ltrb 6 0 6 0>}}
  }
 ]
}
